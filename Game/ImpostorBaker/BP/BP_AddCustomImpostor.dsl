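; Editor-tool Blueprint that bakes a custom impostor mesh, presumably into an existing static mesh asset.
; The "3) Create Static Assets" event passes the target material ID, LOD index, asset path and source mesh to a B_GeoScript_ImpostorMeshCustom1 child actor, then calls UpdateMesh and BakeMesh.
; It then loads the existing asset and copies the source mesh's first material into its slot 2.
; Gameplay events and the construction script are empty.
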(event EventBeginPlay)

(event Collision|EventActorBeginOverlap
  (params (OtherActor :type Actor%20Object%20Reference)))

(event EventTick
  (params (DeltaSeconds :type Float%20%28single-precision%29)))

(event Custom|3%29%20Create%20Static%20Assets
  (bind _existingmeshassetname (Variables|Default|GetExistingMeshAssetName))
  (bind _asb_geo_script_impostor_mesh_custom_1 (Utilities|Casting|CastToB_GeoScript_ImpostorMeshCustom1 (Variables|/Script/Engine.ChildActorComponent|GetChildActor (Variables|Default|GetB_GeoScript_ImpostorMesh))))
  (Variables|/Game/ImpostorBaker/BP/B_GeoScript_ImpostorMeshCustom1.B_GeoScript_ImpostorMeshCustom1_C|SetTargetMaterialID (Variables|Default|GetTargetMaterialID) _asb_geo_script_impostor_mesh_custom_1)
  (Variables|/Game/ImpostorBaker/BP/B_GeoScript_ImpostorMeshCustom1.B_GeoScript_ImpostorMeshCustom1_C|SetTargetLODIndex (Variables|Default|GetTargetLODIndex) _asb_geo_script_impostor_mesh_custom_1)
  (Variables|/Game/ImpostorBaker/BP/B_GeoScript_ImpostorMeshCustom1.B_GeoScript_ImpostorMeshCustom1_C|SetExistingStaticMeshPath _existingmeshassetname _asb_geo_script_impostor_mesh_custom_1)
  (Variables|/Game/ImpostorBaker/BP/B_GeoScript_ImpostorMeshCustom1.B_GeoScript_ImpostorMeshCustom1_C|SetFrom%20Static%20Mesh%20Asset (Variables|Default|GetStaticmesh) _asb_geo_script_impostor_mesh_custom_1)
  (Class|BGeoScriptImpostorMeshCustom1|UpdateMesh _asb_geo_script_impostor_mesh_custom_1)
  (Class|BGeoScriptImpostorMeshCustom1|BakeMesh _asb_geo_script_impostor_mesh_custom_1)
  (bind _returnvalue (EditorScripting|Asset|LoadAsset _existingmeshassetname :!class /Script/EditorScriptingUtilities.EditorAssetLibrary))
  (bind _asstatic_mesh (Utilities|Casting|CastToStaticMesh _returnvalue))
  (StaticMesh|SetMaterial _asstatic_mesh 2 (StaticMesh|GetMaterial (Variables|Default|GetStaticmesh) 0)))

(fn ConstructionScript)
