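; Editor-only impostor baker Blueprint that renders a static mesh into sprite atlases and saves them as assets.
; A scene capture steps through a hemi-octahedral grid of views, drawing downsampled frames into per-channel render targets selected through a material parameter collection.
; Normal/depth and base color/custom lighting can be combined, alpha feeds a mesh cutout, and results become textures, a material instance and an impostor mesh.
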(event Custom|1%29%20Clear%20RTs
  ;[1 1 1] Clear RTs
  (CallFunction|ClearRenderTargets))

(event Custom|2%29%20RenderFrames
  (bind _combinedbasecolorrt (Variables|RenderTargets|GetCombinedBaseColorRT))
  ;[1 1 1] Render Impostor Frames
  (Rendering|SetVisibility (Variables|/Script/Engine.StaticMeshActor|GetStaticMeshComponent (Variables|Default|GetStatic%20Mesh%20Actor)) true false)
  (CallFunction|ClearRenderTargets)
  (Rendering|Material|SetScalarParameterValue 0 "/Game/ImpostorBaker/MPC/MPC_ImposterCapture.MPC_ImposterCapture" "Dilation" 1 :!class /Script/Engine.KismetMaterialLibrary)
  (Rendering|Material|SetScalarParameterValue 0 "/Game/ImpostorBaker/MPC/MPC_ImposterCapture.MPC_ImposterCapture" "EnableCapture" :!class /Script/Engine.KismetMaterialLibrary)
  (Rendering|Material|SetScalarParameterValue 0 "/Game/ImpostorBaker/MPC/MPC_ImposterCapture.MPC_ImposterCapture" "CustomLighting" :!class /Script/Engine.KismetMaterialLibrary)
  ;[1 1 1] Capture Render Item
  (for _array_element (Variables|Debug|GetRenderList)
    (if (.Render_13_EBE02B9E4D6EEC154D95B8BEE1D0FD38 (Utilities|Struct|BreakImpostorBakeItem _array_element))
      (Variables|Default|SetCurrentTargetRT (Utilities|Struct|BreakImpostorBakeItem _array_element))
      (Variables|Default|SetDF%20Alpha (.DFAlpha_8_F64D11AF48CB3BF2E57523B5128030CE (Utilities|Struct|BreakImpostorBakeItem _array_element)))
      (Rendering|Material|SetScalarParameterValue 0 "/Game/ImpostorBaker/MPC/MPC_ImposterCapture.MPC_ImposterCapture" (.Name_11_5D34CF8B4F6D130EC53EEF9013333D0C (Utilities|Struct|BreakImpostorBakeItem _array_element)) 1 :!class /Script/Engine.KismetMaterialLibrary)
      (Rendering|Material|SetScalarParameterValue 0 "/Game/ImpostorBaker/MPC/MPC_ImposterCapture.MPC_ImposterCapture" "EnableCapture" 1 :!class /Script/Engine.KismetMaterialLibrary)
      (CallFunction|CaptureImposterGrid)
      (Rendering|Material|SetScalarParameterValue 0 "/Game/ImpostorBaker/MPC/MPC_ImposterCapture.MPC_ImposterCapture" (.Name_11_5D34CF8B4F6D130EC53EEF9013333D0C (Utilities|Struct|BreakImpostorBakeItem _array_element)) :!class /Script/Engine.KismetMaterialLibrary)
      (Rendering|Material|SetScalarParameterValue 0 "/Game/ImpostorBaker/MPC/MPC_ImposterCapture.MPC_ImposterCapture" "EnableCapture" :!class /Script/Engine.KismetMaterialLibrary)
      (Variables|Default|SetDF%20Alpha false)))
  ;[0.2768 0.4998 1] Combined Normal Depth
  (Rendering|DrawMaterialtoRenderTarget 0 (Variables|RenderTargets|GetCombinedNormalRT) (Variables|MIDs|GetCombineNormalsDepthMID))
  ;[1 1 1] Basecolor * Custom Lighting
  (Utilities|FlowControl|Sequence
    (:then_0
      (Utilities|IsValid _combinedbasecolorrt
        (:Is%20Valid
          (Rendering|ClearRenderTarget2D 0 _combinedbasecolorrt)
          (Rendering|DrawMaterialtoRenderTarget 0 _combinedbasecolorrt (Variables|MIDs|GetBaseColorCustomLightingMID)))))
    (:then_1
      ;[1 1 1] Mesh Cutout
      (Utilities|ReadTextureData
        (:Out
          (Utilities|GenerateMeshCutout))))))

(event Custom|Combine%20Basecolor%20CustomLighting
  (bind _combinedbasecolorrt (Variables|RenderTargets|GetCombinedBaseColorRT))
  ;[1 1 1] Basecolor * Custom Lighting
  (Utilities|IsValid _combinedbasecolorrt
    (:Is%20Valid
      (Rendering|ClearRenderTarget2D 0 _combinedbasecolorrt)
      (Rendering|DrawMaterialtoRenderTarget 0 _combinedbasecolorrt (Variables|MIDs|GetBaseColorCustomLightingMID)))))

(event Custom|DrawSingleFrame
  (params (RT :type Texture%20Render%20Target%202D%20Object%20Reference))
  ;[1 1 1] Draw Frame to Atlas
  (bind (_canvas _size _context) (Rendering|BeginDrawCanvastoRenderTarget 0 RT))
  (Canvas|DrawMaterial :self _canvas :RenderMaterial (select (Variables|Default|GetDF%20Alpha) (Variables|MIDs|GetSampleFrame_DFAlphaMID) (Variables|MIDs|GetSampleFrameMID)) :ScreenPosition (* (Math|Vector2D|MakeVector2D (/ (Math|Vector2D|BreakVector2D _size) (Math|Conversions|ToFloat(Integer) (Variables|Default|GetFrames%20XY))) (/ (.Y (Math|Vector2D|BreakVector2D _size)) (Math|Conversions|ToFloat(Integer) (Variables|Default|GetFrames%20XY)))) (Utilities|1dTo2dIndex (Variables|Default|Getcurframe) (Variables|Default|GetFrames%20XY))) :ScreenSize (Math|Vector2D|MakeVector2D (/ (Math|Vector2D|BreakVector2D _size) (Math|Conversions|ToFloat(Integer) (Variables|Default|GetFrames%20XY))) (/ (.Y (Math|Vector2D|BreakVector2D _size)) (Math|Conversions|ToFloat(Integer) (Variables|Default|GetFrames%20XY)))) :CoordinateSize "(X=1.000000,Y=1.000000)" :Rotation 0.0 :PivotPoint "(X=0.500000,Y=0.500000)")
  (Rendering|EndDrawCanvastoRenderTarget 0 _context)
  (if (Utilities|Equal(Object) (Variables|Default|GetCurrentTargetRT) (Variables|RenderTargets|GetBaseColorRT))
    (Rendering|DrawMaterialtoRenderTarget 0 (Variables|RenderTargets|GetCombinedAlphasRT) (Variables|MIDs|GetAddAlphasMID))))

(event Custom|3%29%20Create%20Static%20Assets
  (bind _resolution (Variables|Advanced|GetResolution))
  (bind _returnvalue (Math|Conversions|ToFloat(Integer) _resolution))
  (bind _returnvalue_2 (Math|Float|Log _returnvalue 2))
  (bind _returnvalue_3 (Math|Float|Floor _returnvalue_2))
  (bind _returnvalue_4 (- _returnvalue_3 9))
  (bind _new_asset_mic_name (Variables|Default|GetNew%20Asset%20MIC%20Name))
  (bind _new_asset_path (Variables|Default|GetNew%20Asset%20Path))
  (bind _returnvalue_5 (Utilities|String|Append _new_asset_path _new_asset_mic_name))
  (bind _new_asset_impostormeshname (Variables|Default|GetNew%20Asset%20ImpostorMeshName))
  (bind _new_asset_path_2 (Variables|Default|GetNew%20Asset%20Path))
  (bind _returnvalue_6 (Utilities|String|Append _new_asset_path_2 _new_asset_impostormeshname))
  (bind _new_asset_texture_prefix (Variables|Default|GetNew%20Asset%20Texture%20Prefix))
  (bind _returnvalue_7 (Utilities|String|Append _new_asset_texture_prefix "_"))
  (bind _returnvalue_8 (Math|Integer|Min(Integer) _returnvalue_4 3))
  (bind _returnvalue_9 (Utilities|Select :Index _returnvalue_8 :Option%200 "/Game/ImpostorBaker/Textures/RT_512.RT_512" :Option%201 "/Game/ImpostorBaker/Textures/RT_1024.RT_1024" :Option%202 "/Game/ImpostorBaker/Textures/RT_2048.RT_2048" :Option%203 "/Game/ImpostorBaker/Textures/RT_4096.RT_4096"))
  (bind _new_asset_path_3 (Variables|Default|GetNew%20Asset%20Path))
  (bind _created_mic (Variables|Debug|GetCreated%20MIC))
  (bind _created_mic_2 (Variables|Debug|GetCreated%20MIC))
  ;[1 1 1] Create Static Assets
  (Utilities|Array|Clear (Variables|Debug|GetSavedTextures))
  (Utilities|FlowControl|Sequence
    (:then_0
      (bind _returnvalue_10 (EditorScripting|Asset|DoesAssetExist _returnvalue_5 :!class /Script/EditorScriptingUtilities.EditorAssetLibrary))
      (if _returnvalue_10
        (bind _returnvalue_11 (EditorScripting|Asset|LoadAsset _returnvalue_5 :!class /Script/EditorScriptingUtilities.EditorAssetLibrary))
        (bind _asmaterial_instance_constant (Utilities|Casting|CastToMaterialInstanceConstant _returnvalue_11))
        (Variables|Debug|SetCreated%20MIC _asmaterial_instance_constant)
        ;[1 1 1] Set thsee if Asset Exists
        (Rendering|SetMICScalarParameterEditorOnly _created_mic "Scatter Mask Min" (Variables|MaterialSettings|GetScatter%20Mask%20Min))
        (Rendering|SetMICScalarParameterEditorOnly _created_mic "Scatter Mask Len" (Variables|MaterialSettings|GetScatter%20Mask%20Length))
        (Rendering|SetMICScalarParameterEditorOnly _created_mic "FramesXY" (Math|Conversions|ToFloat(Integer) (Variables|Default|GetFrames%20XY)))
        (Rendering|SetMICScalarParameterEditorOnly _created_mic "Default Mesh Size" (* (Variables|Debug|GetObject%20Radius) 2))
        (Rendering|SetMICVectorParameterEditorOnly _created_mic "Pivot Offset" (Math|Conversions|ToLinearColor(Vector) (Variables|Debug|GetOffset%20Vector)))
        (else
          (bind _returnvalue_12 (Rendering|CreateMICEditorOnly (Variables|MaterialSettings|GetPreview%20Parent%20Material) _returnvalue_5))
          (Variables|Debug|SetCreated%20MIC _returnvalue_12)
          (Utilities|SetMICParameters))))
    (:then_1
      (for _array_element (Variables|Debug|GetRenderList)
        (if (.Save_10_E6E463E1468C1DC63C58EB92C7688B07 (Utilities|Struct|BreakImpostorBakeItem _array_element))
          (CallFunction|ResampleRT (Utilities|Struct|BreakImpostorBakeItem _array_element) _returnvalue_9)
          (bind _returnvalue_13 (EditorScripting|Asset|DoesAssetExist (Utilities|String|Append _new_asset_path_3 (Utilities|String|Append _returnvalue_7 (Utilities|String|ToString(Name) (.Name_11_5D34CF8B4F6D130EC53EEF9013333D0C (Utilities|Struct|BreakImpostorBakeItem _array_element))))) :!class /Script/EditorScriptingUtilities.EditorAssetLibrary))
          (if _returnvalue_13
            (bind _returnvalue_14 (EditorScripting|Asset|LoadAsset (Utilities|String|Append _new_asset_path_3 (Utilities|String|Append _returnvalue_7 (Utilities|String|ToString(Name) (.Name_11_5D34CF8B4F6D130EC53EEF9013333D0C (Utilities|Struct|BreakImpostorBakeItem _array_element))))) :!class /Script/EditorScriptingUtilities.EditorAssetLibrary))
            (bind _astexture_2d (Utilities|Casting|CastToTexture2D _returnvalue_14))
            (EditorScripting|Rendering|ConvertRenderTargettoTexture2DEditorOnly 0 _returnvalue_9 _astexture_2d)
            (Utilities|FlowControl|Sequence
              (:then_0
                (Variables|/Script/Engine.Texture|SetLODGroup (select (Utilities|String|EqualExactly(String) (Utilities|String|ToString(Name) (.Name_11_5D34CF8B4F6D130EC53EEF9013333D0C (Utilities|Struct|BreakImpostorBakeItem _array_element))) "Normal") "TEXTUREGROUP_ImpostorNormalDepth" "TEXTUREGROUP_Impostor") _astexture_2d))
              (:then_1
                (Rendering|SetMICTextureParameterEditorOnly (Variables|Debug|GetCreated%20MIC) (Utilities|String|ToString(Name) (.Name_11_5D34CF8B4F6D130EC53EEF9013333D0C (Utilities|Struct|BreakImpostorBakeItem _array_element))))
                (Utilities|Array|Add (Variables|Debug|GetSavedTextures) (Utilities|Struct|MakeImpostorStaticTextureData (Utilities|Struct|BreakImpostorBakeItem _array_element) _astexture_2d (Utilities|String|ToString(Name) (.Name_11_5D34CF8B4F6D130EC53EEF9013333D0C (Utilities|Struct|BreakImpostorBakeItem _array_element)))))))
            (else
              (bind _returnvalue_15 (EditorScripting|Rendering|RenderTarget2DCreateStaticTexture2DEditorOnly _returnvalue_9 (Utilities|String|Append _new_asset_path_3 (Utilities|String|Append _returnvalue_7 (Utilities|String|ToString(Name) (.Name_11_5D34CF8B4F6D130EC53EEF9013333D0C (Utilities|Struct|BreakImpostorBakeItem _array_element)))))))
              (Utilities|FlowControl|Sequence
                (:then_0
                  (Variables|/Script/Engine.Texture|SetLODGroup (select (Utilities|String|EqualExactly(String) (Utilities|String|ToString(Name) (.Name_11_5D34CF8B4F6D130EC53EEF9013333D0C (Utilities|Struct|BreakImpostorBakeItem _array_element))) "Normal") "TEXTUREGROUP_ImpostorNormalDepth" "TEXTUREGROUP_Impostor") _returnvalue_15))
                (:then_1
                  (Rendering|SetMICTextureParameterEditorOnly (Variables|Debug|GetCreated%20MIC) (Utilities|String|ToString(Name) (.Name_11_5D34CF8B4F6D130EC53EEF9013333D0C (Utilities|Struct|BreakImpostorBakeItem _array_element))) _returnvalue_15)
                  (Utilities|Array|Add (Variables|Debug|GetSavedTextures) (Utilities|Struct|MakeImpostorStaticTextureData (Utilities|Struct|BreakImpostorBakeItem _array_element) _returnvalue_15 (Utilities|String|ToString(Name) (.Name_11_5D34CF8B4F6D130EC53EEF9013333D0C (Utilities|Struct|BreakImpostorBakeItem _array_element)))))))))
          ;;[1 0 0.08227] LOD BIAS
          ;;
          (Variables|/Script/Engine.Texture|SetLODBias (Variables|Default|GetLOD%20Bias) _astexture_2d))))
    (:then_2
      (bind _asb_geo_script_impostor_mesh (Utilities|Casting|CastToB_GeoScript_ImpostorMesh (Variables|/Script/Engine.ChildActorComponent|GetChildActor (Variables|Default|GetB_GeoScript_ImpostorMesh))))
      (Variables|/Game/ImpostorBaker/BP/B_GeoScript_ImpostorMesh.B_GeoScript_ImpostorMesh_C|SetTargetMaterialIDIndex (Variables|Default|GetTargetMaterialIDIndex) _asb_geo_script_impostor_mesh)
      (Variables|/Game/ImpostorBaker/BP/B_GeoScript_ImpostorMesh.B_GeoScript_ImpostorMesh_C|SetTargetLODIndex (Variables|Default|GetTargetLODIndex) _asb_geo_script_impostor_mesh)
      (Variables|/Game/ImpostorBaker/BP/B_GeoScript_ImpostorMesh.B_GeoScript_ImpostorMesh_C|SetNewAssetPathName _returnvalue_6 _asb_geo_script_impostor_mesh)
      (Variables|/Game/ImpostorBaker/BP/B_GeoScript_ImpostorMesh.B_GeoScript_ImpostorMesh_C|SetExistingStaticMeshPath (Variables|Default|GetExistingMeshAssetName) _asb_geo_script_impostor_mesh)
      (Class|BGeoScriptImpostorMesh|UpdateMesh _asb_geo_script_impostor_mesh)
      (Class|BGeoScriptImpostorMesh|BakeMesh _asb_geo_script_impostor_mesh)
      (bind _returnvalue_16 (EditorScripting|Asset|LoadAsset _returnvalue_6 :!class /Script/EditorScriptingUtilities.EditorAssetLibrary))
      (bind _asstatic_mesh (Utilities|Casting|CastToStaticMesh _returnvalue_16))
      (Variables|/Script/Engine.StaticMesh|SetStaticMaterials (Utilities|Array|MakeArray (Utilities|Struct|MakeStaticMaterial _created_mic_2 "ImpostorMat")) _asstatic_mesh)
      (StaticMesh|SetMaterial _asstatic_mesh 0 _created_mic_2))))

(event Custom|4%29%20Overwrite%20Existing%20Assets
  (bind _resolution (Variables|Advanced|GetResolution))
  (bind _returnvalue (Math|Conversions|ToFloat(Integer) _resolution))
  (bind _returnvalue_2 (Math|Float|Log _returnvalue 2))
  (bind _returnvalue_3 (Math|Float|Floor _returnvalue_2))
  (bind _returnvalue_4 (- _returnvalue_3 9))
  (bind _returnvalue_5 (Math|Integer|Min(Integer) _returnvalue_4 3))
  (bind _returnvalue_6 (Utilities|Select :Index _returnvalue_5 :Option%200 "/Game/ImpostorBaker/Textures/RT_512.RT_512" :Option%201 "/Game/ImpostorBaker/Textures/RT_1024.RT_1024" :Option%202 "/Game/ImpostorBaker/Textures/RT_2048.RT_2048" :Option%203 "/Game/ImpostorBaker/Textures/RT_4096.RT_4096"))
  ;[1 1 1] Overwrite Assets
  (for _array_element (Variables|Debug|GetSavedTextures)
    (Utilities|IsValid (Utilities|Struct|BreakImpostorStaticTextureData _array_element)
      (:Is%20Valid
        (Utilities|IsValid (.Texture_9_17E16A0B4CEF418C90DBC290044B325E (Utilities|Struct|BreakImpostorStaticTextureData _array_element))
          (:Is%20Valid
            (CallFunction|ResampleRT (Utilities|Struct|BreakImpostorStaticTextureData _array_element) _returnvalue_6)
            (EditorScripting|Rendering|ConvertRenderTargettoTexture2DEditorOnly 0 _returnvalue_6 (.Texture_9_17E16A0B4CEF418C90DBC290044B325E (Utilities|Struct|BreakImpostorStaticTextureData _array_element))))))))
  (Utilities|SetMICParameters))

(event Custom|SetupSavedList
  (Utilities|Array|Clear (Variables|Debug|GetSavedTextures))
  (for _array_element (Variables|Debug|GetRenderList)
    (if (.Save_10_E6E463E1468C1DC63C58EB92C7688B07 (Utilities|Struct|BreakImpostorBakeItem _array_element))
      (Utilities|Array|Add (Variables|Debug|GetSavedTextures) (Utilities|Struct|MakeImpostorStaticTextureData (Utilities|Struct|BreakImpostorBakeItem _array_element) "" (Utilities|String|ToString(Name) (.Name_11_5D34CF8B4F6D130EC53EEF9013333D0C (Utilities|Struct|BreakImpostorBakeItem _array_element))))))))

(event Custom|Display%20Color%20data
  (bind _returnvalue (Math|Color|MakeColor 0.0 0.0 16 16))
  (bind _combinedalphasrt (Variables|RenderTargets|GetCombinedAlphasRT))
  (bind _returnvalue_2 (Rendering|RenderTargetSampleRectangleEditorOnly _combinedalphasrt _returnvalue))
  ;[1 1 1] Display Color Data
  (Variables|Debug|SetColorData _returnvalue_2)
  (Variables|Debug|SetTextureData (Utilities|Struct|MakeBPTextureData 16 16 _returnvalue_2))
  (Rendering|Debug|FlushPersistentDebugLines 0)
  (bind (_current_point _total_points _0_1_phase _poly_center _x_row _y_row) (Utilities|2DGridExecutionMacro (Variables|Debug|GetDebugTexelSize) (+ (Transformation|GetActorLocation) "0, 0,-2.000000") 16 16)
    (:Per%20Poly%20Execution
      (Rendering|Debug|DrawDebugPoint 0 _poly_center 10 (Utilities|Array|Get(aref) (Variables|Debug|GetColorData) (Utilities|2dTo1dIndex (Utilities|RotatedCoords (Math|Vector2D|MakeVector2D (Math|Conversions|ToFloat(Integer) _x_row) (Math|Conversions|ToFloat(Integer) _y_row)) 16) 16)) 5000))))

(event Custom|asdf
  (Development|PrintString 0 (Utilities|Select :Index (Variables|Default|Geti) :Option%200 "0" :Option%201 "1" :Option%202 "2" :Option%203 "3")))

(event Custom|GenerateMeshCutout
  (Utilities|ReadTextureData
    (:Out
      (Utilities|GenerateMeshCutout))))

(event Custom|Manual%20Corner%20cutout
  (CallFunction|DisplayColorData)
  (Utilities|MakeProcMeshfromPointArray))

(event Custom|Manual%204%20vert%20cutout
  (Utilities|MakeManualQuad))

(fn CacheMeshMaterials
  (bind _static_mesh_actor (Variables|Default|GetStatic%20Mesh%20Actor))
  (bind _staticmeshcomponent (Variables|/Script/Engine.StaticMeshActor|GetStaticMeshComponent _static_mesh_actor))
  (for _index (range (Rendering|Material|GetNumMaterials _staticmeshcomponent))
    (Utilities|Array|Add (Variables|Default|Gettempmats) (Rendering|Material|GetMaterial _staticmeshcomponent _index))))

(fn CaptureImposterGrid
  (bind _static_mesh_actor (Variables|Default|GetStatic%20Mesh%20Actor))
  (bind _staticmeshcomponent (Variables|/Script/Engine.StaticMeshActor|GetStaticMeshComponent _static_mesh_actor))
  (bind (_origin _boxextent _sphereradius) (Collision|GetComponentBounds _staticmeshcomponent))
  (bind _frames_xy (Variables|Default|GetFrames%20XY))
  (bind _returnvalue (Math|Conversions|ToFloat(Integer) _frames_xy))
  (bind _returnvalue_2 (- _returnvalue 1.0))
  (bind _phasey (Variables|Default|GetphaseY))
  (bind _cur_y (Variables|Debug|Getcur%20Y'))
  (bind _returnvalue_3 (Math|Conversions|ToFloat(Integer) _cur_y))
  (bind _returnvalue_4 (+ _returnvalue_3 _phasey))
  (bind _returnvalue_5 (/ _returnvalue_4 _returnvalue_2))
  (bind _returnvalue_6 (Math|Float|Clamp(Float) _returnvalue_5))
  (bind _returnvalue_7 (* _returnvalue_6 2))
  (bind _returnvalue_8 (- _returnvalue_7 1.0))
  (bind _phasex (Variables|Default|GetphaseX))
  (bind _cur_x (Variables|Debug|Getcur%20X'))
  (bind _returnvalue_9 (Math|Conversions|ToFloat(Integer) _cur_x))
  (bind _returnvalue_10 (+ _returnvalue_9 _phasex))
  (bind _returnvalue_11 (/ _returnvalue_10 _returnvalue_2))
  (bind _returnvalue_12 (Math|Float|Clamp(Float) _returnvalue_11))
  (bind _returnvalue_13 (* _returnvalue_12 2))
  (bind _returnvalue_14 (- _returnvalue_13 1.0))
  (bind _returnvalue_15 (Math|Vector2D|MakeVector2D _returnvalue_14 _returnvalue_8))
  (bind _unit_vector (Class|HemiOctahedra|HemiOctahedrontoUnitVector _returnvalue_15))
  (bind _frames_xy_2 (Variables|Default|GetFrames%20XY))
  (bind _returnvalue_16 (Math|Conversions|ToFloat(Integer) _frames_xy_2))
  (bind _returnvalue_17 (- _returnvalue_16 1.0))
  (bind _cur_y_2 (Variables|Debug|Getcur%20Y'))
  (bind _returnvalue_18 (Math|Conversions|ToFloat(Integer) _cur_y_2))
  (bind _returnvalue_19 (/ _returnvalue_18 _returnvalue_17))
  (bind _returnvalue_20 (Math|Float|Clamp(Float) _returnvalue_19))
  (bind _returnvalue_21 (* _returnvalue_20 2))
  (bind _returnvalue_22 (- _returnvalue_21 1.0))
  (bind _cur_x_2 (Variables|Debug|Getcur%20X'))
  (bind _returnvalue_23 (Math|Conversions|ToFloat(Integer) _cur_x_2))
  (bind _returnvalue_24 (/ _returnvalue_23 _returnvalue_17))
  (bind _returnvalue_25 (Math|Float|Clamp(Float) _returnvalue_24))
  (bind _returnvalue_26 (* _returnvalue_25 2))
  (bind _returnvalue_27 (- _returnvalue_26 1.0))
  (bind _returnvalue_28 (Math|Vector2D|MakeVector2D _returnvalue_27 _returnvalue_22))
  (bind _unit_vector_2 (Class|HemiOctahedra|HemiOctahedrontoUnitVector _returnvalue_28))
  (bind _returnvalue_29 (* _unit_vector_2 -1))
  (bind _returnvalue_30 (Math|Conversions|RotationFromXVector _returnvalue_29))
  (bind (_x _y _z) (Math|Rotator|GetAxes _returnvalue_30))
  (bind _substeps (Variables|Debug|GetSubSteps))
  (bind _returnvalue_31 (- _substeps 1))
  (bind _returnvalue_32 (Math|Conversions|ToFloat(Integer) _returnvalue_31))
  (bind (_origin_2 _boxextent_2 _sphereradius_2) (Collision|GetComponentBounds (Variables|/Script/Engine.StaticMeshActor|GetStaticMeshComponent (Variables|Default|GetStatic%20Mesh%20Actor))))
  (CallFunction|SetShowFlagsforCapture)
  (if (Utilities|Equal(Object) (Variables|Default|GetCurrentTargetRT) (Variables|RenderTargets|GetCustomLightingRT))
    (Utilities|DisplayCustomLighting
      (:out
        _)))
  (Variables|Default|Setcurframe 0)
  (for _index (range (Variables|Default|GetFrames%20XY))
    (Variables|Debug|Setcur%20Y' _index)
    (for _index_2 (range (Variables|Default|GetFrames%20XY))
      (Variables|Debug|Setcur%20X' _index_2)
      (Utilities|FlowControl|Sequence
        (:then_0
          (if (and (> (Variables|Debug|GetSubSteps) 0) (Variables|Default|GetDilation%20Passes))
            (for _index_3 (range _substeps)
              (Variables|Default|SetphaseX (select (Variables|Debug|Get-X%20Pass) (* (* (- 1 (/ (Math|Conversions|ToFloat(Integer) _index_3) _returnvalue_32)) (Math|Conversions|ToFloat(Boolean) (Variables|Debug|GetX%20pass))) -1) (* (- 1 (/ (Math|Conversions|ToFloat(Integer) _index_3) _returnvalue_32)) (Math|Conversions|ToFloat(Boolean) (Variables|Debug|GetX%20pass)))))
              (Variables|Default|SetphaseY (select (Variables|Debug|Get-Y%20Pass) (* (* (- 1 (/ (Math|Conversions|ToFloat(Integer) _index_3) _returnvalue_32)) (Math|Conversions|ToFloat(Boolean) (Variables|Debug|GetY%20pass))) -1) (* (- 1 (/ (Math|Conversions|ToFloat(Integer) _index_3) _returnvalue_32)) (Math|Conversions|ToFloat(Boolean) (Variables|Debug|GetY%20pass)))))
              (Transformation|SetWorldLocation (Variables|Default|GetSceneCaptureComponent2D) (+ _origin_2 (* _unit_vector (select (Variables|Advanced|GetOrtho) (* _sphereradius_2 1.25) (Variables|Advanced|GetCamera%20Distance)))) false false)
              (Transformation|SetWorldRotation (Variables|Default|GetSceneCaptureComponent2D) (Math|Conversions|RotationFromXVector (* _unit_vector -1)) false false)
              (Rendering|Material|SetVectorParameterValue 0 "/Game/ImpostorBaker/MPC/MPC_ImposterCapture.MPC_ImposterCapture" "Origin" (Math|Conversions|ToLinearColor(Vector) _origin) :!class /Script/Engine.KismetMaterialLibrary)
              (Rendering|Material|SetScalarParameterValue 0 "/Game/ImpostorBaker/MPC/MPC_ImposterCapture.MPC_ImposterCapture" "R" _sphereradius :!class /Script/Engine.KismetMaterialLibrary)
              (Rendering|Material|SetVectorParameterValue 0 "/Game/ImpostorBaker/MPC/MPC_ImposterCapture.MPC_ImposterCapture" "X" (Math|Conversions|ToLinearColor(Vector) _y) :!class /Script/Engine.KismetMaterialLibrary)
              (Rendering|Material|SetVectorParameterValue 0 "/Game/ImpostorBaker/MPC/MPC_ImposterCapture.MPC_ImposterCapture" "Y" (Math|Conversions|ToLinearColor(Vector) (* _z -1)) :!class /Script/Engine.KismetMaterialLibrary)
              (Rendering|Material|SetVectorParameterValue 0 "/Game/ImpostorBaker/MPC/MPC_ImposterCapture.MPC_ImposterCapture" "Z" (Math|Conversions|ToLinearColor(Vector) _x) :!class /Script/Engine.KismetMaterialLibrary)
              (Rendering|SceneCapture|CaptureScene (Variables|Default|GetSceneCaptureComponent2D) :!class /Script/Engine.SceneCaptureComponent2D)
              (Utilities|DownsampleRT)
              (CallFunction|DrawSingleFrame (Variables|Default|GetCurrentTargetRT)))))
        (:then_1
          (Math|Integer|IncrementInt (Variables|Default|Getcurframe))))))
  (Utilities|RestoreLightingState))

(fn ClearRenderTargets
  (for _array_element (Utilities|Array|MakeArray (Variables|RenderTargets|GetLitRT) (Variables|RenderTargets|GetBaseColorRT) (Variables|RenderTargets|GetSpecularRT) (Variables|RenderTargets|GetRoughnessRT) (Variables|RenderTargets|GetSubsurfaceRT) (Variables|RenderTargets|GetDepthRT) (Variables|RenderTargets|GetWorldNormalRT) (Variables|RenderTargets|GetCustomLightingRT) (Variables|RenderTargets|GetCombinedNormalRT) (Variables|RenderTargets|GetCombinedBaseColorRT) (Variables|RenderTargets|GetCombinedAlphasRT) (Variables|RenderTargets|GetSceneCaptureRT) (Variables|RenderTargets|GetSceneCaptureRT_mip1) (Variables|RenderTargets|GetSceneCaptureRT_mip2) (Variables|RenderTargets|GetSceneCaptureRT_mip3) (Variables|RenderTargets|GetSceneCaptureRT_mip4) (Variables|RenderTargets|GetSceneCaptureRT_mip5) (Variables|RenderTargets|GetSceneCaptureRT_mip6))
    (Rendering|ClearRenderTarget2D 0 _array_element)))

(fn CreateRenderTargets
  (bind _scene_capture_resolution (Variables|Advanced|GetScene%20Capture%20Resolution))
  (bind _returnvalue (Math|Conversions|ToFloat(Integer) _scene_capture_resolution))
  (bind _returnvalue_2 (Math|Float|Log _returnvalue 2))
  (bind _returnvalue_3 (Math|Float|Floor _returnvalue_2))
  (bind _returnvalue_4 (Math|Integer|Clamp(Integer) _returnvalue_3 7 11))
  (Variables|Advanced|SetResolution (Math|Float|Truncate (Math|Float|Power 2 (Math|Conversions|ToFloat(Integer) (Math|Float|Floor (Math|Float|Log (Math|Conversions|ToFloat(Integer) (Variables|Advanced|GetResolution)) 2))))))
  (Variables|Advanced|SetScene%20Capture%20Resolution (Math|Float|Truncate (Math|Float|Power 2 (Math|Conversions|ToFloat(Integer) _returnvalue_4))))
  (Variables|Advanced|SetSceneCaptureMips _returnvalue_4)
  (Utilities|CheckandCreateRT (Variables|RenderTargets|GetSceneCaptureRT) (Variables|Advanced|GetScene%20Capture%20Resolution) (Variables|Debug|GetCachedSceneCaptureRes) true
    (:Out
      (Utilities|CheckandCreateRT (Variables|RenderTargets|GetSceneCaptureRT_mip1) (/ (Variables|Advanced|GetScene%20Capture%20Resolution) 2) (/ (Variables|Debug|GetCachedSceneCaptureRes) 2) true
        (:Out
          (Utilities|CheckandCreateRT (Variables|RenderTargets|GetSceneCaptureRT_mip2) (/ (Variables|Advanced|GetScene%20Capture%20Resolution) 4) (/ (Variables|Debug|GetCachedSceneCaptureRes) 4) true
            (:Out
              (Utilities|CheckandCreateRT (Variables|RenderTargets|GetSceneCaptureRT_mip3) (/ (Variables|Advanced|GetScene%20Capture%20Resolution) 8) (/ (Variables|Debug|GetCachedSceneCaptureRes) 8) true
                (:Out
                  (Utilities|CheckandCreateRT (Variables|RenderTargets|GetSceneCaptureRT_mip4) (/ (Variables|Advanced|GetScene%20Capture%20Resolution) 16) (/ (Variables|Debug|GetCachedSceneCaptureRes) 16) true
                    (:Out
                      (Utilities|CheckandCreateRT (Variables|RenderTargets|GetSceneCaptureRT_mip5) (/ (Variables|Advanced|GetScene%20Capture%20Resolution) 32) (/ (Variables|Debug|GetCachedSceneCaptureRes) 32) true
                        (:Out
                          (Utilities|CheckandCreateRT (Variables|RenderTargets|GetSceneCaptureRT_mip6) (/ (Variables|Advanced|GetScene%20Capture%20Resolution) 64) (/ (Variables|Debug|GetCachedSceneCaptureRes) 64) true
                            (:Out
                              (Variables|Debug|SetCachedResolution (Variables|Advanced|GetResolution))
                              (Variables|Debug|SetCachedSceneCaptureRes (Variables|Advanced|GetScene%20Capture%20Resolution))
                              (Variables|Debug|SetCachedFramesXY (Variables|Default|GetFrames%20XY)))))))))))))))))

(fn ForceShaderCompilebeforebake
  (bind _returnvalue (AddComponent|NODE_AddStaticMeshComponent-0 :RelativeTransform (Math|Transform|MakeTransform "-1000.000000, 0, 0")))
  (Rendering|Material|SetMaterial _returnvalue 0 "/Game/Environments/Asteria/Foliage/Trees/MasterMaterials/M_Asteria_Tree_ImpostorBake_Leaves.M_Asteria_Tree_ImpostorBake_Leaves")
  (bind _returnvalue_2 (AddComponent|NODE_AddStaticMeshComponent-1 :RelativeTransform (Math|Transform|MakeTransform "-1000.000000, 0, 0")))
  (Rendering|Material|SetMaterial _returnvalue_2 0 "/Game/Environments/Asteria/Foliage/Trees/MasterMaterials/M_Asteria_Tree_ImpostorBake_Leaves_Trunk.M_Asteria_Tree_ImpostorBake_Leaves_Trunk")
  (bind _returnvalue_3 (Rendering|Material|CreateDynamicMaterialInstance 0 (Variables|MaterialSettings|GetPreview%20Parent%20Material) "None" "None" :!class /Script/Engine.KismetMaterialLibrary))
  (bind _returnvalue_4 (AddComponent|NODE_AddStaticMeshComponent-3 :RelativeTransform (Math|Transform|MakeTransform "-1000.000000, 0, 0")))
  (Rendering|Material|SetMaterial _returnvalue_4 0 _returnvalue_3)
  (Rendering|SetVisibility _returnvalue_4 false false)
  (bind _returnvalue_5 (Rendering|Material|CreateDynamicMaterialInstance 0 "/Game/ImpostorBaker/Materials/Generation/M_BaseColor_CustomLighting.M_BaseColor_CustomLighting" "None" "None" :!class /Script/Engine.KismetMaterialLibrary))
  (bind _returnvalue_6 (AddComponent|NODE_AddStaticMeshComponent-4 :RelativeTransform (Math|Transform|MakeTransform "-1000.000000, 0, 0")))
  (Rendering|Material|SetMaterial _returnvalue_6 0 _returnvalue_5)
  (bind _returnvalue_7 (Rendering|Material|CreateDynamicMaterialInstance 0 "/Game/ImpostorBaker/Materials/Generation/M_CombineNormalsDepth.M_CombineNormalsDepth" "None" "None" :!class /Script/Engine.KismetMaterialLibrary))
  (bind _returnvalue_8 (AddComponent|NODE_AddStaticMeshComponent-5 :RelativeTransform (Math|Transform|MakeTransform "-1000.000000, 0, 0")))
  (Rendering|Material|SetMaterial _returnvalue_8 0 _returnvalue_7)
  (bind _returnvalue_9 (Rendering|Material|CreateDynamicMaterialInstance 0 "/Game/ImpostorBaker/Materials/Generation/M_Sample_Frame.M_Sample_Frame" "None" "None" :!class /Script/Engine.KismetMaterialLibrary))
  (bind _returnvalue_10 (AddComponent|NODE_AddStaticMeshComponent-6 :RelativeTransform (Math|Transform|MakeTransform "-1000.000000, 0, 0")))
  (Rendering|Material|SetMaterial _returnvalue_10 0 _returnvalue_9)
  (bind _returnvalue_11 (Rendering|Material|CreateDynamicMaterialInstance 0 "/Game/ImpostorBaker/Materials/Generation/M_Sample_Frame_DFAlpha.M_Sample_Frame_DFAlpha" "None" "None" :!class /Script/Engine.KismetMaterialLibrary))
  (bind _returnvalue_12 (AddComponent|NODE_AddStaticMeshComponent-7 :RelativeTransform (Math|Transform|MakeTransform "-1000.000000, 0, 0")))
  (Rendering|Material|SetMaterial _returnvalue_12 0 _returnvalue_11)
  (bind _returnvalue_13 (Rendering|Material|CreateDynamicMaterialInstance 0 "/Game/ImpostorBaker/Materials/Generation/M_AddAlpha.M_AddAlpha" "None" "None" :!class /Script/Engine.KismetMaterialLibrary))
  (bind _returnvalue_14 (AddComponent|NODE_AddStaticMeshComponent-8 :RelativeTransform (Math|Transform|MakeTransform "-1000.000000, 0, 0")))
  (Rendering|Material|SetMaterial _returnvalue_14 0 _returnvalue_13)
  (bind _returnvalue_15 (Rendering|Material|CreateDynamicMaterialInstance 0 "/Game/ImpostorBaker/Materials/Generation/M_Resample_RT.M_Resample_RT" "None" "None" :!class /Script/Engine.KismetMaterialLibrary))
  (bind _returnvalue_16 (AddComponent|NODE_AddStaticMeshComponent-9 :RelativeTransform (Math|Transform|MakeTransform "-1000.000000, 0, 0")))
  (Rendering|Material|SetMaterial _returnvalue_16 0 _returnvalue_15)
  (bind _returnvalue_17 (Rendering|Material|CreateDynamicMaterialInstance 0 "/Game/ImpostorBaker/Materials/M_Imposter_Depth_01.M_Imposter_Depth_01" "None" "None" :!class /Script/Engine.KismetMaterialLibrary))
  (bind _returnvalue_18 (AddComponent|NODE_AddStaticMeshComponent-10 :RelativeTransform (Math|Transform|MakeTransform "-1000.000000, 0, 0")))
  (Rendering|Material|SetMaterial _returnvalue_18 0 _returnvalue_17)
  (bind _returnvalue_19 (Rendering|Material|CreateDynamicMaterialInstance 0 "/Game/ImpostorBaker/Materials/Generation/M_ViewFrame_UV.M_ViewFrame_UV" "None" "None" :!class /Script/Engine.KismetMaterialLibrary))
  (bind _returnvalue_20 (AddComponent|NODE_AddStaticMeshComponent-11 :RelativeTransform (Math|Transform|MakeTransform "-1000.000000, 0, 0")))
  (Rendering|Material|SetMaterial _returnvalue_20 0 _returnvalue_19))

(fn RadialTrace
  (params (Directions :type Integer))
  (bind _returnvalue (- Directions 1))
  (for _index (range Directions)
    (LocalVariables|SetCurDirection (Math|Conversions|ToVector2D(Vector) (Math|Vector|RotateVectorAroundAxis (Math|Conversions|ToVector(Vector2D) (LocalVariables|GetStartDirection)) (* (/ (Math|Conversions|ToFloat(Integer) _index) (+ (Math|Conversions|ToFloat(Integer) _returnvalue) 1.0)) 360) "0, 0,1.000000")))
    (CallFunction|TracefromCenter (LocalVariables|GetCurDirection))))

(fn ResampleRT
  (params (Source%20RT :type Texture%20Render%20Target%202D%20Object%20Reference) (Dest%20RT :type Texture%20Render%20Target%202D%20Object%20Reference))
  (Rendering|Material|SetTextureParameterValue (Variables|MIDs|GetResample%20MID) "RT" Source%20RT)
  (Rendering|ClearRenderTarget2D 0 Dest%20RT)
  (Rendering|DrawMaterialtoRenderTarget 0 Dest%20RT (Variables|MIDs|GetResample%20MID)))

(fn SetShowFlagsForCapture
  (bind _scenecapturecomponent2d (Variables|Default|GetSceneCaptureComponent2D))
  (bind _newshowflagsettings (LocalVariables|GetNewShowflagSettings))
  (for _array_element (Variables|/Script/Engine.SceneCaptureComponent|GetShowFlagSettings _scenecapturecomponent2d)
    (Utilities|Array|Add (LocalVariables|GetNewShowflagSettings) _array_element))
  (Utilities|Array|Add _newshowflagsettings (Utilities|Struct|MakeEngineShowFlagsSetting "Atmosphere"))
  (Variables|/Script/Engine.SceneCaptureComponent|SetShowFlagSettings _newshowflagsettings _scenecapturecomponent2d))

(fn SetupRTandBakeList
  (bind _frames_xy (Variables|Default|GetFrames%20XY))
  (bind _resolution (Variables|Advanced|GetResolution))
  (bind _returnvalue (/ _resolution _frames_xy))
  (bind _returnvalue_2 (Math|Conversions|ToFloat(Integer) _returnvalue))
  (bind _returnvalue_3 (Math|Float|Log _returnvalue_2 2))
  (bind _returnvalue_4 (Math|Float|Floor _returnvalue_3))
  (bind _returnvalue_5 (Math|Integer|Clamp(Integer) _returnvalue_4 0 12))
  (bind _scene_capture_resolution (Variables|Advanced|GetScene%20Capture%20Resolution))
  (bind _returnvalue_6 (Math|Conversions|ToFloat(Integer) _scene_capture_resolution))
  (bind _returnvalue_7 (Math|Float|Log _returnvalue_6 2))
  (bind _returnvalue_8 (Math|Float|Floor _returnvalue_7))
  (bind _returnvalue_9 (Math|Integer|Clamp(Integer) _returnvalue_8 7 11))
  (bind _basecolor (Variables|Default|GetBaseColor))
  (bind _specular (Variables|Default|GetSpecular))
  (bind _lit (Variables|Default|GetLit))
  (bind _roughness (Variables|Default|GetRoughness))
  (bind _subsurface (Variables|Default|GetSubsurface))
  (bind _worldnormal (Variables|Default|GetWorldNormal))
  (bind _customlighting (Variables|Default|GetCustomLighting))
  (bind _depth_map (Variables|Default|GetDepth%20Map))
  (Utilities|Array|Clear (Variables|Debug|GetRenderList))
  ;[1 1 1] Resolution Setup
  (Variables|Advanced|SetResolution (Math|Float|Truncate (Math|Float|Power 2 (Math|Conversions|ToFloat(Integer) (Math|Integer|Clamp(Integer) (Math|Float|Floor (Math|Float|Log (Math|Conversions|ToFloat(Integer) (Variables|Advanced|GetResolution)) 2)) 9 13)))))
  (Variables|Advanced|SetScene%20Capture%20Resolution (Math|Float|Truncate (Math|Float|Power 2 (Math|Conversions|ToFloat(Integer) _returnvalue_9))))
  (Variables|Advanced|SetSceneCaptureMips _returnvalue_9)
  (Variables|Advanced|SetMipTarget (- _returnvalue_9 4))
  (Variables|Advanced|SetFinalSubframeMips _returnvalue_5)
  (Variables|Advanced|SetMipTargetForDownsample (- (Variables|Advanced|GetSceneCaptureMips) _returnvalue_5))
  (Utilities|FlowControl|Sequence
    (:then_0
      (Utilities|CheckandCreateRT (Variables|RenderTargets|GetLitRT) (Variables|Advanced|GetResolution) (Variables|Debug|GetCachedResolution) _lit
        (:Out
          (if _lit
            (Utilities|Array|Add (Variables|Debug|GetRenderList) (Utilities|Struct|MakeImpostorBakeItem (Variables|RenderTargets|GetLitRT) "Lit" false true true))))))
    (:then_2
      (Utilities|CheckandCreateRT (Variables|RenderTargets|GetBaseColorRT) (Variables|Advanced|GetResolution) (Variables|Debug|GetCachedResolution) _basecolor
        (:Out
          (if _basecolor
            (Utilities|Array|Add (Variables|Debug|GetRenderList) (Utilities|Struct|MakeImpostorBakeItem (Variables|RenderTargets|GetBaseColorRT) "BaseColor" (Variables|Default|GetUse%20Distance%20Field%20Alpha) (not (and (Variables|Default|GetCombine%20Customlighting%20and%20Basecolor) (Variables|Default|GetCustomLighting))) true))))))
    (:then_4
      (Utilities|CheckandCreateRT (Variables|RenderTargets|GetSpecularRT) (Variables|Advanced|GetResolution) (Variables|Debug|GetCachedResolution) _specular
        (:Out
          (if _specular
            (Utilities|Array|Add (Variables|Debug|GetRenderList) (Utilities|Struct|MakeImpostorBakeItem (Variables|RenderTargets|GetSpecularRT) "Specular" false true true))))))
    (:then_6
      (Utilities|CheckandCreateRT (Variables|RenderTargets|GetRoughnessRT) (Variables|Advanced|GetResolution) (Variables|Debug|GetCachedResolution) _roughness
        (:Out
          (if _roughness
            (Utilities|Array|Add (Variables|Debug|GetRenderList) (Utilities|Struct|MakeImpostorBakeItem (Variables|RenderTargets|GetRoughnessRT) "Roughness" false true true))))))
    (:then_8
      (Utilities|CheckandCreateRT (Variables|RenderTargets|GetSubsurfaceRT) (Variables|Advanced|GetResolution) (Variables|Debug|GetCachedResolution) _subsurface
        (:Out
          (if _subsurface
            (Utilities|Array|Add (Variables|Debug|GetRenderList) (Utilities|Struct|MakeImpostorBakeItem (Variables|RenderTargets|GetSubsurfaceRT) "Subsurface" false true true))))))
    (:then_10
      (Utilities|CheckandCreateRT (Variables|RenderTargets|GetWorldNormalRT) (Variables|Advanced|GetResolution) (Variables|Debug|GetCachedResolution) _worldnormal
        (:Out
          (if _worldnormal
            (Utilities|Array|Add (Variables|Debug|GetRenderList) (Utilities|Struct|MakeImpostorBakeItem (Variables|RenderTargets|GetWorldNormalRT) "Normal" false (not (Variables|Default|GetCombine%20Normal%20and%20Depth)) true))))))
    (:then_12
      (Utilities|CheckandCreateRT (Variables|RenderTargets|GetDepthRT) (Variables|Advanced|GetResolution) (Variables|Debug|GetCachedResolution) _depth_map
        (:Out
          (if _depth_map
            (Utilities|Array|Add (Variables|Debug|GetRenderList) (Utilities|Struct|MakeImpostorBakeItem (Variables|RenderTargets|GetDepthRT) "Depth" false (not (and (Variables|Default|GetWorldNormal) (Variables|Default|GetCombine%20Normal%20and%20Depth))) true))
            (if (and (Variables|Default|GetWorldNormal) (Variables|Default|GetCombine%20Normal%20and%20Depth))
              (Utilities|CheckandCreateRT (Variables|RenderTargets|GetCombinedNormalRT) (Variables|Advanced|GetResolution) (Variables|Debug|GetCachedResolution) true
                (:Out
                  (Utilities|Array|Add (Variables|Debug|GetRenderList) (Utilities|Struct|MakeImpostorBakeItem (Variables|RenderTargets|GetCombinedNormalRT) "Normal" false true)))))))))
    (:then_14
      (Utilities|CheckandCreateRT (Variables|RenderTargets|GetCustomLightingRT) (Variables|Advanced|GetResolution) (Variables|Debug|GetCachedResolution) _customlighting
        (:Out
          (if _customlighting
            (Utilities|Array|Add (Variables|Debug|GetRenderList) (Utilities|Struct|MakeImpostorBakeItem (Variables|RenderTargets|GetCustomLightingRT) "CustomLighting" false (not (and (Variables|Default|GetBaseColor) (Variables|Default|GetCombine%20Customlighting%20and%20Basecolor))) true))
            (if (and (Variables|Default|GetBaseColor) (Variables|Default|GetCombine%20Customlighting%20and%20Basecolor))
              (Utilities|CheckandCreateRT (Variables|RenderTargets|GetCombinedBaseColorRT) (Variables|Advanced|GetResolution) (Variables|Debug|GetCachedResolution) true
                (:Out
                  (Utilities|Array|Add (Variables|Debug|GetRenderList) (Utilities|Struct|MakeImpostorBakeItem (Variables|RenderTargets|GetCombinedBaseColorRT) "BaseColor" false true)))))))))
    (:then_16
      ;[1 1 1] Scene Capture RTs
      (Utilities|CheckandCreateRT (Variables|RenderTargets|GetSceneCaptureRT) (Variables|Advanced|GetScene%20Capture%20Resolution) (Variables|Debug|GetCachedSceneCaptureRes) true
        (:Out
          (Utilities|CheckandCreateRT (Variables|RenderTargets|GetSceneCaptureRT_mip1) (/ (Variables|Advanced|GetScene%20Capture%20Resolution) 2) (/ (Variables|Debug|GetCachedSceneCaptureRes) 2) true
            (:Out
              (Utilities|CheckandCreateRT (Variables|RenderTargets|GetSceneCaptureRT_mip2) (/ (Variables|Advanced|GetScene%20Capture%20Resolution) 4) (/ (Variables|Debug|GetCachedSceneCaptureRes) 4) true
                (:Out
                  (Utilities|CheckandCreateRT (Variables|RenderTargets|GetSceneCaptureRT_mip3) (/ (Variables|Advanced|GetScene%20Capture%20Resolution) 8) (/ (Variables|Debug|GetCachedSceneCaptureRes) 8) true
                    (:Out
                      (Utilities|CheckandCreateRT (Variables|RenderTargets|GetSceneCaptureRT_mip4) (/ (Variables|Advanced|GetScene%20Capture%20Resolution) 16) (/ (Variables|Debug|GetCachedSceneCaptureRes) 16) true
                        (:Out
                          (Utilities|CheckandCreateRT (Variables|RenderTargets|GetSceneCaptureRT_mip5) (/ (Variables|Advanced|GetScene%20Capture%20Resolution) 32) (/ (Variables|Debug|GetCachedSceneCaptureRes) 32) true
                            (:Out
                              (Utilities|CheckandCreateRT (Variables|RenderTargets|GetSceneCaptureRT_mip6) (/ (Variables|Advanced|GetScene%20Capture%20Resolution) 64) (/ (Variables|Debug|GetCachedSceneCaptureRes) 64) true
                                (:Out
                                  ;[1 1 1] Value Caching to reuse RTs between Construction Script runs
                                  (Variables|Debug|SetCachedResolution (Variables|Advanced|GetResolution))
                                  (Variables|Debug|SetCachedSceneCaptureRes (Variables|Advanced|GetScene%20Capture%20Resolution))
                                  (Variables|Debug|SetCachedFramesXY (Variables|Default|GetFrames%20XY))
                                  (Variables|Default|SetMipChain (Utilities|Array|MakeArray (Variables|RenderTargets|GetSceneCaptureRT) (Variables|RenderTargets|GetSceneCaptureRT_mip1) (Variables|RenderTargets|GetSceneCaptureRT_mip2) (Variables|RenderTargets|GetSceneCaptureRT_mip3) (Variables|RenderTargets|GetSceneCaptureRT_mip4) (Variables|RenderTargets|GetSceneCaptureRT_mip5) (Variables|RenderTargets|GetSceneCaptureRT_mip6)))
                                  (Utilities|CheckandCreateRT (Variables|RenderTargets|GetCombinedAlphasRT) (+ 16 0) (+ 16 0) true
                                    (:Out
                                      (Variables|Advanced|Setresult (Variables|/Script/Engine.TextureRenderTarget2D|GetSizeX (Utilities|Array|Get(acopy) (Variables|Default|GetMipChain) (Math|Integer|Min(Integer) (Utilities|Array|LastIndex (Variables|Default|GetMipChain)) (Variables|Advanced|GetMipTarget)))))
                                      (Variables|Advanced|SetresultDownsample (Variables|/Script/Engine.TextureRenderTarget2D|GetSizeX (Utilities|Array|Get(acopy) (Variables|Default|GetMipChain) (Math|Integer|Min(Integer) (Utilities|Array|LastIndex (Variables|Default|GetMipChain)) (Variables|Advanced|GetMipTargetForDownsample))))))))))))))))))))))))

(fn SetupAll
  (bind _preview_customlighting (Variables|CustomLighting|GetPreview%20CustomLighting))
  (bind _returnvalue (Math|Conversions|ToFloat(Boolean) _preview_customlighting))
  (bind _static_mesh_actor (Variables|Default|GetStatic%20Mesh%20Actor))
  (bind _staticmeshcomponent (Variables|/Script/Engine.StaticMeshActor|GetStaticMeshComponent _static_mesh_actor))
  (bind (_origin _boxextent _sphereradius) (Collision|GetComponentBounds _staticmeshcomponent))
  (bind _upward_bias (Variables|CustomLighting|GetUpward%20Bias))
  (bind _frames_xy (Variables|Default|GetFrames%20XY))
  (bind _returnvalue_2 (Math|Conversions|ToFloat(Integer) _frames_xy))
  (bind _returnvalue_3 (- _returnvalue_2 1.0))
  (bind _cury (Variables|GetcurY))
  (bind _returnvalue_4 (Math|Conversions|ToFloat(Integer) _cury))
  (bind _returnvalue_5 (/ _returnvalue_4 _returnvalue_3))
  (bind _returnvalue_6 (* _returnvalue_5 2))
  (bind _returnvalue_7 (- _returnvalue_6 1.0))
  (bind _curx (Variables|GetcurX))
  (bind _returnvalue_8 (Math|Conversions|ToFloat(Integer) _curx))
  (bind _returnvalue_9 (/ _returnvalue_8 _returnvalue_3))
  (bind _returnvalue_10 (* _returnvalue_9 2))
  (bind _returnvalue_11 (- _returnvalue_10 1.0))
  (bind _returnvalue_12 (Math|Vector2D|MakeVector2D _returnvalue_11 _returnvalue_7))
  (bind _unit_vector (Class|HemiOctahedra|HemiOctahedrontoUnitVector _returnvalue_12))
  (bind _returnvalue_13 (Math|Vector|Lerp(Vector) _unit_vector "0,0.000000,1.000000" _upward_bias))
  (bind _returnvalue_14 (* _returnvalue_13 1000))
  (bind _proceduralmesh (Variables|Default|GetProceduralMesh))
  (bind _previewcard (Variables|Default|GetPreviewCard))
  (CallFunction|SetupRTAndBakeList)
  (Variables|Debug|SetOffset%20Vector (- _origin (Transformation|GetActorLocation _static_mesh_actor)))
  (Variables|Debug|SetObject%20Radius (+ _sphereradius (Variables|Default|GetObjectRadiusMargin)))
  ; Debug Octahedron Layout
  (Utilities|Array|Clear (Variables|Debug|GetCaptureOcts))
  (Rendering|Debug|FlushPersistentDebugLines 0)
  (for _index (range (Variables|CustomLighting|GetLighting%20Grid%20XY))
    (Variables|SetcurY _index)
    (for _index_2 (range (Variables|CustomLighting|GetLighting%20Grid%20XY))
      (Variables|SetcurX _index_2)
      (Utilities|Array|Add (Variables|Default|GetLightVecs) (Math|Vector|Lerp(Vector) (Class|HemiOctahedra|HemiOctahedrontoUnitVector (Math|Vector2D|MakeVector2D (- (* (/ (Math|Conversions|ToFloat(Integer) (Variables|GetcurX)) (- (Math|Conversions|ToFloat(Integer) (Variables|CustomLighting|GetLighting%20Grid%20XY)) 1.0)) 2) 1.0) (- (* (/ (Math|Conversions|ToFloat(Integer) (Variables|GetcurY)) (- (Math|Conversions|ToFloat(Integer) (Variables|CustomLighting|GetLighting%20Grid%20XY)) 1.0)) 2) 1.0))) "0,0.000000,1.000000" (Variables|CustomLighting|GetUpward%20Bias)))))
  (Rendering|Debug|FlushPersistentDebugLines 0)
  (for _index_3 (range (Variables|Default|GetFrames%20XY))
    (Variables|SetcurY _index_3)
    (for _index_4 (range (Variables|Default|GetFrames%20XY))
      (Variables|SetcurX _index_4)
      (Utilities|Array|Add (Variables|Debug|GetCaptureOcts) _returnvalue_12)
      (if (Variables|Debug|GetDraw%20Debug%20Spheres)
        (Rendering|Debug|DrawDebugPoint 0 _returnvalue_14 25 "(R=0.000000,G=1.000000,B=0.062989,A=1.000000)" 500)
        (bind _returnvalue_15 (AddComponent|NODE_AddTextRenderComponent-1 :bManualAttachment true :RelativeTransform (Math|Conversions|ToTransform(Vector) _returnvalue_14)))
        (Rendering|Components|TextRender|SetText _returnvalue_15 (Utilities|Text|ToText(Vector2d) _returnvalue_12)))))
  ;[0.4695 0.588 1] Scene Capture Setup
  (Variables|/Script/Engine.SceneCaptureComponent2D|SetTextureTarget (Variables|RenderTargets|GetSceneCaptureRT) (Variables|Default|GetSceneCaptureComponent2D))
  (Rendering|SceneCapture|ShowOnlyActorComponents (Variables|Default|GetSceneCaptureComponent2D) (Variables|Default|GetStatic%20Mesh%20Actor) false)
  (Variables|/Script/Engine.SceneCaptureComponent|SetShowFlagSettings (Utilities|Array|MakeArray (Utilities|Struct|MakeEngineShowFlagsSetting "AmbientOcclusion") (Utilities|Struct|MakeEngineShowFlagsSetting "Fog") (Utilities|Struct|MakeEngineShowFlagsSetting "AtmosphericFog") (Utilities|Struct|MakeEngineShowFlagsSetting "AtmosphericFog")) (Variables|Default|GetSceneCaptureComponent2D))
  (Variables|/Script/Engine.SceneCaptureComponent2D|SetProjectionType (select (Variables|Advanced|GetOrtho) "Orthographic" "Perspective") (Variables|Default|GetSceneCaptureComponent2D))
  (Variables|/Script/Engine.SceneCaptureComponent2D|SetOrthoWidth (* (Variables|Debug|GetObject%20Radius) 2) (Variables|Default|GetSceneCaptureComponent2D))
  (Variables|/Script/Engine.SceneCaptureComponent2D|SetFOVAngle (* (Math|Trig|Atan(Degrees) (/ (Variables|Debug|GetObject%20Radius) (Variables|Advanced|GetCamera%20Distance))) 2) (Variables|Default|GetSceneCaptureComponent2D))
  (Utilities|FlowControl|Sequence
    (:then_0
      (Utilities|CreateAndSetupMIDs))
    (:then_1
      (Rendering|Material|SetScalarParameterValue 0 "/Game/ImpostorBaker/MPC/MPC_ImposterCapture.MPC_ImposterCapture" "EnableCapture" _returnvalue :!class /Script/Engine.KismetMaterialLibrary)
      (Rendering|Material|SetScalarParameterValue 0 "/Game/ImpostorBaker/MPC/MPC_ImposterCapture.MPC_ImposterCapture" "CustomLighting" _returnvalue :!class /Script/Engine.KismetMaterialLibrary)
      (if (Variables|CustomLighting|GetPreview%20CustomLighting)
        (Utilities|DisplayCustomLighting)
        (else
          (Utilities|RestoreLightingState))))
    (:then_3
      (CallFunction|CombineBasecolorCustomLighting)
      (Components|ProceduralMesh|CreateMeshSection :self (Variables|Default|GetProceduralMesh) :SectionIndex 0 :Vertices (Variables|Default|GetVerts) :Triangles (Variables|Default|GetTris) :UV0 (Variables|Default|GetUVs) :bCreateCollision false :bSRGBConversion false)
      (Rendering|Material|SetMaterial (Variables|Default|GetPreviewCard) 0 (Variables|Debug|GetImpostorPreviewMID))
      (if (Variables|Advanced|GetPreview%20Proc%20Mesh)
        (Rendering|SetVisibility _proceduralmesh true false)
        (Rendering|SetVisibility _previewcard false false)
        (else
          (Rendering|SetVisibility _proceduralmesh false false)
          (Rendering|SetVisibility _previewcard true false))))))

(fn TraceFromCenter
  (params (Direction :type Vector%202D%20Structure))
  (bind _curdirection (LocalVariables|GetCurDirection))
  (bind _returnvalue (Math|Vector2D|Normalize2D _curdirection))
  (bind (_x _y) (Math|Vector2D|BreakVector2D _returnvalue))
  (bind _returnvalue_2 (Math|Float|Absolute(Float) _x))
  (bind _returnvalue_3 (Math|Float|Absolute(Float) _y))
  (bind (_x_2 _y_2) (Math|Vector2D|BreakVector2D (LocalVariables|GetFurthestPos)))
  (bind _returnvalue_4 (Math|Float|Max(Float) _returnvalue_2 _returnvalue_3))
  (LocalVariables|SetCurDirection Direction)
  (LocalVariables|SetCurpos (+ (/ (/ Direction (/ 16 2)) _returnvalue_4) 0.5))
  (LocalVariables|SetFurthestPos "(X=-1.000000,Y=-1.000000)")
  (for _index (range 7)
    (bind _output_get (LocalVariables|SetCurpos (+ (LocalVariables|GetCurpos) (/ (/ _returnvalue 16) _returnvalue_4))))
    (Rendering|Debug|DrawDebugSphere 0 (Math|Vector|MakeVector (Math|Vector2D|BreakVector2D (* (- _output_get 0.5) (* (Variables|Debug|GetDebugTexelSize) 16))) (.Y (Math|Vector2D|BreakVector2D (* (- _output_get 0.5) (* (Variables|Debug|GetDebugTexelSize) 16))))) 2 12 (Utilities|SampleTextureData0-1 _output_get) 5000)
    (if (> (Utilities|Struct|BreakLinearColor (Utilities|SampleTextureData0-1 _output_get)) 0.0)
      _
      (elif (== _index 6)
        (if (and (== _x_2 -1) (== _y_2 -1))
          _)))
    (LocalVariables|SetFurthestPos (+ (LocalVariables|GetCurpos) (* (LocalVariables|GetCurDirection) 32))))
  (Utilities|Array|Add (Variables|Debug|GetRadialExtents) (LocalVariables|GetFurthestPos)))

(fn TraceUntilIntersection
  (params (StartX :type Integer) (StartY :type Integer) (TraceX :type Boolean))
  (bind (_0 _90 _180 _270) (Utilities|RotatedCoords (Math|Vector2D|MakeVector2D (Math|Conversions|ToFloat(Integer) (LocalVariables|GetcurX)) (Math|Conversions|ToFloat(Integer) (LocalVariables|GetcurY))) 16))
  (bind _curx (LocalVariables|GetcurX))
  (bind _cury (LocalVariables|GetcurY))
  (LocalVariables|SetcurX StartX)
  (LocalVariables|SetcurY StartY)
  (bind _index (Utilities|FlowControl|ForLoopwithBreak 0 15)
    (:LoopBody
      (if (<= (Utilities|Struct|BreakLinearColor (Utilities|SampleTextureDataPoint (Utilities|Select :Index (Variables|Default|GetCurrentCornerPass) :Option%200 _0 :Option%201 _90 :Option%202 _180 :Option%203 _270))) 0.0001)
        _
        (else
          _))
      (Math|Integer|IncrementInt (select TraceX (LocalVariables|GetcurX) (LocalVariables|GetcurY))))
    (:Completed
      _))
  (Development|PrintString 0 (Utilities|String|ToString(Integer) (select TraceX _curx _cury)))
  (return _index _curx _cury))

(fn ConstructionScript
  (CallFunction|SetupAll)
  (CallFunction|ForceShaderCompileBeforeBake)
  (CallFunction|SetShowFlagsforCapture))
