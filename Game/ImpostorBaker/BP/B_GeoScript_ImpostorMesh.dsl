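; Editor-side Geometry Script actor for Impostor Baker that builds an impostor mesh and bakes it to static mesh assets.
; UpdateMesh and the duplicate rebuild event add stored vertices and triangles, fill four UV channels (channel 1 zeroed) and assign the target material ID.
; BakeMesh writes the mesh into a LOD of an existing static mesh, then resets material IDs and overwrites or creates a standalone impostor asset.
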(event Custom|UpdateMesh
  (Development|PrintString "Update Mesh")
  (bind _returnvalue (DynamicMeshComponent|GetDynamicMesh (Variables|DynamicMeshActor|GetDynamicMeshComponent)))
  (bind _output_get (Variables|Default|SetDynamicMesh _returnvalue))
  (bind _vectorlist (GeometryScript|ListUtils|ConvertArraytoVectorList (Variables|Default|GetVerts)))
  (GeometryScript|MeshEdits|AddVerticestoMesh _output_get _vectorlist)
  (Utilities|FlowControl|Sequence
    (:then_0
      (for _index (range (/ (Utilities|Array|Length (Variables|Default|GetTris)) 3))
        (bind _newtriangleindex (GeometryScript|MeshEdits|AddTriangletoMesh (Variables|Default|GetDynamicMesh) (Utilities|Struct|MakeIntVector (Utilities|Array|Get(acopy) (Variables|Default|GetTris) (* _index 3)) (Utilities|Array|Get(acopy) (Variables|Default|GetTris) (+ (* _index 3) 1)) (Utilities|Array|Get(acopy) (Variables|Default|GetTris) (+ (* _index 3) 2)))))
        (Utilities|Array|Add (Variables|Default|GetNew%20Triangle%20Index) _newtriangleindex)))
    (:then_1
      (GeometryScript|UVs|SetNumUVChannels (Variables|Default|GetDynamicMesh) 4)
      (Utilities|FlowControl|Sequence
        (:then_0
          (if (> (Utilities|Array|Length (Variables|Default|GetUVs)) 0)
            (for (_ _array_index) (Variables|Default|GetNew%20Triangle%20Index)
              (GeometryScript|UVs|SetMeshTriangleUVs (Variables|Default|GetDynamicMesh) 0 _array_index (Utilities|Struct|MakeGeometryScriptUVTriangle (Utilities|Array|Get(acopy) (Variables|Default|GetUVs) (Utilities|Struct|BreakIntVector (.ReturnValue (GeometryScript|MeshQueries|GetTriangleIndices (Variables|Default|GetDynamicMesh) _array_index)))) (Utilities|Array|Get(acopy) (Variables|Default|GetUVs) (.Y (Utilities|Struct|BreakIntVector (.ReturnValue (GeometryScript|MeshQueries|GetTriangleIndices (Variables|Default|GetDynamicMesh) _array_index))))) (Utilities|Array|Get(acopy) (Variables|Default|GetUVs) (.Z (Utilities|Struct|BreakIntVector (.ReturnValue (GeometryScript|MeshQueries|GetTriangleIndices (Variables|Default|GetDynamicMesh) _array_index))))))))))
        (:then_1
          (for (_ _array_index_2) (Variables|Default|GetNew%20Triangle%20Index)
            (GeometryScript|UVs|SetMeshTriangleUVs (Variables|Default|GetDynamicMesh) 1 _array_index_2 (Utilities|Struct|MakeGeometryScriptUVTriangle "(X=0.000000,Y=0.000000)"))))
        (:then_2
          (if (> (Utilities|Array|Length (Variables|Default|GetUVs_2)) 0)
            (for (_ _array_index_3) (Variables|Default|GetNew%20Triangle%20Index)
              (GeometryScript|UVs|SetMeshTriangleUVs (Variables|Default|GetDynamicMesh) 2 _array_index_3 (Utilities|Struct|MakeGeometryScriptUVTriangle (Utilities|Array|Get(acopy) (Variables|Default|GetUVs_2) (Utilities|Struct|BreakIntVector (.ReturnValue (GeometryScript|MeshQueries|GetTriangleIndices (Variables|Default|GetDynamicMesh) _array_index_3)))) (Utilities|Array|Get(acopy) (Variables|Default|GetUVs_2) (.Y (Utilities|Struct|BreakIntVector (.ReturnValue (GeometryScript|MeshQueries|GetTriangleIndices (Variables|Default|GetDynamicMesh) _array_index_3))))) (Utilities|Array|Get(acopy) (Variables|Default|GetUVs_2) (.Z (Utilities|Struct|BreakIntVector (.ReturnValue (GeometryScript|MeshQueries|GetTriangleIndices (Variables|Default|GetDynamicMesh) _array_index_3))))))))))
        (:then_3
          (if (> (Utilities|Array|Length (Variables|Default|GetUVs_3)) 0)
            (for (_ _array_index_4) (Variables|Default|GetNew%20Triangle%20Index)
              (GeometryScript|UVs|SetMeshTriangleUVs (Variables|Default|GetDynamicMesh) 3 _array_index_4 (Utilities|Struct|MakeGeometryScriptUVTriangle (Utilities|Array|Get(acopy) (Variables|Default|GetUVs_3) (Utilities|Struct|BreakIntVector (.ReturnValue (GeometryScript|MeshQueries|GetTriangleIndices (Variables|Default|GetDynamicMesh) _array_index_4)))) (Utilities|Array|Get(acopy) (Variables|Default|GetUVs_3) (.Y (Utilities|Struct|BreakIntVector (.ReturnValue (GeometryScript|MeshQueries|GetTriangleIndices (Variables|Default|GetDynamicMesh) _array_index_4))))) (Utilities|Array|Get(acopy) (Variables|Default|GetUVs_3) (.Z (Utilities|Struct|BreakIntVector (.ReturnValue (GeometryScript|MeshQueries|GetTriangleIndices (Variables|Default|GetDynamicMesh) _array_index_4))))))))))
        (:then_4
          (bind (_triangleidlist _bhastriangleidgaps _returnvalue_2) (GeometryScript|MeshQueries|GetAllTriangleIDs (Variables|Default|GetDynamicMesh)))
          (GeometryScript|Materials|SetMaterialIDOnTriangles _returnvalue_2 _triangleidlist (Variables|Default|GetTargetMaterialIDIndex)))))))

(event Events|EventOnRebuildGeneratedMesh
  (params (TargetMesh :type Dynamic%20Mesh%20Object%20Reference))
  (Development|PrintString "Update Mesh")
  (bind _returnvalue (DynamicMeshComponent|GetDynamicMesh (Variables|DynamicMeshActor|GetDynamicMeshComponent)))
  (bind _output_get (Variables|Default|SetDynamicMesh _returnvalue))
  (bind _vectorlist (GeometryScript|ListUtils|ConvertArraytoVectorList (Variables|Default|GetVerts)))
  (GeometryScript|MeshEdits|AddVerticestoMesh _output_get _vectorlist)
  (Utilities|FlowControl|Sequence
    (:then_0
      (for _index (range (/ (Utilities|Array|Length (Variables|Default|GetTris)) 3))
        (bind _newtriangleindex (GeometryScript|MeshEdits|AddTriangletoMesh (Variables|Default|GetDynamicMesh) (Utilities|Struct|MakeIntVector (Utilities|Array|Get(acopy) (Variables|Default|GetTris) (* _index 3)) (Utilities|Array|Get(acopy) (Variables|Default|GetTris) (+ (* _index 3) 1)) (Utilities|Array|Get(acopy) (Variables|Default|GetTris) (+ (* _index 3) 2)))))
        (Utilities|Array|Add (Variables|Default|GetNew%20Triangle%20Index) _newtriangleindex)))
    (:then_1
      (GeometryScript|UVs|SetNumUVChannels (Variables|Default|GetDynamicMesh) 4)
      (Utilities|FlowControl|Sequence
        (:then_0
          (if (> (Utilities|Array|Length (Variables|Default|GetUVs)) 0)
            (for (_ _array_index) (Variables|Default|GetNew%20Triangle%20Index)
              (GeometryScript|UVs|SetMeshTriangleUVs (Variables|Default|GetDynamicMesh) 0 _array_index (Utilities|Struct|MakeGeometryScriptUVTriangle (Utilities|Array|Get(acopy) (Variables|Default|GetUVs) (Utilities|Struct|BreakIntVector (.ReturnValue (GeometryScript|MeshQueries|GetTriangleIndices (Variables|Default|GetDynamicMesh) _array_index)))) (Utilities|Array|Get(acopy) (Variables|Default|GetUVs) (.Y (Utilities|Struct|BreakIntVector (.ReturnValue (GeometryScript|MeshQueries|GetTriangleIndices (Variables|Default|GetDynamicMesh) _array_index))))) (Utilities|Array|Get(acopy) (Variables|Default|GetUVs) (.Z (Utilities|Struct|BreakIntVector (.ReturnValue (GeometryScript|MeshQueries|GetTriangleIndices (Variables|Default|GetDynamicMesh) _array_index))))))))))
        (:then_1
          (for (_ _array_index_2) (Variables|Default|GetNew%20Triangle%20Index)
            (GeometryScript|UVs|SetMeshTriangleUVs (Variables|Default|GetDynamicMesh) 1 _array_index_2 (Utilities|Struct|MakeGeometryScriptUVTriangle "(X=0.000000,Y=0.000000)"))))
        (:then_2
          (if (> (Utilities|Array|Length (Variables|Default|GetUVs_2)) 0)
            (for (_ _array_index_3) (Variables|Default|GetNew%20Triangle%20Index)
              (GeometryScript|UVs|SetMeshTriangleUVs (Variables|Default|GetDynamicMesh) 2 _array_index_3 (Utilities|Struct|MakeGeometryScriptUVTriangle (Utilities|Array|Get(acopy) (Variables|Default|GetUVs_2) (Utilities|Struct|BreakIntVector (.ReturnValue (GeometryScript|MeshQueries|GetTriangleIndices (Variables|Default|GetDynamicMesh) _array_index_3)))) (Utilities|Array|Get(acopy) (Variables|Default|GetUVs_2) (.Y (Utilities|Struct|BreakIntVector (.ReturnValue (GeometryScript|MeshQueries|GetTriangleIndices (Variables|Default|GetDynamicMesh) _array_index_3))))) (Utilities|Array|Get(acopy) (Variables|Default|GetUVs_2) (.Z (Utilities|Struct|BreakIntVector (.ReturnValue (GeometryScript|MeshQueries|GetTriangleIndices (Variables|Default|GetDynamicMesh) _array_index_3))))))))))
        (:then_3
          (if (> (Utilities|Array|Length (Variables|Default|GetUVs_3)) 0)
            (for (_ _array_index_4) (Variables|Default|GetNew%20Triangle%20Index)
              (GeometryScript|UVs|SetMeshTriangleUVs (Variables|Default|GetDynamicMesh) 3 _array_index_4 (Utilities|Struct|MakeGeometryScriptUVTriangle (Utilities|Array|Get(acopy) (Variables|Default|GetUVs_3) (Utilities|Struct|BreakIntVector (.ReturnValue (GeometryScript|MeshQueries|GetTriangleIndices (Variables|Default|GetDynamicMesh) _array_index_4)))) (Utilities|Array|Get(acopy) (Variables|Default|GetUVs_3) (.Y (Utilities|Struct|BreakIntVector (.ReturnValue (GeometryScript|MeshQueries|GetTriangleIndices (Variables|Default|GetDynamicMesh) _array_index_4))))) (Utilities|Array|Get(acopy) (Variables|Default|GetUVs_3) (.Z (Utilities|Struct|BreakIntVector (.ReturnValue (GeometryScript|MeshQueries|GetTriangleIndices (Variables|Default|GetDynamicMesh) _array_index_4))))))))))
        (:then_4
          (bind (_triangleidlist _bhastriangleidgaps _returnvalue_2) (GeometryScript|MeshQueries|GetAllTriangleIDs (Variables|Default|GetDynamicMesh)))
          (GeometryScript|Materials|SetMaterialIDOnTriangles _returnvalue_2 _triangleidlist (Variables|Default|GetTargetMaterialIDIndex)))))))

(fn BakeMesh
  (bind _existingstaticmeshpath (Variables|Default|GetExistingStaticMeshPath))
  (DynamicMeshComponent|GetDynamicMesh (Variables|DynamicMeshActor|GetDynamicMeshComponent))
  (Utilities|FlowControl|Sequence
    (:then_0
      (bind _returnvalue (EditorScripting|Asset|DoesAssetExist _existingstaticmeshpath :!class /Script/EditorScriptingUtilities.EditorAssetLibrary))
      (if _returnvalue
        (bind _returnvalue_2 (EditorScripting|Asset|LoadAsset _existingstaticmeshpath :!class /Script/EditorScriptingUtilities.EditorAssetLibrary))
        (bind _asstatic_mesh (Utilities|Casting|CastToStaticMesh _returnvalue_2))
        (GeometryScript|StaticMesh|CopyMeshtoStaticMesh (Variables|Default|GetDynamicMesh) _asstatic_mesh (Utilities|Struct|MakeGeometryScriptCopyMeshtoAssetOptions true true false "DoNothing" false false true true "SetLODMaterials" (Utilities|Array|MakeArray "/Game/ImpostorBaker/Materials/M_ImposterBump01_SimpleGrid.M_ImposterBump01_SimpleGrid") (Utilities|Array|MakeArray "Impostor") true) (Utilities|Struct|MakeGeometryScriptMeshWriteLOD false (Math|Integer|SelectInt 3 (Variables|Default|GetTargetLODIndex) (== (Variables|Default|GetTargetLODIndex) -1))))))
    (:then_1
      ; Set it to Material ID 0 for Stand alone Impostor mesh
      (bind (_triangleidlist _bhastriangleidgaps _returnvalue_3) (GeometryScript|MeshQueries|GetAllTriangleIDs (Variables|Default|GetDynamicMesh)))
      (GeometryScript|Materials|SetMaterialIDOnTriangles _returnvalue_3 _triangleidlist)
      (bind _returnvalue_4 (EditorScripting|Asset|DoesAssetExist (Variables|Default|GetNewAssetPathName) :!class /Script/EditorScriptingUtilities.EditorAssetLibrary))
      (if _returnvalue_4
        (bind _returnvalue_5 (EditorScripting|Asset|LoadAsset (Variables|Default|GetNewAssetPathName) :!class /Script/EditorScriptingUtilities.EditorAssetLibrary))
        (bind _asstatic_mesh_2 (Utilities|Casting|CastToStaticMesh _returnvalue_5))
        (GeometryScript|StaticMesh|CopyMeshtoStaticMesh (Variables|Default|GetDynamicMesh) _asstatic_mesh_2 (Utilities|Struct|MakeGeometryScriptCopyMeshtoAssetOptions true true) (Utilities|Struct|MakeGeometryScriptMeshWriteLOD false))
        (else
          (GeometryScript|AssetManagement|CreateNewStaticMeshAssetfromMesh (Variables|Default|GetDynamicMesh) (Variables|Default|GetNewAssetPathName) (Utilities|Struct|MakeGeometryScriptCreateNewStaticMeshAssetOptions false false false "(bEnabled=False,bExplicitTangents=False,bLerpUVs=True,bSeparable=False,bVoxelNDF=True,bVoxelOpacity=False,ShapePreservation=None,PositionPrecision=-2147483648,NormalPrecision=-1,TangentPrecision=-1,BoneWeightPrecision=-1,TargetMinimumResidencyInKB=0,KeepPercentTriangles=1.000000,TrimRelativeError=0.000000,GenerateFallback=PlatformDefault,FallbackTarget=Auto,FallbackPercentTriangles=1.000000,FallbackRelativeError=1.000000,MaxEdgeLengthFactor=0.000000,NumRays=64,VoxelLevel=0,RayBackUp=0.000000,DisplacementUVChannel=0,DisplacementMaps=,NaniteAssemblyData=(Parts=,Nodes=))" 0.0 false)))))) ; For now do both to debug \nDepending on the work flow and impotor mesh needs\nwe might need impostors as a seperate mesh
)

(fn ConstructionScript)
