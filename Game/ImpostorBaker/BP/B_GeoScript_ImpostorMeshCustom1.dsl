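; Geometry Script dynamic mesh actor in the impostor baker that bakes a supplied static mesh into a LOD of an existing static mesh asset.
; On rebuild or UpdateMesh, it copies the source mesh with section materials into its dynamic mesh.
; BakeMesh loads the asset at the stored path, remaps material ID 0 to the target ID, and writes the target LOD, using LOD 3 when the index is -1.
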(event EventBeginPlay)

(event Collision|EventActorBeginOverlap
  (params (OtherActor :type Actor%20Object%20Reference)))

(event EventTick
  (params (DeltaSeconds :type Float%20%28single-precision%29)))

(event Custom|UpdateMesh
  (Utilities|FlowControl|Sequence
    (:then_0
      (Development|PrintString "Update Mesh")
      (bind _returnvalue (DynamicMeshComponent|GetDynamicMesh (Variables|DynamicMeshActor|GetDynamicMeshComponent)))
      (bind _output_get (Variables|Default|SetDynamicMesh _returnvalue))
      (GeometryScript|StaticMesh|CopyMeshFromStaticMeshwithSectionMaterials (Variables|Default|GetFrom%20Static%20Mesh%20Asset) _output_get (Utilities|Struct|MakeGeometryScriptCopyMeshfromAssetOptions true) (Utilities|Struct|MakeGeometryScriptMeshReadLOD "MaxAvailable")))))

(event Events|EventOnRebuildGeneratedMesh
  (params (TargetMesh :type Dynamic%20Mesh%20Object%20Reference))
  (Utilities|FlowControl|Sequence
    (:then_0
      (Development|PrintString "Update Mesh")
      (bind _returnvalue (DynamicMeshComponent|GetDynamicMesh (Variables|DynamicMeshActor|GetDynamicMeshComponent)))
      (bind _output_get (Variables|Default|SetDynamicMesh _returnvalue))
      (GeometryScript|StaticMesh|CopyMeshFromStaticMeshwithSectionMaterials (Variables|Default|GetFrom%20Static%20Mesh%20Asset) _output_get (Utilities|Struct|MakeGeometryScriptCopyMeshfromAssetOptions true) (Utilities|Struct|MakeGeometryScriptMeshReadLOD "MaxAvailable")))))

(fn BakeMesh
  (bind _existingstaticmeshpath (Variables|Default|GetExistingStaticMeshPath))
  (DynamicMeshComponent|GetDynamicMesh (Variables|DynamicMeshActor|GetDynamicMeshComponent))
  (Utilities|FlowControl|Sequence
    (:then_1
      (bind _returnvalue (EditorScripting|Asset|DoesAssetExist _existingstaticmeshpath :!class /Script/EditorScriptingUtilities.EditorAssetLibrary))
      (if _returnvalue
        (bind _returnvalue_2 (EditorScripting|Asset|LoadAsset _existingstaticmeshpath :!class /Script/EditorScriptingUtilities.EditorAssetLibrary))
        (bind _asstatic_mesh (Utilities|Casting|CastToStaticMesh _returnvalue_2))
        (bind _returnvalue_3 (GeometryScript|Materials|RemapMaterialIDs (Variables|Default|GetDynamicMesh) 0 (Variables|Default|GetTargetMaterialID)))
        (GeometryScript|StaticMesh|CopyMeshtoStaticMesh _returnvalue_3 _asstatic_mesh (Utilities|Struct|MakeGeometryScriptCopyMeshtoAssetOptions true true) (Utilities|Struct|MakeGeometryScriptMeshWriteLOD false (Math|Integer|SelectInt 3 (Variables|Default|GetTargetLODIndex) (== (Variables|Default|GetTargetLODIndex) -1))))))) ; For now do both to debug \nDepending on the work flow and impotor mesh needs\nwe might need impostors as a seperate mesh
)

(fn ConstructionScript)
