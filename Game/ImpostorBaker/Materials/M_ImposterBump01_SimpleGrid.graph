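# Impostor material that renders baked multi-view captures from a frame grid, with bump offset implied by its name.
# View direction maps through hemi-octahedral encoding to grid frames, while world position offset moves the mesh.
# Custom shadow-depth switches substitute camera inputs and UVs in the shadow pass; a distance-field switch selects alternate values.
# Per-instance custom data, desaturation and hue variation nodes vary appearance, and an orthographic check gates pixel depth offset.
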Material /Game/ImpostorBaker/Materials/M_ImposterBump01_SimpleGrid
graph {
 expressions {
  MaterialExpressionAbs_0 {class=MaterialExpressionAbs, inputs [None {from=MaterialExpressionDotProduct_3, output=""}], properties {}}
  MaterialExpressionAbs_1 {class=MaterialExpressionAbs, inputs [None {from=MaterialExpressionDotProduct_1, output=""}], properties {}}
  MaterialExpressionAbs_2 {class=MaterialExpressionAbs, inputs [None {from=MaterialExpressionDotProduct_2, output=""}], properties {}}
  MaterialExpressionAbs_3 {class=MaterialExpressionAbs, inputs [None {from=MaterialExpressionSubtract_1, output=""}], properties {}}
  MaterialExpressionActorPositionWS_5 {class=MaterialExpressionActorPositionWS, inputs [], properties {}}
  MaterialExpressionAdd_0 {
   class=MaterialExpressionAdd
   inputs [
    A {from=MaterialExpressionReroute_8}
    B {from=MaterialExpressionMultiply_2}
   ]
  }
  MaterialExpressionAdd_1 {
   class=MaterialExpressionAdd
   inputs [
    A {from=MaterialExpressionFloor_4}
    B {from=MaterialExpressionLinearInterpolate_12}
   ]
  }
  MaterialExpressionAdd_10 {
   class=MaterialExpressionAdd
   inputs [
    A {from=MaterialExpressionAdd_4}
    B {from=MaterialExpressionConstant_4}
   ]
  }
  MaterialExpressionAdd_11 {
   class=MaterialExpressionAdd
   inputs [
    A {from=MaterialExpressionReroute_12}
    B {from=MaterialExpressionAppendVector_0}
   ]
  }
  MaterialExpressionAdd_12 {
   class=MaterialExpressionAdd
   inputs [
    A {from=MaterialExpressionMaterialFunctionCall_6, output=World position offset}
    B {from=MaterialExpressionWorldPosition_1, output=XYZ}
   ]
  }
  MaterialExpressionAdd_13 {
   class=MaterialExpressionAdd
   inputs [
    A {from=MaterialExpressionMaterialFunctionCall_42, output=Result}
    B {from=MaterialExpressionScalarParameter_49}
   ]
  }
  MaterialExpressionAdd_14 {
   class=MaterialExpressionAdd
   inputs [
    A {from=MaterialExpressionStaticSwitchParameter_12}
    B {from=MaterialExpressionMultiply_48}
   ]
  }
  MaterialExpressionAdd_15 {
   class=MaterialExpressionAdd
   inputs [
    A {from=MaterialExpressionAdd_7}
    B {from=MaterialExpressionDivide_5}
   ]
  }
  MaterialExpressionAdd_16 {
   class=MaterialExpressionAdd
   inputs [
    A {from=MaterialExpressionMultiply_25}
    B {from=MaterialExpressionMultiply_37}
   ]
  }
  MaterialExpressionAdd_17 {
   class=MaterialExpressionAdd
   inputs [
    A {from=MaterialExpressionMultiply_21}
    B {from=MaterialExpressionMultiply_11}
   ]
  }
  MaterialExpressionAdd_18 {
   class=MaterialExpressionAdd
   inputs [
    A {from=MaterialExpressionAdd_19}
    B {from=""}
   ]
   properties {constB=2}
  }
  MaterialExpressionAdd_19 {
   class=MaterialExpressionAdd
   inputs [
    A {from=MaterialExpressionSine_0}
    B {from=MaterialExpressionSine_1}
   ]
  }
  MaterialExpressionAdd_2 {
   class=MaterialExpressionAdd
   inputs [
    A {from=MaterialExpressionFloor_4}
    B {from=MaterialExpressionConstant2Vector_1}
   ]
  }
  MaterialExpressionAdd_20 {
   class=MaterialExpressionAdd
   inputs [
    A {from=MaterialExpressionDivide_21}
    B {from=MaterialExpressionDivide_1}
   ]
  }
  MaterialExpressionAdd_21 {
   class=MaterialExpressionAdd
   inputs [
    A {from=MaterialExpressionReroute_3}
    B {from=MaterialExpressionDivide_17}
   ]
  }
  MaterialExpressionAdd_22 {
   class=MaterialExpressionAdd
   inputs [
    A {from=MaterialExpressionAdd_23}
    B {from=MaterialExpressionMultiply_39}
   ]
  }
  MaterialExpressionAdd_23 {
   class=MaterialExpressionAdd
   inputs [
    A {from=MaterialExpressionAdd_21}
    B {from=MaterialExpressionDivide_16}
   ]
  }
  MaterialExpressionAdd_24 {
   class=MaterialExpressionAdd
   inputs [
    A {from=MaterialExpressionMultiply_64}
    B {from=""}
   ]
  }
  MaterialExpressionAdd_25 {
   class=MaterialExpressionAdd
   inputs [
    A {from=MaterialExpressionScalarParameter_52}
    B {from=MaterialExpressionScalarParameter_75}
   ]
  }
  MaterialExpressionAdd_26 {
   class=MaterialExpressionAdd
   inputs [
    A {from=MaterialExpressionVectorParameter_9, output=A}
    B {from=MaterialExpressionPerInstanceCustomData_4}
   ]
  }
  MaterialExpressionAdd_27 {
   class=MaterialExpressionAdd
   inputs [
    A {from=MaterialExpressionMultiply_73}
    B {from=""}
   ]
  }
  MaterialExpressionAdd_28 {
   class=MaterialExpressionAdd
   inputs [
    A {from=MaterialExpressionMaterialFunctionCall_6, output=World position offset}
    B {from=MaterialExpressionCustom_2}
   ]
   properties {constB=0}
  }
  MaterialExpressionAdd_29 {
   class=MaterialExpressionAdd
   inputs [
    A {from=MaterialExpressionSine_2}
    B {from=MaterialExpressionSine_3}
   ]
  }
  MaterialExpressionAdd_3 {
   class=MaterialExpressionAdd
   inputs [
    A {from=MaterialExpressionMaterialFunctionCall_7, output=WPO}
    B {from=MaterialExpressionMultiply_35}
   ]
  }
  MaterialExpressionAdd_30 {
   class=MaterialExpressionAdd
   inputs [
    A {from=MaterialExpressionVectorParameter_20, output=RGB}
    B {from=MaterialExpressionPerInstanceCustomData3Vector_3}
   ]
  }
  MaterialExpressionAdd_31 {
   class=MaterialExpressionAdd
   inputs [
    A {from=MaterialExpressionNamedRerouteUsage_34}
    B {from=MaterialExpressionScalarParameter_73}
   ]
  }
  MaterialExpressionAdd_32 {
   class=MaterialExpressionAdd
   inputs [
    A {from=MaterialExpressionMultiply_84}
    B {from=MaterialExpressionMultiply_85}
   ]
  }
  MaterialExpressionAdd_33 {
   class=MaterialExpressionAdd
   inputs [
    A {from=MaterialExpressionMultiply_87}
    B {from=MaterialExpressionMultiply_83}
   ]
  }
  MaterialExpressionAdd_34 {
   class=MaterialExpressionAdd
   inputs [
    A {from=MaterialExpressionMultiply_88}
    B {from=""}
   ]
  }
  MaterialExpressionAdd_4 {
   class=MaterialExpressionAdd
   inputs [
    A {from=MaterialExpressionMultiply_7}
    B {from=MaterialExpressionSphereMask_2}
   ]
  }
  MaterialExpressionAdd_5 {
   class=MaterialExpressionAdd
   inputs [
    A {from=MaterialExpressionMultiply_41}
    B {from=MaterialExpressionReroute_6}
   ]
  }
  MaterialExpressionAdd_6 {
   class=MaterialExpressionAdd
   inputs [
    A {from=MaterialExpressionWorldPosition_0, output=XYZ}
    B {from=MaterialExpressionMultiply_5}
   ]
  }
  MaterialExpressionAdd_7 {
   class=MaterialExpressionAdd
   inputs [
    A {from=MaterialExpressionReroute_2}
    B {from=MaterialExpressionDivide_2}
   ]
  }
  MaterialExpressionAdd_8 {
   class=MaterialExpressionAdd
   inputs [
    A {from=MaterialExpressionPerInstanceCustomData3Vector_2}
    B {from=MaterialExpressionVectorParameter_9, output=RGB}
   ]
  }
  MaterialExpressionAdd_9 {
   class=MaterialExpressionAdd
   inputs [
    A {from=MaterialExpressionAdd_15}
    B {from=MaterialExpressionMultiply_10}
   ]
  }
  MaterialExpressionAppendVector_0 {
   class=MaterialExpressionAppendVector
   inputs [
    A {from=MaterialExpressionConstant2Vector_5, output=""}
    B {from=MaterialExpressionIf_0, output=""}
   ]
  }
  MaterialExpressionAppendVector_10 {
   class=MaterialExpressionAppendVector
   inputs [
    A {from=MaterialExpressionShadingPathSwitch_3, output=""}
    B {from=MaterialExpressionConstant_24, output=""}
   ]
  }
  MaterialExpressionAppendVector_11 {
   class=MaterialExpressionAppendVector
   inputs [
    A {from=MaterialExpressionReroute_21, output=""}
    B {from=MaterialExpressionConstant_25, output=""}
   ]
  }
  MaterialExpressionAppendVector_2 {
   class=MaterialExpressionAppendVector
   inputs [
    A {from=MaterialExpressionConstant2Vector_7, output=""}
    B {from=MaterialExpressionScalarParameter_85, output=""}
   ]
  }
  MaterialExpressionAppendVector_5 {
   class=MaterialExpressionAppendVector
   inputs [
    A {from=MaterialExpressionComponentMask_2, output=""}
    B {from=MaterialExpressionComponentMask_5, output=""}
   ]
  }
  MaterialExpressionAppendVector_6 {
   class=MaterialExpressionAppendVector
   inputs [
    A {from=MaterialExpressionMultiply_28, output=""}
    B {from=MaterialExpressionConstant_10, output=""}
   ]
  }
  MaterialExpressionAppendVector_9 {
   class=MaterialExpressionAppendVector
   inputs [
    A {from=MaterialExpressionReroute_21, output=""}
    B {from=MaterialExpressionConstant_23, output=""}
   ]
  }
  MaterialExpressionCameraPositionWS_1 {class=MaterialExpressionCameraPositionWS, inputs [], properties {}}
  MaterialExpressionCameraPositionWS_2 {class=MaterialExpressionCameraPositionWS, inputs [], properties {}}
  MaterialExpressionCameraPositionWS_3 {class=MaterialExpressionCameraPositionWS, inputs [], properties {desc=Ray Origin}}
  MaterialExpressionCameraPositionWS_4 {class=MaterialExpressionCameraPositionWS, inputs [], properties {}}
  MaterialExpressionCameraPositionWS_5 {class=MaterialExpressionCameraPositionWS, inputs [], properties {}}
  MaterialExpressionCameraPositionWS_6 {class=MaterialExpressionCameraPositionWS, inputs [], properties {}}
  MaterialExpressionCameraPositionWS_7 {class=MaterialExpressionCameraPositionWS, inputs [], properties {}}
  MaterialExpressionCameraPositionWS_8 {class=MaterialExpressionCameraPositionWS, inputs [], properties {}}
  MaterialExpressionCameraVectorWS_0 {class=MaterialExpressionCameraVectorWS, inputs [], properties {}}
  MaterialExpressionCameraVectorWS_1 {class=MaterialExpressionCameraVectorWS, inputs [], properties {}}
  MaterialExpressionClamp_0 {
   class=MaterialExpressionClamp
   inputs [
    None {from=MaterialExpressionConstantBiasScale_2, output=""}
    Min {from="", output=""}
    Max {from="", output=""}
   ]
   properties {}
  }
  MaterialExpressionCollectionParameter_0 {
   class=MaterialExpressionCollectionParameter
   inputs []
   properties {
    collection=&FortniteMaterialParameters
    parameterName=SunlightDirectionalVector
   }
  }
  MaterialExpressionCollectionParameter_3 {class=MaterialExpressionCollectionParameter, inputs [], properties {parameterName=Direction}}
  MaterialExpressionComponentMask_0 {
   class=MaterialExpressionComponentMask
   inputs [None {from=MaterialExpressionMaterialFunctionCall_30, output=Local Z Vector}]
   properties {g=true, r=true}
  }
  MaterialExpressionComponentMask_1 {
   class=MaterialExpressionComponentMask
   inputs [None {from=MaterialExpressionMaterialFunctionCall_31, output=Local Z Vector}]
   properties {g=true, r=true}
  }
  MaterialExpressionComponentMask_10 {class=MaterialExpressionComponentMask, inputs [None {from=MaterialExpressionMaterialFunctionCall_49, output=Result}], properties {a=true}}
  MaterialExpressionComponentMask_11 {
   class=MaterialExpressionComponentMask
   inputs [None {from=MaterialExpressionMaterialFunctionCall_50, output=Result}]
   properties {b=true, g=true, r=true}
  }
  MaterialExpressionComponentMask_13 {class=MaterialExpressionComponentMask, inputs [None {from=MaterialExpressionMaterialFunctionCall_34, output=Result}], properties {r=true}}
  MaterialExpressionComponentMask_14 {class=MaterialExpressionComponentMask, inputs [None {from=MaterialExpressionMaterialFunctionCall_37, output=Result}], properties {r=true}}
  MaterialExpressionComponentMask_15 {class=MaterialExpressionComponentMask, inputs [None {from=MaterialExpressionMaterialFunctionCall_49, output=R Sample}], properties {a=true}}
  MaterialExpressionComponentMask_16 {class=MaterialExpressionComponentMask, inputs [None {from=MaterialExpressionCustom_6}], properties {b=true}}
  MaterialExpressionComponentMask_17 {class=MaterialExpressionComponentMask, inputs [None {from=MaterialExpressionCollectionParameter_0}], properties {b=true, g=true, r=true}}
  MaterialExpressionComponentMask_18 {class=MaterialExpressionComponentMask, inputs [None {from=MaterialExpressionDivide_19}], properties {r=true}}
  MaterialExpressionComponentMask_19 {class=MaterialExpressionComponentMask, inputs [None {from=MaterialExpressionCollectionParameter_3}], properties {b=true, g=true, r=true}}
  MaterialExpressionComponentMask_2 {class=MaterialExpressionComponentMask, inputs [None {from=MaterialExpressionFloor_3}], properties {r=true}}
  MaterialExpressionComponentMask_20 {class=MaterialExpressionComponentMask, inputs [None {from=MaterialExpressionReroute_9}], properties {g=true}}
  MaterialExpressionComponentMask_23 {class=MaterialExpressionComponentMask, inputs [None {from=MaterialExpressionAdd_14}], properties {b=true, g=true, r=true}}
  MaterialExpressionComponentMask_24 {class=MaterialExpressionComponentMask, inputs [None {from=MaterialExpressionNamedRerouteUsage_25}], properties {b=true, g=true, r=true}}
  MaterialExpressionComponentMask_3 {class=MaterialExpressionComponentMask, inputs [None {from=MaterialExpressionMaterialFunctionCall_50, output=Result}], properties {a=true}}
  MaterialExpressionComponentMask_4 {class=MaterialExpressionComponentMask, inputs [None {from=MaterialExpressionTextureCoordinate_0}], properties {g=true}}
  MaterialExpressionComponentMask_5 {class=MaterialExpressionComponentMask, inputs [None {from=MaterialExpressionFloor_2}], properties {r=true}}
  MaterialExpressionComponentMask_6 {class=MaterialExpressionComponentMask, inputs [None {from=MaterialExpressionNamedRerouteUsage_15}], properties {b=true, g=true, r=true}}
  MaterialExpressionComponentMask_7 {class=MaterialExpressionComponentMask, inputs [None {from=MaterialExpressionNamedRerouteUsage_18}], properties {b=true}}
  MaterialExpressionComponentMask_8 {
   class=MaterialExpressionComponentMask
   inputs [None {from=MaterialExpressionMaterialFunctionCall_32, output=Local Z Vector}]
   properties {g=true, r=true}
  }
  MaterialExpressionComponentMask_9 {
   class=MaterialExpressionComponentMask
   inputs [None {from=MaterialExpressionMaterialFunctionCall_49, output=Result}]
   properties {b=true, g=true, r=true}
  }
  MaterialExpressionConstant2Vector_0 {class=MaterialExpressionConstant2Vector, properties {g=1}}
  MaterialExpressionConstant2Vector_1 {class=MaterialExpressionConstant2Vector, properties {g=1, r=1}}
  MaterialExpressionConstant2Vector_2 {class=MaterialExpressionConstant2Vector, properties {r=1}}
  MaterialExpressionConstant2Vector_3 {class=MaterialExpressionConstant2Vector, properties {g=1}}
  MaterialExpressionConstant2Vector_4 {class=MaterialExpressionConstant2Vector, properties {r=1}}
  MaterialExpressionConstant2Vector_5 {class=MaterialExpressionConstant2Vector}
  MaterialExpressionConstant2Vector_7 {class=MaterialExpressionConstant2Vector}
  MaterialExpressionConstant3Vector_0 {class=MaterialExpressionConstant3Vector, properties {constant=<rgba 0 0 1 0>}}
  MaterialExpressionConstant3Vector_1 {class=MaterialExpressionConstant3Vector, properties {constant=<rgba 1 1 0 0>}}
  MaterialExpressionConstant3Vector_2 {class=MaterialExpressionConstant3Vector}
  MaterialExpressionConstant3Vector_3 {class=MaterialExpressionConstant3Vector}
  MaterialExpressionConstant3Vector_4 {class=MaterialExpressionConstant3Vector}
  MaterialExpressionConstant3Vector_5 {class=MaterialExpressionConstant3Vector, properties {constant=<rgba 1 1 1 0>}}
  MaterialExpressionConstant3Vector_6 {class=MaterialExpressionConstant3Vector}
  MaterialExpressionConstant3Vector_7 {class=MaterialExpressionConstant3Vector}
  MaterialExpressionConstantBiasScale_0 {
   class=MaterialExpressionConstantBiasScale
   inputs [None {from=MaterialExpressionMaterialFunctionCall_36, output=Result}]
   properties {bias=-0.5, scale=2}
  }
  MaterialExpressionConstantBiasScale_2 {class=MaterialExpressionConstantBiasScale, inputs [None {from=MaterialExpressionMaterialFunctionCall_19, output=Hemi Octahedron}]}
  MaterialExpressionConstant_0 {class=MaterialExpressionConstant, properties {r=-5000}}
  MaterialExpressionConstant_1 {class=MaterialExpressionConstant, properties {r=1}}
  MaterialExpressionConstant_10 {class=MaterialExpressionConstant}
  MaterialExpressionConstant_11 {class=MaterialExpressionConstant, properties {r=1}}
  MaterialExpressionConstant_12 {class=MaterialExpressionConstant, properties {r=1}}
  MaterialExpressionConstant_13 {class=MaterialExpressionConstant}
  MaterialExpressionConstant_14 {class=MaterialExpressionConstant, properties {r=1}}
  MaterialExpressionConstant_15 {class=MaterialExpressionConstant}
  MaterialExpressionConstant_16 {class=MaterialExpressionConstant}
  MaterialExpressionConstant_17 {class=MaterialExpressionConstant, properties {r=0.5}}
  MaterialExpressionConstant_18 {class=MaterialExpressionConstant, properties {r=1}}
  MaterialExpressionConstant_19 {class=MaterialExpressionConstant, properties {r=-0.5}}
  MaterialExpressionConstant_2 {class=MaterialExpressionConstant, properties {r=24}}
  MaterialExpressionConstant_20 {class=MaterialExpressionConstant, properties {r=0.5}}
  MaterialExpressionConstant_21 {class=MaterialExpressionConstant, properties {desc=Snow albedo, r=0.7}}
  MaterialExpressionConstant_22 {class=MaterialExpressionConstant, properties {r=1.5}}
  MaterialExpressionConstant_23 {class=MaterialExpressionConstant, properties {r=0.125}}
  MaterialExpressionConstant_24 {class=MaterialExpressionConstant, properties {r=0.375}}
  MaterialExpressionConstant_25 {class=MaterialExpressionConstant, properties {r=0.625}}
  MaterialExpressionConstant_26 {class=MaterialExpressionConstant}
  MaterialExpressionConstant_27 {class=MaterialExpressionConstant, properties {r=1}}
  MaterialExpressionConstant_28 {class=MaterialExpressionConstant, properties {r=1}}
  MaterialExpressionConstant_29 {class=MaterialExpressionConstant}
  MaterialExpressionConstant_3 {class=MaterialExpressionConstant}
  MaterialExpressionConstant_30 {class=MaterialExpressionConstant, properties {desc=Fixed value For Wold HueVariation, r=0.0125}}
  MaterialExpressionConstant_32 {class=MaterialExpressionConstant, properties {r=0.1}}
  MaterialExpressionConstant_34 {class=MaterialExpressionConstant}
  MaterialExpressionConstant_37 {class=MaterialExpressionConstant, properties {r=13000}}
  MaterialExpressionConstant_38 {class=MaterialExpressionConstant, properties {r=11000}}
  MaterialExpressionConstant_4 {class=MaterialExpressionConstant}
  MaterialExpressionConstant_5 {class=MaterialExpressionConstant}
  MaterialExpressionConstant_6 {class=MaterialExpressionConstant}
  MaterialExpressionConstant_7 {class=MaterialExpressionConstant}
  MaterialExpressionConstant_8 {class=MaterialExpressionConstant, properties {desc=Was originally CubePathThreshold}}
  MaterialExpressionConstant_9 {class=MaterialExpressionConstant}
  MaterialExpressionConvert_1 {
   class=MaterialExpressionConvert
   inputs [
    None {from=""}
    None {from=""}
    None {from=""}
   ]
   properties {
    nodeName=Make Vector3
    convertInputs [
     {
      type=Scalar
      defaultValue=<rgba 1 0 0 1>
      expressionInput {expression=null, inputName=None, mask=0, maskA=0, maskB=0, maskG=0, maskR=0, outputIndex=0}
     }
     {type=Scalar, expressionInput {expression=null, inputName=None, mask=0, maskA=0, maskB=0, maskG=0, maskR=0, outputIndex=0}}
     {type=Scalar, expressionInput {expression=null, inputName=None, mask=0, maskA=0, maskB=0, maskG=0, maskR=0, outputIndex=0}}
    ]
    convertOutputs [{type=Vector3}]
   }
  }
  MaterialExpressionCustom_0 {
   class=MaterialExpressionCustom
   inputs [InWorldVector {from="", output=""}]
   properties {
    code="\r\n##if USE_INSTANCING || IS_MESHPARTICLE_FACTORY\r\n\t\tfloat3 temp;\r\n\t\ttemp.x = length(TransformLocalVectorToWorld(Parameters, float3(1,0,0)));\r\n\t\ttemp.y = length(TransformLocalVectorToWorld(Parameters, float3(0,1,0)));\r\n\t\ttemp.z = length(TransformLocalVectorToWorld(Parameters, float3(0,0,1)));\r\n\t\treturn mul(InWorldVector, (MaterialFloat3x3)transpose(Parameters.InstanceLocalToWorld)) / (temp*temp);\r\n\t##else\r\n\t\treturn mul(InWorldVector, (MaterialFloat3x3)Primitive.WorldToLocal);\r\n\t##endif"
    description=World to Loc
    inputs [{inputName=InWorldVector, input {expression=null, inputName=None, mask=0, maskA=0, maskB=0, maskG=0, maskR=0, outputIndex=0}}]
   }
  }
  MaterialExpressionCustom_1 {
   class=MaterialExpressionCustom
   inputs [
    Main {from=MaterialExpressionCameraPositionWS_4, output=""}
    Shadow {from=MaterialExpressionNamedRerouteUsage_20, output=""}
   ]
   properties {
    code="##if SHADOW_DEPTH_SHADER \r\n{\r\nreturn Shadow;\r\n}\r\n##endif\r\nreturn Main;"
    description=Shadow Pass Switch
    inputs [
     {
      inputName=Main
      input {
       expression=&self:MaterialExpressionCameraPositionWS_4
       inputName=None
       mask=0
       maskA=0
       maskB=0
       maskG=0
       maskR=0
       outputIndex=0
      }
     }
     {
      inputName=Shadow
      input {
       expression=&self:MaterialExpressionNamedRerouteUsage_20
       inputName=None
       mask=0
       maskA=0
       maskB=0
       maskG=0
       maskR=0
       outputIndex=0
      }
     }
    ]
   }
  }
  MaterialExpressionCustom_10 {
   class=MaterialExpressionCustom
   inputs [
    Main {from=MaterialExpressionCameraVectorWS_1, output=""}
    Shadow {from=MaterialExpressionNamedRerouteUsage_47, output=""}
   ]
   properties {
    code="##if SHADOW_DEPTH_SHADER \r\n{\r\nreturn Shadow;\r\n}\r\n##endif\r\nreturn Main;"
    description=Shadow Pass Switch
    inputs [
     {
      inputName=Main
      input {
       expression=&self:MaterialExpressionCameraVectorWS_1
       inputName=None
       mask=0
       maskA=0
       maskB=0
       maskG=0
       maskR=0
       outputIndex=0
      }
     }
     {
      inputName=Shadow
      input {
       expression=&self:MaterialExpressionNamedRerouteUsage_47
       inputName=None
       mask=0
       maskA=0
       maskB=0
       maskG=0
       maskR=0
       outputIndex=0
      }
     }
    ]
   }
  }
  MaterialExpressionCustom_11 {
   class=MaterialExpressionCustom
   inputs [
    Main {from=MaterialExpressionMaterialFunctionCall_27, output=Result}
    Shadow {from=MaterialExpressionNamedRerouteUsage_48, output=""}
   ]
   properties {
    code="##if SHADOW_DEPTH_SHADER \r\n{\r\nreturn Shadow;\r\n}\r\n##endif\r\nreturn Main;"
    description=Shadow Pass Switch
    inputs [
     {
      inputName=Main
      input {
       expression=&self:MaterialExpressionMaterialFunctionCall_27
       inputName=None
       mask=0
       maskA=0
       maskB=0
       maskG=0
       maskR=0
       outputIndex=0
      }
     }
     {
      inputName=Shadow
      input {
       expression=&self:MaterialExpressionNamedRerouteUsage_48
       inputName=None
       mask=0
       maskA=0
       maskB=0
       maskG=0
       maskR=0
       outputIndex=0
      }
     }
    ]
   }
  }
  MaterialExpressionCustom_12 {
   class=MaterialExpressionCustom
   inputs [
    Main {from=MaterialExpressionMaterialFunctionCall_22, output=Result}
    Shadow {from=MaterialExpressionNamedRerouteUsage_49, output=""}
   ]
   properties {
    code="##if SHADOW_DEPTH_SHADER \r\n{\r\nreturn Shadow;\r\n}\r\n##endif\r\nreturn Main;"
    description=Shadow Pass Switch
    inputs [
     {
      inputName=Main
      input {
       expression=&self:MaterialExpressionMaterialFunctionCall_22
       inputName=None
       mask=0
       maskA=0
       maskB=0
       maskG=0
       maskR=0
       outputIndex=0
      }
     }
     {
      inputName=Shadow
      input {
       expression=&self:MaterialExpressionNamedRerouteUsage_49
       inputName=None
       mask=0
       maskA=0
       maskB=0
       maskG=0
       maskR=0
       outputIndex=0
      }
     }
    ]
   }
  }
  MaterialExpressionCustom_2 {
   class=MaterialExpressionCustom
   inputs [
    Main {from=MaterialExpressionShadingPathSwitch_2, output=""}
    Shadow {from=MaterialExpressionConstant_6, output=""}
   ]
   properties {
    code="##if SHADOW_DEPTH_SHADER \r\n{\r\nreturn Shadow;\r\n}\r\n##endif\r\nreturn Main;"
    description=Shadow Pass Switch
    inputs [
     {
      inputName=Main
      input {
       expression=&self:MaterialExpressionShadingPathSwitch_2
       inputName=None
       mask=0
       maskA=0
       maskB=0
       maskG=0
       maskR=0
       outputIndex=0
      }
     }
     {
      inputName=Shadow
      input {
       expression=&self:MaterialExpressionConstant_6
       inputName=None
       mask=0
       maskA=0
       maskB=0
       maskG=0
       maskR=0
       outputIndex=0
      }
     }
    ]
   }
  }
  MaterialExpressionCustom_3 {
   class=MaterialExpressionCustom
   inputs [InWorldVector {from="", output=""}]
   properties {
    code="\r\n##if USE_INSTANCING || IS_MESHPARTICLE_FACTORY\r\n\t\tfloat3 temp;\r\n\t\ttemp.x = length(TransformLocalVectorToWorld(Parameters, float3(1,0,0)));\r\n\t\ttemp.y = length(TransformLocalVectorToWorld(Parameters, float3(0,1,0)));\r\n\t\ttemp.z = length(TransformLocalVectorToWorld(Parameters, float3(0,0,1)));\r\n\t\treturn mul(InWorldVector, (MaterialFloat3x3)transpose(Parameters.InstanceLocalToWorld)) / (temp*temp);\r\n\t##else\r\n\t\treturn mul(InWorldVector, (MaterialFloat3x3)Primitive.WorldToLocal);\r\n\t##endif"
    description=World to Loc
    inputs [{inputName=InWorldVector, input {expression=null, inputName=None, mask=0, maskA=0, maskB=0, maskG=0, maskR=0, outputIndex=0}}]
   }
  }
  MaterialExpressionCustom_4 {
   class=MaterialExpressionCustom
   inputs []
   properties {code="View.ViewToClip[3][3]", desc=Ignore pixel depth offset is camera is ortho, outputType=CMOT_Float1, inputs []}
  }
  MaterialExpressionCustom_5 {
   class=MaterialExpressionCustom
   inputs [
    Main {from=MaterialExpressionComponentMask_10, output=""}
    Shadow {from=MaterialExpressionComponentMask_15, output=""}
   ]
   properties {
    code="##if SHADOW_DEPTH_SHADER \r\n{\r\nreturn Shadow;\r\n}\r\n##endif\r\nreturn Main;"
    description=Shadow Pass Switch
    inputs [
     {
      inputName=Main
      input {
       expression=&self:MaterialExpressionComponentMask_10
       inputName=None
       mask=0
       maskA=0
       maskB=0
       maskG=0
       maskR=0
       outputIndex=0
      }
     }
     {
      inputName=Shadow
      input {
       expression=&self:MaterialExpressionComponentMask_15
       inputName=None
       mask=0
       maskA=0
       maskB=0
       maskG=0
       maskR=0
       outputIndex=0
      }
     }
    ]
   }
  }
  MaterialExpressionCustom_6 {
   class=MaterialExpressionCustom
   inputs [
    Main {from=MaterialExpressionCameraPositionWS_1, output=""}
    Shadow {from=MaterialExpressionNamedRerouteUsage_44, output=""}
   ]
   properties {
    code="##if SHADOW_DEPTH_SHADER \r\n{\r\nreturn Shadow;\r\n}\r\n##endif\r\nreturn Main;"
    description=Shadow Pass Switch
    inputs [
     {
      inputName=Main
      input {
       expression=&self:MaterialExpressionCameraPositionWS_1
       inputName=None
       mask=0
       maskA=0
       maskB=0
       maskG=0
       maskR=0
       outputIndex=0
      }
     }
     {
      inputName=Shadow
      input {
       expression=&self:MaterialExpressionNamedRerouteUsage_44
       inputName=None
       mask=0
       maskA=0
       maskB=0
       maskG=0
       maskR=0
       outputIndex=0
      }
     }
    ]
   }
  }
  MaterialExpressionCustom_7 {
   class=MaterialExpressionCustom
   inputs [
    Main {from=MaterialExpressionAdd_0, output=""}
    Shadow {from=MaterialExpressionReroute_8, output=""}
   ]
   properties {
    code="##if SHADOW_DEPTH_SHADER \r\n{\r\nreturn Shadow;\r\n}\r\n##endif\r\nreturn Main;"
    description=Shadow Pass Switch
    outputType=CMOT_Float2
    inputs [
     {
      inputName=Main
      input {
       expression=&self:MaterialExpressionAdd_0
       inputName=None
       mask=0
       maskA=0
       maskB=0
       maskG=0
       maskR=0
       outputIndex=0
      }
     }
     {
      inputName=Shadow
      input {
       expression=&self:MaterialExpressionReroute_8
       inputName=None
       mask=0
       maskA=0
       maskB=0
       maskG=0
       maskR=0
       outputIndex=0
      }
     }
    ]
   }
  }
  MaterialExpressionCustom_8 {
   class=MaterialExpressionCustom
   inputs [
    Main {from=MaterialExpressionCameraPositionWS_3, output=""}
    Shadow {from=MaterialExpressionNamedRerouteUsage_45, output=""}
   ]
   properties {
    code="##if SHADOW_DEPTH_SHADER \r\n{\r\nreturn Shadow;\r\n}\r\n##endif\r\nreturn Main;"
    description=Shadow Pass Switch
    inputs [
     {
      inputName=Main
      input {
       expression=&self:MaterialExpressionCameraPositionWS_3
       inputName=None
       mask=0
       maskA=0
       maskB=0
       maskG=0
       maskR=0
       outputIndex=0
      }
     }
     {
      inputName=Shadow
      input {
       expression=&self:MaterialExpressionNamedRerouteUsage_45
       inputName=None
       mask=0
       maskA=0
       maskB=0
       maskG=0
       maskR=0
       outputIndex=0
      }
     }
    ]
   }
  }
  MaterialExpressionCustom_9 {
   class=MaterialExpressionCustom
   inputs [
    Main {from=MaterialExpressionMaterialFunctionCall_23, output=Result}
    Shadow {from=MaterialExpressionNamedRerouteUsage_47, output=""}
   ]
   properties {
    code="##if SHADOW_DEPTH_SHADER \r\n{\r\nreturn Shadow;\r\n}\r\n##endif\r\nreturn Main;"
    description=Shadow Pass Switch
    inputs [
     {
      inputName=Main
      input {
       expression=&self:MaterialExpressionMaterialFunctionCall_23
       inputName=None
       mask=0
       maskA=0
       maskB=0
       maskG=0
       maskR=0
       outputIndex=0
      }
     }
     {
      inputName=Shadow
      input {
       expression=&self:MaterialExpressionNamedRerouteUsage_47
       inputName=None
       mask=0
       maskA=0
       maskB=0
       maskG=0
       maskR=0
       outputIndex=0
      }
     }
    ]
   }
  }
  MaterialExpressionDesaturation_0 {
   class=MaterialExpressionDesaturation
   inputs [
    None {from=MaterialExpressionNamedRerouteUsage_22, output=""}
    Fraction {from=MaterialExpressionScalarParameter_62, output=""}
   ]
   properties {}
  }
  MaterialExpressionDesaturation_1 {
   class=MaterialExpressionDesaturation
   inputs [
    None {from=MaterialExpressionReroute_4, output=""}
    Fraction {from=MaterialExpressionConstant_14, output=""}
   ]
   properties {}
  }
  MaterialExpressionDistanceFieldsRenderingSwitch_0 {
   class=MaterialExpressionDistanceFieldsRenderingSwitch
   inputs [
    No {from=MaterialExpressionAdd_3, output=""}
    Yes {from=MaterialExpressionReroute_31, output=""}
   ]
   properties {}
  }
  MaterialExpressionDistanceFieldsRenderingSwitch_1 {
   class=MaterialExpressionDistanceFieldsRenderingSwitch
   inputs [
    No {from=MaterialExpressionConstant_7, output=""}
    Yes {from=MaterialExpressionAdd_17, output=""}
   ]
   properties {}
  }
  MaterialExpressionDistance_0 {
   class=MaterialExpressionDistance
   inputs [
    A {from=MaterialExpressionCustom_6, output=""}
    B {from=MaterialExpressionMaterialFunctionCall_11, output=Position}
   ]
   properties {}
  }
  MaterialExpressionDistance_2 {
   class=MaterialExpressionDistance
   inputs [
    A {from=MaterialExpressionCustom_6, output=""}
    B {from=MaterialExpressionMaterialFunctionCall_28, output=Position}
   ]
   properties {}
  }
  MaterialExpressionDivide_0 {
   class=MaterialExpressionDivide
   inputs [
    A {from=MaterialExpressionComponentMask_0}
    B {from=MaterialExpressionScalarParameter_14}
   ]
   properties {constB=16}
  }
  MaterialExpressionDivide_1 {
   class=MaterialExpressionDivide
   inputs [
    A {from=MaterialExpressionSaturate_5}
    B {from=MaterialExpressionScalarParameter_13}
   ]
  }
  MaterialExpressionDivide_10 {
   class=MaterialExpressionDivide
   inputs [
    A {from=MaterialExpressionSubtract_20}
    B {from=MaterialExpressionScalarParameter_47}
   ]
  }
  MaterialExpressionDivide_11 {
   class=MaterialExpressionDivide
   inputs [
    A {from=MaterialExpressionSubtract_15}
    B {from=MaterialExpressionScalarParameter_64}
   ]
  }
  MaterialExpressionDivide_12 {
   class=MaterialExpressionDivide
   inputs [
    A {from=MaterialExpressionDotProduct_5}
    B {from=MaterialExpressionMaterialFunctionCall_41, output=V2 Length}
   ]
  }
  MaterialExpressionDivide_13 {
   class=MaterialExpressionDivide
   inputs [
    A {from=MaterialExpressionMaterialFunctionCall_43, output=R}
    B {from=MaterialExpressionScalarParameter_54}
   ]
  }
  MaterialExpressionDivide_14 {
   class=MaterialExpressionDivide
   inputs [
    A {from=MaterialExpressionAdd_18}
    B {from=""}
   ]
   properties {constB=4}
  }
  MaterialExpressionDivide_15 {
   class=MaterialExpressionDivide
   inputs [
    A {from=MaterialExpressionComponentMask_8}
    B {from=MaterialExpressionScalarParameter_31}
   ]
   properties {constB=16}
  }
  MaterialExpressionDivide_16 {
   class=MaterialExpressionDivide
   inputs [
    A {from=MaterialExpressionSaturate_7}
    B {from=MaterialExpressionScalarParameter_36}
   ]
  }
  MaterialExpressionDivide_17 {
   class=MaterialExpressionDivide
   inputs [
    A {from=MaterialExpressionConstant_12}
    B {from=MaterialExpressionScalarParameter_37}
   ]
  }
  MaterialExpressionDivide_18 {
   class=MaterialExpressionDivide
   inputs [
    A {from=MaterialExpressionMaterialFunctionCall_43, output=G}
    B {from=MaterialExpressionScalarParameter_56}
   ]
  }
  MaterialExpressionDivide_19 {
   class=MaterialExpressionDivide
   inputs [
    A {from=MaterialExpressionAdd_19}
    B {from=""}
   ]
  }
  MaterialExpressionDivide_2 {
   class=MaterialExpressionDivide
   inputs [
    A {from=MaterialExpressionLinearInterpolate_13}
    B {from=MaterialExpressionScalarParameter_11}
   ]
  }
  MaterialExpressionDivide_20 {
   class=MaterialExpressionDivide
   inputs [
    A {from=MaterialExpressionMaterialFunctionCall_63, output=G}
    B {from=MaterialExpressionStaticSwitchParameter_19}
   ]
  }
  MaterialExpressionDivide_21 {
   class=MaterialExpressionDivide
   inputs [
    A {from=MaterialExpressionFloor_1}
    B {from=MaterialExpressionScalarParameter_40}
   ]
  }
  MaterialExpressionDivide_22 {
   class=MaterialExpressionDivide
   inputs [
    A {from=MaterialExpressionAdd_29}
    B {from=""}
   ]
  }
  MaterialExpressionDivide_23 {
   class=MaterialExpressionDivide
   inputs [
    A {from=MaterialExpressionNamedRerouteUsage_42}
    B {from=MaterialExpressionVectorParameter_22, output=RGB}
   ]
  }
  MaterialExpressionDivide_24 {
   class=MaterialExpressionDivide
   inputs [
    A {from=MaterialExpressionNamedRerouteUsage_40}
    B {from=MaterialExpressionVectorParameter_24, output=RGB}
   ]
  }
  MaterialExpressionDivide_3 {
   class=MaterialExpressionDivide
   inputs [
    A {from=MaterialExpressionSubtract_7}
    B {from=MaterialExpressionScalarParameter_55}
   ]
  }
  MaterialExpressionDivide_4 {
   class=MaterialExpressionDivide
   inputs [
    A {from=MaterialExpressionComponentMask_1}
    B {from=MaterialExpressionScalarParameter_15}
   ]
   properties {constB=16}
  }
  MaterialExpressionDivide_5 {
   class=MaterialExpressionDivide
   inputs [
    A {from=MaterialExpressionSaturate_6}
    B {from=MaterialExpressionScalarParameter_34}
   ]
  }
  MaterialExpressionDivide_6 {
   class=MaterialExpressionDivide
   inputs [
    A {from=MaterialExpressionSubtract_6}
    B {from=MaterialExpressionScalarParameter_42}
   ]
  }
  MaterialExpressionDivide_7 {
   class=MaterialExpressionDivide
   inputs [
    A {from=MaterialExpressionSubtract_16}
    B {from=MaterialExpressionScalarParameter_46}
   ]
  }
  MaterialExpressionDivide_8 {
   class=MaterialExpressionDivide
   inputs [
    A {from=MaterialExpressionSubtract_19}
    B {from=MaterialExpressionMultiply_4}
   ]
  }
  MaterialExpressionDivide_9 {
   class=MaterialExpressionDivide
   inputs [
    A {from=MaterialExpressionMaterialFunctionCall_63, output=R}
    B {from=MaterialExpressionStaticSwitchParameter_18}
   ]
  }
  MaterialExpressionDotProduct_0 {
   class=MaterialExpressionDotProduct
   inputs [
    A {from=MaterialExpressionComponentMask_19, output=""}
    B {from=MaterialExpressionCameraVectorWS_0, output=""}
   ]
   properties {desc=L dot camera}
  }
  MaterialExpressionDotProduct_1 {
   class=MaterialExpressionDotProduct
   inputs [
    A {from=MaterialExpressionNormalize_11, output=""}
    B {from=MaterialExpressionNormalize_14, output=""}
   ]
   properties {}
  }
  MaterialExpressionDotProduct_2 {
   class=MaterialExpressionDotProduct
   inputs [
    A {from=MaterialExpressionNormalize_0, output=""}
    B {from=MaterialExpressionCustom_12, output=""}
   ]
   properties {}
  }
  MaterialExpressionDotProduct_3 {
   class=MaterialExpressionDotProduct
   inputs [
    A {from=MaterialExpressionNormalize_15, output=""}
    B {from=MaterialExpressionCustom_9, output=""}
   ]
   properties {}
  }
  MaterialExpressionDotProduct_4 {
   class=MaterialExpressionDotProduct
   inputs [
    A {from=MaterialExpressionNormalize_5, output=""}
    B {from=MaterialExpressionCustom_11, output=""}
   ]
   properties {}
  }
  MaterialExpressionDotProduct_5 {
   class=MaterialExpressionDotProduct
   inputs [
    A {from=MaterialExpressionMultiply_40, output=""}
    B {from=MaterialExpressionNormalize_4, output=""}
   ]
   properties {}
  }
  MaterialExpressionDotProduct_6 {
   class=MaterialExpressionDotProduct
   inputs [
    A {from=MaterialExpressionNormalize_7, output=""}
    B {from=MaterialExpressionConstant3Vector_0, output=""}
   ]
   properties {}
  }
  MaterialExpressionDotProduct_7 {
   class=MaterialExpressionDotProduct
   inputs [
    A {from=MaterialExpressionReroute_18, output=""}
    B {from=MaterialExpressionNamedRerouteUsage_19, output=""}
   ]
   properties {}
  }
  MaterialExpressionFloor_1 {class=MaterialExpressionFloor, inputs [None {from=MaterialExpressionVertexInterpolator_5, output=PS}], properties {}}
  MaterialExpressionFloor_2 {class=MaterialExpressionFloor, inputs [None {from=MaterialExpressionMultiply_17, output=""}], properties {}}
  MaterialExpressionFloor_3 {class=MaterialExpressionFloor, inputs [None {from=MaterialExpressionMultiply_27, output=""}], properties {}}
  MaterialExpressionFloor_4 {class=MaterialExpressionFloor, inputs [None {from=MaterialExpressionMultiply_1, output=""}], properties {}}
  MaterialExpressionFrac_1 {class=MaterialExpressionFrac, inputs [None {from=MaterialExpressionMultiply_1, output=""}], properties {}}
  MaterialExpressionFrac_10 {class=MaterialExpressionFrac, inputs [None {from=MaterialExpressionVertexInterpolator_5, output=PS}], properties {}}
  MaterialExpressionFrac_4 {class=MaterialExpressionFrac, inputs [None {from=MaterialExpressionMultiply_18, output=""}], properties {}}
  MaterialExpressionFrac_5 {class=MaterialExpressionFrac, inputs [None {from=MaterialExpressionMaterialFunctionCall_46, output=B}], properties {}}
  MaterialExpressionFrac_7 {class=MaterialExpressionFrac, inputs [None {from=MaterialExpressionMultiply_17, output=""}], properties {}}
  MaterialExpressionFrac_8 {class=MaterialExpressionFrac, inputs [None {from=MaterialExpressionVertexInterpolator_5, output=PS}], properties {}}
  MaterialExpressionGetMaterialAttributes_0 {
   class=MaterialExpressionGetMaterialAttributes
   inputs [None {from=MaterialExpressionSetMaterialAttributes_2, output=""}]
   properties {attributeGetTypes [
     #1
     #2
     #3
     #4
     #5
    ]}
  }
  MaterialExpressionGetMaterialAttributes_1 {
   class=MaterialExpressionGetMaterialAttributes
   inputs [None {from=MaterialExpressionSetMaterialAttributes_0, output=""}]
   properties {attributeGetTypes [#1]}
  }
  MaterialExpressionIf_0 {
   class=MaterialExpressionIf
   inputs [
    A {from=MaterialExpressionDivide_12, output=""}
    B {from=MaterialExpressionConstant_8, output=""}
    "A > B" {from=MaterialExpressionConstant_5, output=""}
    "A == B" {from=MaterialExpressionConstant_0, output=""}
    "A < B" {from=MaterialExpressionConstant_0, output=""}
   ]
   properties {}
  }
  MaterialExpressionLength_1 {class=MaterialExpressionLength, inputs [None {from=MaterialExpressionMultiply_80, output=""}], properties {}}
  MaterialExpressionLinearInterpolate_0 {
   class=MaterialExpressionLinearInterpolate
   inputs [
    A {from=MaterialExpressionVectorParameter_18, output=RGB}
    B {from=MaterialExpressionMultiply_61}
    Alpha {from=MaterialExpressionNamedRerouteUsage_31}
   ]
  }
  MaterialExpressionLinearInterpolate_1 {
   class=MaterialExpressionLinearInterpolate
   inputs [
    A {from=MaterialExpressionNormalize_2}
    B {from=MaterialExpressionNormalize_6}
    Alpha {from=MaterialExpressionMultiply_34}
   ]
  }
  MaterialExpressionLinearInterpolate_10 {
   class=MaterialExpressionLinearInterpolate
   inputs [
    A {from=""}
    B {from=""}
    Alpha {from=MaterialExpressionMaterialFunctionCall_21, output=Mask}
   ]
   properties {constA=1, constB=0}
  }
  MaterialExpressionLinearInterpolate_11 {
   class=MaterialExpressionLinearInterpolate
   inputs [
    A {from=MaterialExpressionLinearInterpolate_10}
    B {from=MaterialExpressionLinearInterpolate_9}
    Alpha {from=MaterialExpressionMaterialFunctionCall_29, output=Mask}
   ]
  }
  MaterialExpressionLinearInterpolate_12 {
   class=MaterialExpressionLinearInterpolate
   inputs [
    A {from=MaterialExpressionConstant2Vector_0}
    B {from=MaterialExpressionConstant2Vector_2}
    Alpha {from=MaterialExpressionMaterialFunctionCall_35, output=Mask}
   ]
  }
  MaterialExpressionLinearInterpolate_13 {
   class=MaterialExpressionLinearInterpolate
   inputs [
    A {from=MaterialExpressionConstant2Vector_3}
    B {from=MaterialExpressionConstant2Vector_4}
    Alpha {from=MaterialExpressionMaterialFunctionCall_20, output=Mask}
   ]
  }
  MaterialExpressionLinearInterpolate_14 {
   class=MaterialExpressionLinearInterpolate
   inputs [
    A {from=MaterialExpressionConstant_11}
    B {from=MaterialExpressionNamedRerouteUsage_11}
    Alpha {from=MaterialExpressionReroute_15}
   ]
  }
  MaterialExpressionLinearInterpolate_15 {
   class=MaterialExpressionLinearInterpolate
   inputs [
    A {from=MaterialExpressionConstant_13}
    B {from=MaterialExpressionNamedRerouteUsage_9}
    Alpha {from=MaterialExpressionVertexColor_1, output=R}
   ]
  }
  MaterialExpressionLinearInterpolate_16 {
   class=MaterialExpressionLinearInterpolate
   inputs [
    A {from=MaterialExpressionStaticSwitchParameter_44}
    B {from=MaterialExpressionStaticSwitchParameter_43}
    Alpha {from=MaterialExpressionSaturate_12}
   ]
  }
  MaterialExpressionLinearInterpolate_17 {
   class=MaterialExpressionLinearInterpolate
   inputs [
    A {from=MaterialExpressionVectorParameter_2, output=RGB}
    B {from=MaterialExpressionReroute_5}
    Alpha {from=MaterialExpressionScalarParameter_41}
   ]
  }
  MaterialExpressionLinearInterpolate_18 {
   class=MaterialExpressionLinearInterpolate
   inputs [
    A {from=MaterialExpressionLinearInterpolate_16}
    B {from=MaterialExpressionStaticSwitchParameter_46}
    Alpha {from=MaterialExpressionOneMinus_2}
   ]
  }
  MaterialExpressionLinearInterpolate_19 {
   class=MaterialExpressionLinearInterpolate
   inputs [
    A {from=MaterialExpressionConstant3Vector_4}
    B {from=MaterialExpressionGetMaterialAttributes_0, output=World Position Offset}
    Alpha {from=MaterialExpressionVertexColor_0, output=R}
   ]
  }
  MaterialExpressionLinearInterpolate_2 {
   class=MaterialExpressionLinearInterpolate
   inputs [
    A {from=MaterialExpressionAdd_28}
    B {from=MaterialExpressionAdd_28}
    Alpha {from=MaterialExpressionSaturate_8}
   ]
  }
  MaterialExpressionLinearInterpolate_20 {
   class=MaterialExpressionLinearInterpolate
   inputs [
    A {from=MaterialExpressionConstant_26}
    B {from=MaterialExpressionGetMaterialAttributes_0, output=Pixel Depth Offset}
    Alpha {from=MaterialExpressionVertexColor_0, output=R}
   ]
  }
  MaterialExpressionLinearInterpolate_21 {
   class=MaterialExpressionLinearInterpolate
   inputs [
    A {from=MaterialExpressionConstant_27}
    B {from=MaterialExpressionGetMaterialAttributes_0, output=Opacity Mask}
    Alpha {from=MaterialExpressionVertexColor_0, output=R}
   ]
  }
  MaterialExpressionLinearInterpolate_22 {
   class=MaterialExpressionLinearInterpolate
   inputs [
    A {from=MaterialExpressionTextureSampleParameter2D_6, output=RGB}
    B {from=MaterialExpressionGetMaterialAttributes_0, output=Base Color}
    Alpha {from=MaterialExpressionVertexColor_0, output=R}
   ]
  }
  MaterialExpressionLinearInterpolate_23 {
   class=MaterialExpressionLinearInterpolate
   inputs [
    A {from=MaterialExpressionTextureSample_7, output=RGB}
    B {from=MaterialExpressionTextureSample_6, output=RGB}
    Alpha {from=MaterialExpressionSaturate_12}
   ]
  }
  MaterialExpressionLinearInterpolate_24 {
   class=MaterialExpressionLinearInterpolate
   inputs [
    A {from=MaterialExpressionLinearInterpolate_23}
    B {from=MaterialExpressionTextureSample_8, output=RGB}
    Alpha {from=MaterialExpressionOneMinus_2}
   ]
  }
  MaterialExpressionLinearInterpolate_25 {
   class=MaterialExpressionLinearInterpolate
   inputs [
    A {from=MaterialExpressionComponentMask_24}
    B {from=MaterialExpressionReroute_39}
    Alpha {from=MaterialExpressionStaticSwitchParameter_9}
   ]
  }
  MaterialExpressionLinearInterpolate_26 {
   class=MaterialExpressionLinearInterpolate
   inputs [
    A {from=MaterialExpressionVectorParameter_12, output=RGB}
    B {from=MaterialExpressionMultiply_68}
    Alpha {from=MaterialExpressionLinearInterpolate_32}
   ]
  }
  MaterialExpressionLinearInterpolate_27 {
   class=MaterialExpressionLinearInterpolate
   inputs [
    A {from=MaterialExpressionVectorParameter_15, output=RGB}
    B {from=MaterialExpressionLinearInterpolate_31}
    Alpha {from=MaterialExpressionLinearInterpolate_32}
   ]
  }
  MaterialExpressionLinearInterpolate_28 {
   class=MaterialExpressionLinearInterpolate
   inputs [
    A {from=MaterialExpressionVectorParameter_16, output=RGB}
    B {from=MaterialExpressionMultiply_69}
    Alpha {from=MaterialExpressionLinearInterpolate_32}
   ]
  }
  MaterialExpressionLinearInterpolate_29 {
   class=MaterialExpressionLinearInterpolate
   inputs [
    A {from=MaterialExpressionScalarParameter_68}
    B {from=MaterialExpressionLinearInterpolate_34}
    Alpha {from=MaterialExpressionLinearInterpolate_32}
   ]
  }
  MaterialExpressionLinearInterpolate_3 {
   class=MaterialExpressionLinearInterpolate
   inputs [
    A {from=MaterialExpressionScalarParameter_18}
    B {from=""}
    Alpha {from=MaterialExpressionDotProduct_4}
   ]
  }
  MaterialExpressionLinearInterpolate_30 {
   class=MaterialExpressionLinearInterpolate
   inputs [
    A {from=MaterialExpressionScalarParameter_67}
    B {from=MaterialExpressionLinearInterpolate_33}
    Alpha {from=MaterialExpressionLinearInterpolate_32}
   ]
  }
  MaterialExpressionLinearInterpolate_31 {
   class=MaterialExpressionLinearInterpolate
   inputs [
    A {from=MaterialExpressionVectorParameter_19, output=RGB}
    B {from=MaterialExpressionVertexInterpolator_10, output=PS}
    Alpha {from=MaterialExpressionVertexInterpolator_11, output=PS}
   ]
  }
  MaterialExpressionLinearInterpolate_32 {
   class=MaterialExpressionLinearInterpolate
   inputs [
    A {from=MaterialExpressionVectorParameter_19, output=A}
    B {from=MaterialExpressionVertexInterpolator_11, output=PS}
    Alpha {from=MaterialExpressionVertexInterpolator_11, output=PS}
   ]
  }
  MaterialExpressionLinearInterpolate_33 {
   class=MaterialExpressionLinearInterpolate
   inputs [
    A {from=MaterialExpressionScalarParameter_69}
    B {from=MaterialExpressionPerInstanceCustomData_1}
    Alpha {from=MaterialExpressionVertexInterpolator_11, output=PS}
   ]
  }
  MaterialExpressionLinearInterpolate_34 {
   class=MaterialExpressionLinearInterpolate
   inputs [
    A {from=MaterialExpressionScalarParameter_70}
    B {from=MaterialExpressionPerInstanceCustomData_2}
    Alpha {from=MaterialExpressionVertexInterpolator_11, output=PS}
   ]
  }
  MaterialExpressionLinearInterpolate_35 {
   class=MaterialExpressionLinearInterpolate
   inputs [
    A {from=MaterialExpressionMaterialFunctionCall_51, output=Vector_Out}
    B {from=MaterialExpressionAdd_8}
    Alpha {from=MaterialExpressionNamedRerouteDeclaration_14}
   ]
  }
  MaterialExpressionLinearInterpolate_36 {
   class=MaterialExpressionLinearInterpolate
   inputs [
    A {from=MaterialExpressionScalarParameter_19}
    B {from=MaterialExpressionScalarParameter_57}
    Alpha {from=MaterialExpressionVertexColor_3, output=G}
   ]
  }
  MaterialExpressionLinearInterpolate_37 {
   class=MaterialExpressionLinearInterpolate
   inputs [
    A {from=MaterialExpressionReroute_13}
    B {from=MaterialExpressionLinearInterpolate_38}
    Alpha {from=MaterialExpressionSaturate_14}
   ]
  }
  MaterialExpressionLinearInterpolate_38 {
   class=MaterialExpressionLinearInterpolate
   inputs [
    A {from=MaterialExpressionVectorParameter_10, output=RGB}
    B {from=MaterialExpressionAdd_30}
    Alpha {from=MaterialExpressionNamedRerouteDeclaration_14}
   ]
  }
  MaterialExpressionLinearInterpolate_39 {
   class=MaterialExpressionLinearInterpolate
   inputs [
    A {from=MaterialExpressionScalarParameter_76}
    B {from=MaterialExpressionStaticSwitchParameter_17}
    Alpha {from=MaterialExpressionNamedRerouteUsage_32}
   ]
  }
  MaterialExpressionLinearInterpolate_4 {
   class=MaterialExpressionLinearInterpolate
   inputs [
    A {from=MaterialExpressionMaterialFunctionCall_10, output=Position}
    B {from=MaterialExpressionAdd_6}
    Alpha {from=MaterialExpressionSaturate_8}
   ]
  }
  MaterialExpressionLinearInterpolate_40 {
   class=MaterialExpressionLinearInterpolate
   inputs [
    A {from=MaterialExpressionScalarParameter_71}
    B {from=MaterialExpressionScalarParameter_77}
    Alpha {from=MaterialExpressionNamedRerouteUsage_33}
   ]
  }
  MaterialExpressionLinearInterpolate_41 {
   class=MaterialExpressionLinearInterpolate
   inputs [
    A {from=MaterialExpressionDivide_24}
    B {from=MaterialExpressionMultiply_53}
    Alpha {from=MaterialExpressionNamedRerouteUsage_39}
   ]
  }
  MaterialExpressionLinearInterpolate_42 {
   class=MaterialExpressionLinearInterpolate
   inputs [
    A {from=MaterialExpressionConstant3Vector_7}
    B {from=MaterialExpressionMultiply_79}
    Alpha {from=MaterialExpressionNamedRerouteUsage_39}
   ]
  }
  MaterialExpressionLinearInterpolate_43 {
   class=MaterialExpressionLinearInterpolate
   inputs [
    A {from=MaterialExpressionVertexNormalWS_0}
    B {from=MaterialExpressionGetMaterialAttributes_0, output=Normal}
    Alpha {from=MaterialExpressionVertexColor_0, output=R}
   ]
  }
  MaterialExpressionLinearInterpolate_5 {
   class=MaterialExpressionLinearInterpolate
   inputs [
    A {from=MaterialExpressionMultiply_13}
    B {from=MaterialExpressionMultiply_19}
    Alpha {from=MaterialExpressionSaturate_4}
   ]
  }
  MaterialExpressionLinearInterpolate_6 {
   class=MaterialExpressionLinearInterpolate
   inputs [
    A {from=MaterialExpressionTextureSampleParameter2D_2, output=RGB}
    B {from=MaterialExpressionNamedRerouteUsage_3}
    Alpha {from=MaterialExpressionReroute_15}
   ]
  }
  MaterialExpressionLinearInterpolate_7 {
   class=MaterialExpressionLinearInterpolate
   inputs [
    A {from=MaterialExpressionReroute_50}
    B {from=MaterialExpressionConstant_21}
    Alpha {from=MaterialExpressionSaturate_10}
   ]
   properties {constB=0.7}
  }
  MaterialExpressionLinearInterpolate_8 {
   class=MaterialExpressionLinearInterpolate
   inputs [
    A {from=MaterialExpressionConstant3Vector_2}
    B {from=MaterialExpressionNamedRerouteUsage_5}
    Alpha {from=MaterialExpressionVertexColor_1, output=R}
   ]
  }
  MaterialExpressionLinearInterpolate_9 {
   class=MaterialExpressionLinearInterpolate
   inputs [
    A {from=""}
    B {from=""}
    Alpha {from=MaterialExpressionMaterialFunctionCall_21, output=Mask}
   ]
  }
  MaterialExpressionMaterialFunctionCall_0 {
   class=MaterialExpressionMaterialFunctionCall
   inputs [
    Frame {from=MaterialExpressionFloor_4}
    XY Frames {from=MaterialExpressionScalarParameter_33}
   ]
   properties {materialFunction=/Game/ImpostorBaker/MaterialFunctions/ImposterFrameTransform_Setup}
  }
  MaterialExpressionMaterialFunctionCall_1 {
   class=MaterialExpressionMaterialFunctionCall
   inputs [
    Frame {from=MaterialExpressionAdd_1}
    XY Frames {from=MaterialExpressionScalarParameter_1}
   ]
   properties {materialFunction=&ImposterFrameTransform_Setup}
  }
  MaterialExpressionMaterialFunctionCall_10 {class=MaterialExpressionMaterialFunctionCall, inputs [], properties {materialFunction=/Game/ImpostorBaker/MaterialFunctions/ImpostorInfo}}
  MaterialExpressionMaterialFunctionCall_11 {class=MaterialExpressionMaterialFunctionCall, inputs [], properties {materialFunction=&ImpostorInfo}}
  MaterialExpressionMaterialFunctionCall_12 {class=MaterialExpressionMaterialFunctionCall, inputs [], properties {materialFunction=&ImpostorInfo}}
  MaterialExpressionMaterialFunctionCall_13 {
   class=MaterialExpressionMaterialFunctionCall
   inputs [
    Plane Normal Axis {from=MaterialExpressionMaterialFunctionCall_31, output=World Z}
    Plane X Axis {from=MaterialExpressionMaterialFunctionCall_31, output=World Z}
    Plane Y Axis {from=MaterialExpressionMultiply_3}
    UV Scale {from=MaterialExpressionMultiply_63}
    Ray Dir Local {from=MaterialExpressionTransform_14}
    Ray Origin Local {from=MaterialExpressionTransform_13}
   ]
   properties {materialFunction=/Game/ImpostorBaker/MaterialFunctions/VirtualPlaneCoordinates_Imposter}
  }
  MaterialExpressionMaterialFunctionCall_14 {class=MaterialExpressionMaterialFunctionCall, inputs [], properties {materialFunction=&ImpostorInfo}}
  MaterialExpressionMaterialFunctionCall_15 {class=MaterialExpressionMaterialFunctionCall, inputs [], properties {materialFunction=&ImpostorInfo}}
  MaterialExpressionMaterialFunctionCall_16 {
   class=MaterialExpressionMaterialFunctionCall
   inputs [
    Vector2 {from=MaterialExpressionAppendVector_5}
    MaximumNumberOfDigits {from=""}
    UVs {from=MaterialExpressionMaterialFunctionCall_17, output=UVs}
    DebugTextLocation RG_UpperRight BA_LowerLeft {from=""}
    Component Spacing {from=""}
   ]
   properties {materialFunction=/Engine/Functions/Engine_MaterialFunctions02/Utility/DebugFloat2Values}
  }
  MaterialExpressionMaterialFunctionCall_17 {
   class=MaterialExpressionMaterialFunctionCall
   inputs [
    UVs {from=MaterialExpressionFrac_7}
    Texture Scale {from=MaterialExpressionScalarParameter_25}
   ]
   properties {materialFunction=&ScaleUVsByCenter}
  }
  MaterialExpressionMaterialFunctionCall_18 {
   class=MaterialExpressionMaterialFunctionCall
   inputs [
    Default {from=MaterialExpressionComponentMask_9}
    Mobile Override {from=MaterialExpressionMaterialFunctionCall_7, output=BaseColor RGB}
   ]
   properties {materialFunction=/Game/ImpostorBaker/MaterialFunctions/OverrideMobile}
  }
  MaterialExpressionMaterialFunctionCall_19 {
   class=MaterialExpressionMaterialFunctionCall
   inputs [Unit Vector {from=MaterialExpressionNormalize_8}]
   properties {materialFunction=/Game/ImpostorBaker/MaterialFunctions/UnitVectorToHemiOctahedra}
  }
  MaterialExpressionMaterialFunctionCall_2 {
   class=MaterialExpressionMaterialFunctionCall
   inputs [
    Frame {from=MaterialExpressionAdd_2}
    XY Frames {from=MaterialExpressionScalarParameter_5}
   ]
   properties {materialFunction=&ImposterFrameTransform_Setup}
  }
  MaterialExpressionMaterialFunctionCall_20 {
   class=MaterialExpressionMaterialFunctionCall
   inputs [UV {from=MaterialExpressionFrac_8}]
   properties {materialFunction=/Game/ImpostorBaker/MaterialFunctions/TriangleInterpolator}
  }
  MaterialExpressionMaterialFunctionCall_21 {
   class=MaterialExpressionMaterialFunctionCall
   inputs [UV {from=MaterialExpressionFrac_7}]
   properties {materialFunction=&TriangleInterpolator}
  }
  MaterialExpressionMaterialFunctionCall_22 {
   class=MaterialExpressionMaterialFunctionCall
   inputs []
   properties {materialFunction=&CameraDirectionVector}
  }
  MaterialExpressionMaterialFunctionCall_23 {
   class=MaterialExpressionMaterialFunctionCall
   inputs []
   properties {materialFunction=&CameraDirectionVector}
  }
  MaterialExpressionMaterialFunctionCall_24 {
   class=MaterialExpressionMaterialFunctionCall
   inputs [
    Plane Normal Axis {from=MaterialExpressionMaterialFunctionCall_32, output=World Z}
    Plane X Axis {from=MaterialExpressionMaterialFunctionCall_32, output=World Z}
    Plane Y Axis {from=MaterialExpressionMultiply_36}
    UV Scale {from=MaterialExpressionMultiply_63}
    Ray Dir Local {from=MaterialExpressionTransform_14}
    Ray Origin Local {from=MaterialExpressionTransform_13}
   ]
   properties {materialFunction=&VirtualPlaneCoordinates_Imposter}
  }
  MaterialExpressionMaterialFunctionCall_25 {
   class=MaterialExpressionMaterialFunctionCall
   inputs [
    ImposterPosition {from=MaterialExpressionMaterialFunctionCall_45, output=OutputBasePosition}
    WPO_Input {from=""}
    In {from=""}
   ]
   properties {materialFunction=/Game/Athena/Apollo/Environments/Landscape/Materials/Biome/UniqueBiomeMaterialFunctions/MF_Apollo_WaterLevel}
  }
  MaterialExpressionMaterialFunctionCall_26 {class=MaterialExpressionMaterialFunctionCall, inputs [], properties {materialFunction=&ImpostorInfo}}
  MaterialExpressionMaterialFunctionCall_27 {
   class=MaterialExpressionMaterialFunctionCall
   inputs []
   properties {materialFunction=&CameraDirectionVector}
  }
  MaterialExpressionMaterialFunctionCall_28 {class=MaterialExpressionMaterialFunctionCall, inputs [], properties {materialFunction=&ImpostorInfo}}
  MaterialExpressionMaterialFunctionCall_29 {
   class=MaterialExpressionMaterialFunctionCall
   inputs [UV {from=MaterialExpressionFrac_10}]
   properties {materialFunction=&TriangleInterpolator}
  }
  MaterialExpressionMaterialFunctionCall_3 {class=MaterialExpressionMaterialFunctionCall, inputs [], properties {materialFunction=&ImpostorInfo}}
  MaterialExpressionMaterialFunctionCall_30 {
   class=MaterialExpressionMaterialFunctionCall
   inputs [
    InVec {from=MaterialExpressionMultiply_20}
    X Vec {from=""}
    Y Vec {from=""}
    Z Vec {from=MaterialExpressionMaterialFunctionCall_0, output=World Z}
   ]
   properties {materialFunction=/Game/ImpostorBaker/MaterialFunctions/ImposterFrameTransform}
  }
  MaterialExpressionMaterialFunctionCall_31 {
   class=MaterialExpressionMaterialFunctionCall
   inputs [
    InVec {from=MaterialExpressionReroute_0}
    X Vec {from=""}
    Y Vec {from=""}
    Z Vec {from=MaterialExpressionMaterialFunctionCall_1, output=World Z}
   ]
   properties {materialFunction=&ImposterFrameTransform}
  }
  MaterialExpressionMaterialFunctionCall_32 {
   class=MaterialExpressionMaterialFunctionCall
   inputs [
    InVec {from=MaterialExpressionReroute_1}
    X Vec {from=""}
    Y Vec {from=""}
    Z Vec {from=MaterialExpressionMaterialFunctionCall_2, output=World Z}
   ]
   properties {materialFunction=&ImposterFrameTransform}
  }
  MaterialExpressionMaterialFunctionCall_33 {class=MaterialExpressionMaterialFunctionCall, inputs [], properties {materialFunction=&ImpostorInfo}}
  MaterialExpressionMaterialFunctionCall_34 {
   class=MaterialExpressionMaterialFunctionCall
   inputs [
    Default {from=MaterialExpressionCustom_5}
    Mobile Override {from=MaterialExpressionMaterialFunctionCall_7, output=BaseColor A}
   ]
   properties {materialFunction=&OverrideMobile}
  }
  MaterialExpressionMaterialFunctionCall_35 {
   class=MaterialExpressionMaterialFunctionCall
   inputs [UV {from=MaterialExpressionFrac_1}]
   properties {materialFunction=&TriangleInterpolator}
  }
  MaterialExpressionMaterialFunctionCall_36 {
   class=MaterialExpressionMaterialFunctionCall
   inputs [
    Default {from=MaterialExpressionComponentMask_11}
    Mobile Override {from=MaterialExpressionMaterialFunctionCall_7, output=Normal RGB}
   ]
   properties {materialFunction=&OverrideMobile}
  }
  MaterialExpressionMaterialFunctionCall_37 {
   class=MaterialExpressionMaterialFunctionCall
   inputs [
    Default {from=MaterialExpressionComponentMask_3}
    Mobile Override {from=MaterialExpressionMaterialFunctionCall_7, output=Normal A}
   ]
   properties {materialFunction=&OverrideMobile}
  }
  MaterialExpressionMaterialFunctionCall_38 {
   class=MaterialExpressionMaterialFunctionCall
   inputs [
    Default {from=MaterialExpressionReroute_31}
    Mobile Override {from=MaterialExpressionDistanceFieldsRenderingSwitch_0}
   ]
   properties {materialFunction=&OverrideMobile}
  }
  MaterialExpressionMaterialFunctionCall_39 {
   class=MaterialExpressionMaterialFunctionCall
   inputs [Float3 {from=MaterialExpressionNormalize_2}]
   properties {materialFunction=&BreakOutFloat3Components}
  }
  MaterialExpressionMaterialFunctionCall_4 {
   class=MaterialExpressionMaterialFunctionCall
   inputs [
    Plane Normal Axis {from=MaterialExpressionMaterialFunctionCall_30, output=World Z}
    Plane X Axis {from=MaterialExpressionMaterialFunctionCall_30, output=World Z}
    Plane Y Axis {from=MaterialExpressionMultiply_14}
    UV Scale {from=MaterialExpressionMultiply_63}
    Ray Dir Local {from=MaterialExpressionTransform_14}
    Ray Origin Local {from=MaterialExpressionTransform_13}
   ]
   properties {materialFunction=&VirtualPlaneCoordinates_Imposter}
  }
  MaterialExpressionMaterialFunctionCall_40 {
   class=MaterialExpressionMaterialFunctionCall
   inputs [
    X {from=MaterialExpressionMaterialFunctionCall_39, output=R}
    Y {from=MaterialExpressionMaterialFunctionCall_39, output=R}
    Z {from=MaterialExpressionMax_0}
   ]
   properties {materialFunction=&MakeFloat3}
  }
  MaterialExpressionMaterialFunctionCall_41 {
   class=MaterialExpressionMaterialFunctionCall
   inputs [
    Vector 3 {from=""}
    Vector 2 {from=MaterialExpressionSubtract_22}
   ]
   properties {materialFunction=/Engine/Functions/Engine_MaterialFunctions02/Utility/VectorLength}
  }
  MaterialExpressionMaterialFunctionCall_42 {
   class=MaterialExpressionMaterialFunctionCall
   inputs [
    In {from=MaterialExpressionDotProduct_6}
    Contrast {from=MaterialExpressionConstant_9}
   ]
   properties {materialFunction=&CheapContrast}
  }
  MaterialExpressionMaterialFunctionCall_43 {
   class=MaterialExpressionMaterialFunctionCall
   inputs [Float3 {from=MaterialExpressionWorldPosition_4, output=XYZ}]
   properties {materialFunction=&BreakOutFloat3Components}
  }
  MaterialExpressionMaterialFunctionCall_44 {
   class=MaterialExpressionMaterialFunctionCall
   inputs [
    Hue Shift Percentage {from=MaterialExpressionMultiply_56}
    Texture {from=MaterialExpressionLinearInterpolate_18}
   ]
   properties {materialFunction=&HueShift}
  }
  MaterialExpressionMaterialFunctionCall_45 {class=MaterialExpressionMaterialFunctionCall, inputs [], properties {materialFunction=&ImpostorInfo}}
  MaterialExpressionMaterialFunctionCall_46 {
   class=MaterialExpressionMaterialFunctionCall
   inputs [Float3 {from=MaterialExpressionMultiply_86}]
   properties {materialFunction=&BreakOutFloat3Components}
  }
  MaterialExpressionMaterialFunctionCall_47 {
   class=MaterialExpressionMaterialFunctionCall
   inputs [
    In {from=MaterialExpressionStaticSwitchParameter_16}
    Mask {from=MaterialExpressionConstant_28}
    IsImpostor {from=MaterialExpressionStaticBool_3}
   ]
   properties {materialFunction=/Game/Environments/Apollo/Foliage/MaterialFunctions/MF_TreeFresnel}
  }
  MaterialExpressionMaterialFunctionCall_48 {
   class=MaterialExpressionMaterialFunctionCall
   inputs [
    IsStaticSnow {from=""}
    UseAlternateStaticSnowMask {from=""}
    AlternateStaticSnowMask {from=""}
    "UseRoad&&OcclusionMask" {from=MaterialExpressionStaticBool_1}
    IsForGrassMeshToPushDown {from=""}
    AdditionalMaskToRemoveSnow {from=""}
    IsFallingSnowFX {from=""}
    UseMinimapCoords {from=""}
    IncreaseSnowUsingSlope {from=""}
    UseSmoothBorder {from=MaterialExpressionStaticBool_2}
    LerpHighM {from=MaterialExpressionConstant_22}
    TemporalHeightLerpContrast {from=""}
    IsLobbyIsland {from=""}
    UseBR-IslandMask {from=""}
    IsTree {from=MaterialExpressionStaticBool_2}
    IsForRenderTarget {from=""}
    UseSparkle {from=""}
    UseSnowNormals {from=""}
    UseAdditionalMaskToRemoveSnow {from=""}
    IsTangentSpaceInput {from=""}
    UseVertexNormalsForBasicSnow {from=""}
    UseTangentSnowNormals {from=""}
    Material {from=MaterialExpressionReroute_40}
   ]
   properties {materialFunction=/Game/Athena/Apollo/Environments/Landscape/Materials/Biome/UniqueBiomeMaterialFunctions/MF_Apollo_Snow_Progression}
  }
  MaterialExpressionMaterialFunctionCall_49 {
   class=MaterialExpressionMaterialFunctionCall
   inputs [
    Texture Object {from=MaterialExpressionTextureObjectParameter_3}
    R UV {from=MaterialExpressionCustom_7}
    G UV {from=MaterialExpressionAdd_9}
    B UV {from=MaterialExpressionAdd_22}
    Weights {from=MaterialExpressionMultiply_44}
    Derivative UVs {from=MaterialExpressionMultiply_54}
   ]
   properties {materialFunction=/Game/ImpostorBaker/MaterialFunctions/FrameBlendWeights_TextureObject}
  }
  MaterialExpressionMaterialFunctionCall_5 {
   class=MaterialExpressionMaterialFunctionCall
   inputs []
   properties {materialFunction=/Engine/Functions/Engine_MaterialFunctions02/WorldPositionOffset/ObjectScale}
  }
  MaterialExpressionMaterialFunctionCall_50 {
   class=MaterialExpressionMaterialFunctionCall
   inputs [
    Texture Object {from=MaterialExpressionTextureObjectParameter_4}
    R UV {from=MaterialExpressionAdd_0}
    G UV {from=MaterialExpressionAdd_9}
    B UV {from=MaterialExpressionAdd_22}
    Weights {from=MaterialExpressionMultiply_44}
    Derivative UVs {from=MaterialExpressionMultiply_54}
   ]
   properties {materialFunction=&FrameBlendWeights_TextureObject}
  }
  MaterialExpressionMaterialFunctionCall_51 {
   class=MaterialExpressionMaterialFunctionCall
   inputs [
    Float In {from=""}
    Vector_In {from=MaterialExpressionVectorParameter_4, output=RGB}
    Index {from=MaterialExpressionConstant_2}
   ]
   properties {materialFunction=/Game/Environments/Asteria/Foliage/Trees/MasterMaterials/MaterialFunctions/MF_AsteriaTree_ArrayAttribute}
  }
  MaterialExpressionMaterialFunctionCall_52 {
   class=MaterialExpressionMaterialFunctionCall
   inputs [
    ExtraSnowMask {from=""}
    Simple Mask {from=""}
    ForceUseInWSNormal {from=""}
    UseActorPosition {from=MaterialExpressionStaticBool_4}
    UseSnowMaskNormal {from=""}
    UseColorTest {from=""}
    UseColorTestRock {from=""}
    In {from=MaterialExpressionStaticSwitchParameter_16}
   ]
   properties {materialFunction=/Game/Packages/Fortnite_UniqueMaterials_1/MaterialFunctions/Trees/MF_Tree_Snow_Legacy}
  }
  MaterialExpressionMaterialFunctionCall_53 {
   class=MaterialExpressionMaterialFunctionCall
   inputs [
    Fade Length {from=MaterialExpressionScalarParameter_86}
    Fade Offset {from=MaterialExpressionScalarParameter_87}
    For Vertex Shader {from=MaterialExpressionStaticBool_5}
   ]
   properties {materialFunction=&CameraDepthFade}
  }
  MaterialExpressionMaterialFunctionCall_59 {
   class=MaterialExpressionMaterialFunctionCall
   inputs [
    Hue Shift Percentage {from=MaterialExpressionNamedRerouteUsage_36}
    Texture {from=MaterialExpressionNamedRerouteUsage_38}
   ]
   properties {materialFunction=&HueShift}
  }
  MaterialExpressionMaterialFunctionCall_6 {
   class=MaterialExpressionMaterialFunctionCall
   inputs [
    Center Location {from=MaterialExpressionMaterialFunctionCall_12, output=Position}
    0-1 UVs {from=MaterialExpressionMultiply_15}
    XY scale {from=MaterialExpressionMultiply_29}
    OverridePivot {from=""}
    "Normals (Optional)" {from=""}
   ]
   properties {materialFunction=&Sprite}
  }
  MaterialExpressionMaterialFunctionCall_63 {
   class=MaterialExpressionMaterialFunctionCall
   inputs [Float3 {from=MaterialExpressionWorldPosition_3, output=XYZ}]
   properties {materialFunction=&BreakOutFloat3Components}
  }
  MaterialExpressionMaterialFunctionCall_7 {
   class=MaterialExpressionMaterialFunctionCall
   inputs []
   properties {materialFunction=/Game/ImpostorBaker/MaterialFunctions/Impostor_SingleFrameVersion}
  }
  MaterialExpressionMaterialFunctionCall_8 {
   class=MaterialExpressionMaterialFunctionCall
   inputs [
    Vector 3 {from=MaterialExpressionSubtract_12}
    Vector 2 {from=""}
   ]
   properties {desc=Dist to Geo, materialFunction=&VectorLength}
  }
  MaterialExpressionMaterialFunctionCall_9 {
   class=MaterialExpressionMaterialFunctionCall
   inputs [
    UV {from=MaterialExpressionFrac_4}
    Line Width {from=MaterialExpressionScalarParameter_20}
   ]
   properties {materialFunction=/Game/ImpostorBaker/MaterialFunctions/TriangulatedWireframe}
  }
  MaterialExpressionMax_0 {
   class=MaterialExpressionMax
   inputs [
    A {from=MaterialExpressionMaterialFunctionCall_39, output=B}
    B {from=""}
   ]
   properties {constB=0.001}
  }
  MaterialExpressionMax_2 {
   class=MaterialExpressionMax
   inputs [
    A {from=MaterialExpressionSaturate_1}
    B {from=MaterialExpressionScalarParameter_43}
   ]
  }
  MaterialExpressionMultiply_0 {
   class=MaterialExpressionMultiply
   inputs [
    A {from=MaterialExpressionSubtract_0}
    B {from=MaterialExpressionVertexInterpolator_4, output=PS}
   ]
  }
  MaterialExpressionMultiply_1 {
   class=MaterialExpressionMultiply
   inputs [
    A {from=MaterialExpressionClamp_0}
    B {from=MaterialExpressionSubtract_9}
   ]
  }
  MaterialExpressionMultiply_10 {
   class=MaterialExpressionMultiply
   inputs [
    A {from=MaterialExpressionMultiply_6}
    B {from=MaterialExpressionConstant_1}
   ]
  }
  MaterialExpressionMultiply_11 {
   class=MaterialExpressionMultiply
   inputs [
    A {from=MaterialExpressionMultiply_8}
    B {from=MaterialExpressionQualitySwitch_0}
   ]
  }
  MaterialExpressionMultiply_12 {
   class=MaterialExpressionMultiply
   inputs [
    A {from=MaterialExpressionScalarParameter_21}
    B {from=MaterialExpressionMaterialFunctionCall_9, output=Result}
   ]
  }
  MaterialExpressionMultiply_13 {
   class=MaterialExpressionMultiply
   inputs [
    A {from=MaterialExpressionVectorParameter_8, output=RGB}
    B {from=MaterialExpressionVectorParameter_8, output=RGB}
   ]
  }
  MaterialExpressionMultiply_14 {
   class=MaterialExpressionMultiply
   inputs [
    A {from=MaterialExpressionMaterialFunctionCall_30, output=World Y}
    B {from=""}
   ]
   properties {constB=-1}
  }
  MaterialExpressionMultiply_15 {
   class=MaterialExpressionMultiply
   inputs [
    A {from=MaterialExpressionTextureCoordinate_1}
    B {from=MaterialExpressionScalarParameter_26}
   ]
  }
  MaterialExpressionMultiply_16 {
   class=MaterialExpressionMultiply
   inputs [
    A {from=MaterialExpressionAdd_28}
    B {from=MaterialExpressionScalarParameter_7}
   ]
  }
  MaterialExpressionMultiply_17 {
   class=MaterialExpressionMultiply
   inputs [
    A {from=MaterialExpressionTextureCoordinate_3}
    B {from=MaterialExpressionSubtract_2}
   ]
  }
  MaterialExpressionMultiply_18 {
   class=MaterialExpressionMultiply
   inputs [
    A {from=MaterialExpressionTextureCoordinate_6}
    B {from=MaterialExpressionSubtract_4}
   ]
  }
  MaterialExpressionMultiply_19 {
   class=MaterialExpressionMultiply
   inputs [
    A {from=MaterialExpressionVectorParameter_0, output=RGB}
    B {from=MaterialExpressionVectorParameter_0, output=RGB}
   ]
  }
  MaterialExpressionMultiply_2 {
   class=MaterialExpressionMultiply
   inputs [
    A {from=MaterialExpressionMultiply_0}
    B {from=MaterialExpressionConstant_1}
   ]
  }
  MaterialExpressionMultiply_20 {
   class=MaterialExpressionMultiply
   inputs [
    A {from=MaterialExpressionNormalize_6}
    B {from=""}
   ]
  }
  MaterialExpressionMultiply_21 {
   class=MaterialExpressionMultiply
   inputs [
    A {from=MaterialExpressionMultiply_26}
    B {from=MaterialExpressionScalarParameter_17}
   ]
  }
  MaterialExpressionMultiply_22 {
   class=MaterialExpressionMultiply
   inputs [
    A {from=MaterialExpressionMaterialFunctionCall_14, output=Scaled Size}
    B {from=""}
   ]
   properties {constB=0.5}
  }
  MaterialExpressionMultiply_23 {
   class=MaterialExpressionMultiply
   inputs [
    A {from=MaterialExpressionScalarParameter_23}
    B {from=MaterialExpressionAbs_0}
   ]
  }
  MaterialExpressionMultiply_24 {
   class=MaterialExpressionMultiply
   inputs [
    A {from=MaterialExpressionSubtract_13}
    B {from=MaterialExpressionAbs_2}
   ]
  }
  MaterialExpressionMultiply_25 {
   class=MaterialExpressionMultiply
   inputs [
    A {from=MaterialExpressionMultiply_55}
    B {from=MaterialExpressionVectorParameter_6, output=RGB}
   ]
  }
  MaterialExpressionMultiply_26 {
   class=MaterialExpressionMultiply
   inputs [
    A {from=MaterialExpressionMultiply_24}
    B {from=MaterialExpressionNormalize_0}
   ]
  }
  MaterialExpressionMultiply_27 {
   class=MaterialExpressionMultiply
   inputs [
    A {from=MaterialExpressionClamp_0}
    B {from=MaterialExpressionSubtract_2}
   ]
  }
  MaterialExpressionMultiply_28 {
   class=MaterialExpressionMultiply
   inputs [
    A {from=MaterialExpressionScalarParameter_27}
    B {from=MaterialExpressionMaterialFunctionCall_16, output=ColorCodedOutput}
   ]
  }
  MaterialExpressionMultiply_29 {
   class=MaterialExpressionMultiply
   inputs [
    A {from=MaterialExpressionMaterialFunctionCall_12, output=Scaled Size}
    B {from=MaterialExpressionLinearInterpolate_3}
   ]
  }
  MaterialExpressionMultiply_3 {
   class=MaterialExpressionMultiply
   inputs [
    A {from=MaterialExpressionMaterialFunctionCall_31, output=World Y}
    B {from=""}
   ]
   properties {constB=-1}
  }
  MaterialExpressionMultiply_30 {
   class=MaterialExpressionMultiply
   inputs [
    A {from=MaterialExpressionSphereMask_1}
    B {from=MaterialExpressionMultiply_32}
   ]
  }
  MaterialExpressionMultiply_31 {
   class=MaterialExpressionMultiply
   inputs [
    A {from=MaterialExpressionAdd_10}
    B {from=""}
   ]
   properties {constB=2}
  }
  MaterialExpressionMultiply_32 {
   class=MaterialExpressionMultiply
   inputs [
    A {from=MaterialExpressionLinearInterpolate_11}
    B {from=MaterialExpressionMaterialFunctionCall_29, output=Result}
   ]
  }
  MaterialExpressionMultiply_33 {
   class=MaterialExpressionMultiply
   inputs [
    A {from=MaterialExpressionNamedRerouteDeclaration_20}
    B {from=""}
   ]
   properties {constB=-1}
  }
  MaterialExpressionMultiply_34 {
   class=MaterialExpressionMultiply
   inputs [
    A {from=MaterialExpressionScalarParameter_4}
    B {from=MaterialExpressionSaturate_8}
   ]
  }
  MaterialExpressionMultiply_35 {
   class=MaterialExpressionMultiply
   inputs [
    A {from=MaterialExpressionVectorParameter_1, output=RGB}
    B {from=MaterialExpressionSaturate_9}
   ]
  }
  MaterialExpressionMultiply_36 {
   class=MaterialExpressionMultiply
   inputs [
    A {from=MaterialExpressionMaterialFunctionCall_32, output=World Y}
    B {from=""}
   ]
   properties {constB=-1}
  }
  MaterialExpressionMultiply_37 {
   class=MaterialExpressionMultiply
   inputs [
    A {from=MaterialExpressionMultiply_25}
    B {from=MaterialExpressionSaturate_0}
   ]
  }
  MaterialExpressionMultiply_38 {
   class=MaterialExpressionMultiply
   inputs [
    A {from=MaterialExpressionSubtract_8}
    B {from=MaterialExpressionVertexInterpolator_2, output=PS}
   ]
  }
  MaterialExpressionMultiply_39 {
   class=MaterialExpressionMultiply
   inputs [
    A {from=MaterialExpressionMultiply_38}
    B {from=MaterialExpressionConstant_1}
   ]
  }
  MaterialExpressionMultiply_4 {
   class=MaterialExpressionMultiply
   inputs [
    A {from=MaterialExpressionMaterialFunctionCall_3, output=Scaled Size}
    B {from=""}
   ]
   properties {constB=0.25}
  }
  MaterialExpressionMultiply_40 {
   class=MaterialExpressionMultiply
   inputs [
    A {from=MaterialExpressionSubtract_21}
    B {from=MaterialExpressionConstant3Vector_1}
   ]
  }
  MaterialExpressionMultiply_41 {
   class=MaterialExpressionMultiply
   inputs [
    A {from=MaterialExpressionDesaturation_1}
    B {from=""}
   ]
   properties {constB=5}
  }
  MaterialExpressionMultiply_42 {
   class=MaterialExpressionMultiply
   inputs [
    A {from=MaterialExpressionVectorParameter_6, output=A}
    B {from=MaterialExpressionDivide_6}
   ]
  }
  MaterialExpressionMultiply_43 {
   class=MaterialExpressionMultiply
   inputs [
    A {from=MaterialExpressionMultiply_31}
    B {from=MaterialExpressionScalarParameter_39}
   ]
  }
  MaterialExpressionMultiply_44 {
   class=MaterialExpressionMultiply
   inputs [
    A {from=MaterialExpressionReroute_5}
    B {from=""}
   ]
  }
  MaterialExpressionMultiply_45 {
   class=MaterialExpressionMultiply
   inputs [
    A {from=MaterialExpressionNamedRerouteUsage_24}
    B {from=MaterialExpressionMultiply_51}
   ]
  }
  MaterialExpressionMultiply_46 {
   class=MaterialExpressionMultiply
   inputs [
    A {from=MaterialExpressionMultiply_47}
    B {from=MaterialExpressionMultiply_45}
   ]
  }
  MaterialExpressionMultiply_47 {
   class=MaterialExpressionMultiply
   inputs [
    A {from=MaterialExpressionPower_0}
    B {from=MaterialExpressionOneMinus_4}
   ]
  }
  MaterialExpressionMultiply_48 {
   class=MaterialExpressionMultiply
   inputs [
    A {from=MaterialExpressionStaticSwitchParameter_12}
    B {from=MaterialExpressionMultiply_52}
   ]
  }
  MaterialExpressionMultiply_49 {
   class=MaterialExpressionMultiply
   inputs [
    A {from=MaterialExpressionComponentMask_19}
    B {from=""}
   ]
   properties {constB=-1}
  }
  MaterialExpressionMultiply_5 {
   class=MaterialExpressionMultiply
   inputs [
    A {from=MaterialExpressionMaterialFunctionCall_6, output=World position offset}
    B {from=""}
   ]
  }
  MaterialExpressionMultiply_50 {
   class=MaterialExpressionMultiply
   inputs [
    A {from=MaterialExpressionNamedRerouteUsage_23}
    B {from=MaterialExpressionMultiply_46}
   ]
  }
  MaterialExpressionMultiply_51 {
   class=MaterialExpressionMultiply
   inputs [
    A {from=MaterialExpressionVectorParameter_13, output=RGB}
    B {from=MaterialExpressionVectorParameter_13, output=RGB}
   ]
  }
  MaterialExpressionMultiply_52 {
   class=MaterialExpressionMultiply
   inputs [
    A {from=MaterialExpressionStaticSwitchParameter_9}
    B {from=MaterialExpressionShadingPathSwitch_5}
   ]
  }
  MaterialExpressionMultiply_53 {
   class=MaterialExpressionMultiply
   inputs [
    A {from=MaterialExpressionScalarParameter_58}
    B {from=MaterialExpressionMultiply_78}
   ]
  }
  MaterialExpressionMultiply_54 {
   class=MaterialExpressionMultiply
   inputs [
    A {from=MaterialExpressionTextureCoordinate_2}
    B {from=MaterialExpressionScalarParameter_48}
   ]
  }
  MaterialExpressionMultiply_55 {
   class=MaterialExpressionMultiply
   inputs [
    A {from=MaterialExpressionReroute_9}
    B {from=MaterialExpressionSaturate_2}
   ]
  }
  MaterialExpressionMultiply_56 {
   class=MaterialExpressionMultiply
   inputs [
    A {from=MaterialExpressionComponentMask_18}
    B {from=MaterialExpressionStaticSwitchParameter_47}
   ]
  }
  MaterialExpressionMultiply_57 {
   class=MaterialExpressionMultiply
   inputs [
    A {from=MaterialExpressionMultiply_61}
    B {from=MaterialExpressionNamedRerouteUsage_27}
   ]
  }
  MaterialExpressionMultiply_58 {
   class=MaterialExpressionMultiply
   inputs [
    A {from=MaterialExpressionLinearInterpolate_0}
    B {from=MaterialExpressionNamedRerouteUsage_26}
   ]
  }
  MaterialExpressionMultiply_59 {
   class=MaterialExpressionMultiply
   inputs [
    A {from=MaterialExpressionOneMinus_1}
    B {from=MaterialExpressionMultiply_60}
   ]
  }
  MaterialExpressionMultiply_6 {
   class=MaterialExpressionMultiply
   inputs [
    A {from=MaterialExpressionSubtract_3}
    B {from=MaterialExpressionVertexInterpolator_3, output=PS}
   ]
  }
  MaterialExpressionMultiply_60 {
   class=MaterialExpressionMultiply
   inputs [
    A {from=MaterialExpressionMultiply_62}
    B {from=MaterialExpressionMultiply_23}
   ]
   properties {constB=0.5}
  }
  MaterialExpressionMultiply_61 {
   class=MaterialExpressionMultiply
   inputs [
    A {from=MaterialExpressionStaticSwitchParameter_2}
    B {from=MaterialExpressionScalarParameter_0}
   ]
  }
  MaterialExpressionMultiply_62 {
   class=MaterialExpressionMultiply
   inputs [
    A {from=MaterialExpressionScalarParameter_16}
    B {from=MaterialExpressionMaterialFunctionCall_5, output=Scale X}
   ]
  }
  MaterialExpressionMultiply_63 {
   class=MaterialExpressionMultiply
   inputs [
    A {from=MaterialExpressionMaterialFunctionCall_33, output=Scaled Size}
    B {from=MaterialExpressionScalarParameter_32}
   ]
  }
  MaterialExpressionMultiply_64 {
   class=MaterialExpressionMultiply
   inputs [
    A {from=MaterialExpressionComponentMask_18}
    B {from=MaterialExpressionStaticSwitchParameter_48}
   ]
  }
  MaterialExpressionMultiply_65 {
   class=MaterialExpressionMultiply
   inputs [
    A {from=MaterialExpressionAdd_24}
    B {from=MaterialExpressionMaterialFunctionCall_44, output=Result}
   ]
  }
  MaterialExpressionMultiply_66 {
   class=MaterialExpressionMultiply
   inputs [
    A {from=MaterialExpressionVectorParameter_11, output=RGB}
    B {from=MaterialExpressionTextureSampleParameter2D_4, output=RGB}
   ]
  }
  MaterialExpressionMultiply_67 {
   class=MaterialExpressionMultiply
   inputs [
    A {from=MaterialExpressionTextureCoordinate_7}
    B {from=MaterialExpressionScalarParameter_61}
   ]
  }
  MaterialExpressionMultiply_68 {
   class=MaterialExpressionMultiply
   inputs [
    A {from=MaterialExpressionScalarParameter_65}
    B {from=MaterialExpressionLinearInterpolate_31}
   ]
  }
  MaterialExpressionMultiply_69 {
   class=MaterialExpressionMultiply
   inputs [
    A {from=MaterialExpressionScalarParameter_66}
    B {from=MaterialExpressionLinearInterpolate_31}
   ]
  }
  MaterialExpressionMultiply_7 {
   class=MaterialExpressionMultiply
   inputs [
    A {from=MaterialExpressionMultiply_30}
    B {from=""}
   ]
   properties {constB=0.2}
  }
  MaterialExpressionMultiply_70 {
   class=MaterialExpressionMultiply
   inputs [
    A {from=MaterialExpressionQualitySwitch_1}
    B {from=MaterialExpressionSaturate_15}
   ]
  }
  MaterialExpressionMultiply_71 {
   class=MaterialExpressionMultiply
   inputs [
    A {from=MaterialExpressionScalarParameter_8}
    B {from=MaterialExpressionScalarParameter_74}
   ]
   properties {constB=0.66, desc=Compensate for lack of PDO}
  }
  MaterialExpressionMultiply_72 {
   class=MaterialExpressionMultiply
   inputs [
    A {from=MaterialExpressionStaticSwitchParameter_24}
    B {from=MaterialExpressionStaticSwitchParameter_20}
   ]
  }
  MaterialExpressionMultiply_73 {
   class=MaterialExpressionMultiply
   inputs [
    A {from=MaterialExpressionConstant_32}
    B {from=MaterialExpressionNamedRerouteUsage_35}
   ]
  }
  MaterialExpressionMultiply_74 {
   class=MaterialExpressionMultiply
   inputs [
    A {from=MaterialExpressionScalarParameter_80}
    B {from=MaterialExpressionReroute_14}
   ]
  }
  MaterialExpressionMultiply_75 {
   class=MaterialExpressionMultiply
   inputs [
    A {from=MaterialExpressionReroute_14}
    B {from=MaterialExpressionScalarParameter_79}
   ]
  }
  MaterialExpressionMultiply_76 {
   class=MaterialExpressionMultiply
   inputs [
    A {from=MaterialExpressionNamedRerouteUsage_37}
    B {from=MaterialExpressionMaterialFunctionCall_59, output=Result}
   ]
  }
  MaterialExpressionMultiply_78 {
   class=MaterialExpressionMultiply
   inputs [
    A {from=MaterialExpressionMultiply_74}
    B {from=MaterialExpressionNamedRerouteUsage_42}
   ]
  }
  MaterialExpressionMultiply_79 {
   class=MaterialExpressionMultiply
   inputs [
    A {from=MaterialExpressionMultiply_75}
    B {from=MaterialExpressionNamedRerouteUsage_42}
   ]
  }
  MaterialExpressionMultiply_8 {
   class=MaterialExpressionMultiply
   inputs [
    A {from=MaterialExpressionNormalize_3}
    B {from=MaterialExpressionMultiply_22}
   ]
  }
  MaterialExpressionMultiply_80 {
   class=MaterialExpressionMultiply
   inputs [
    A {from=MaterialExpressionMaterialFunctionCall_53, output=Result}
    B {from=MaterialExpressionAppendVector_2}
   ]
  }
  MaterialExpressionMultiply_81 {
   class=MaterialExpressionMultiply
   inputs [
    A {from=MaterialExpressionGetMaterialAttributes_0, output=Base Color}
    B {from=MaterialExpressionReroute_20}
   ]
   properties {constA=1}
  }
  MaterialExpressionMultiply_83 {
   class=MaterialExpressionMultiply
   inputs [
    A {from=MaterialExpressionSubtract_33}
    B {from=""}
   ]
   properties {constB=2}
  }
  MaterialExpressionMultiply_84 {
   class=MaterialExpressionMultiply
   inputs [
    A {from=MaterialExpressionMaterialFunctionCall_46, output=R}
    B {from=""}
   ]
   properties {constB=0.5}
  }
  MaterialExpressionMultiply_85 {
   class=MaterialExpressionMultiply
   inputs [
    A {from=MaterialExpressionMaterialFunctionCall_46, output=G}
    B {from=""}
   ]
   properties {constB=0.5}
  }
  MaterialExpressionMultiply_86 {
   class=MaterialExpressionMultiply
   inputs [
    A {from=MaterialExpressionTransformPosition_1}
    B {from=MaterialExpressionScalarParameter_59}
   ]
   properties {constB=0.0012}
  }
  MaterialExpressionMultiply_87 {
   class=MaterialExpressionMultiply
   inputs [
    A {from=MaterialExpressionSubtract_34}
    B {from=""}
   ]
   properties {constB=2}
  }
  MaterialExpressionMultiply_88 {
   class=MaterialExpressionMultiply
   inputs [
    A {from=MaterialExpressionMultiply_89}
    B {from=""}
   ]
   properties {constB=2}
  }
  MaterialExpressionMultiply_89 {
   class=MaterialExpressionMultiply
   inputs [
    A {from=MaterialExpressionReroute_39}
    B {from=MaterialExpressionStaticSwitchParameter_9}
   ]
  }
  MaterialExpressionMultiply_9 {
   class=MaterialExpressionMultiply
   inputs [
    A {from=MaterialExpressionMultiply_33}
    B {from=""}
   ]
   properties {constB=500000}
  }
  MaterialExpressionMultiply_90 {
   class=MaterialExpressionMultiply
   inputs [
    A {from=MaterialExpressionSubtract_35}
    B {from=MaterialExpressionComponentMask_24}
   ]
  }
  MaterialExpressionNamedRerouteDeclaration_0 {
   class=MaterialExpressionNamedRerouteDeclaration
   inputs [None {from=MaterialExpressionTransform_2}]
   properties {name=WorldSpaceNormal, nodeColor=<rgba 1 0.75 0 1>}
  }
  MaterialExpressionNamedRerouteDeclaration_1 {
   class=MaterialExpressionNamedRerouteDeclaration
   inputs [None {from=MaterialExpressionStaticSwitchParameter_13}]
   properties {name=BaseColorFinal, nodeColor=<rgba 0.2422 1 0 1>}
  }
  MaterialExpressionNamedRerouteDeclaration_10 {
   class=MaterialExpressionNamedRerouteDeclaration
   inputs [None {from=MaterialExpressionStaticSwitchParameter_0}]
   properties {name=BaseColorWithOptionalDesat, nodeColor=<rgba 0.6875 0 1 1>}
  }
  MaterialExpressionNamedRerouteDeclaration_11 {
   class=MaterialExpressionNamedRerouteDeclaration
   inputs [None {from=MaterialExpressionStaticSwitchParameter_1}]
   properties {name=Legacy BaseColor, nodeColor=<rgba 1 0.375 0 1>}
  }
  MaterialExpressionNamedRerouteDeclaration_12 {
   class=MaterialExpressionNamedRerouteDeclaration
   inputs [None {from=MaterialExpressionComponentMask_23}]
   properties {name=ApolloColorSystemBaseColor, nodeColor=<rgba 1 0.1875 0 1>}
  }
  MaterialExpressionNamedRerouteDeclaration_13 {
   class=MaterialExpressionNamedRerouteDeclaration
   inputs [None {from=MaterialExpressionSaturate_3}]
   properties {name=ExtractTrunkMask, nodeColor=<rgba 1 0 0.4687 1>}
  }
  MaterialExpressionNamedRerouteDeclaration_14 {
   class=MaterialExpressionNamedRerouteDeclaration
   inputs [None {from=MaterialExpressionAdd_26}]
   properties {name=PrimDataBlendAlpha, nodeColor=<rgba 0 1 0.3203 1>}
  }
  MaterialExpressionNamedRerouteDeclaration_15 {
   class=MaterialExpressionNamedRerouteDeclaration
   inputs [None {from=MaterialExpressionReroute_13}]
   properties {name=AsteriaLeafColor, nodeColor=<rgba 0 1 0.5547 1>}
  }
  MaterialExpressionNamedRerouteDeclaration_16 {
   class=MaterialExpressionNamedRerouteDeclaration
   inputs [None {from=MaterialExpressionLinearInterpolate_41}]
   properties {name=AsteriaColorSystem, nodeColor=<rgba 0.6172 1 0 1>}
  }
  MaterialExpressionNamedRerouteDeclaration_17 {
   class=MaterialExpressionNamedRerouteDeclaration
   inputs [None {from=MaterialExpressionMultiply_72}]
   properties {name=AsteriaHueShift, nodeColor=<rgba 0 1 0.7891 1>}
  }
  MaterialExpressionNamedRerouteDeclaration_18 {
   class=MaterialExpressionNamedRerouteDeclaration
   inputs [None {from=MaterialExpressionAdd_27}]
   properties {name=AsteriaValueVariation, nodeColor=<rgba 0 0.0625 1 1>}
  }
  MaterialExpressionNamedRerouteDeclaration_19 {
   class=MaterialExpressionNamedRerouteDeclaration
   inputs [None {from=MaterialExpressionMultiply_9}]
   properties {name=CameraPosition Replace, nodeColor=<rgba 1 0.5391 0 1>}
  }
  MaterialExpressionNamedRerouteDeclaration_2 {
   class=MaterialExpressionNamedRerouteDeclaration
   inputs [None {from=MaterialExpressionMaterialFunctionCall_38, output=Result}]
   properties {name=WPO Final, nodeColor=<rgba 0 0.1094 1 1>}
  }
  MaterialExpressionNamedRerouteDeclaration_20 {
   class=MaterialExpressionNamedRerouteDeclaration
   inputs [None {from=MaterialExpressionNormalize_1}]
   properties {name=Camera Vector Replace, nodeColor=<rgba 0 1 0.7656 1>}
  }
  MaterialExpressionNamedRerouteDeclaration_21 {
   class=MaterialExpressionNamedRerouteDeclaration
   inputs [None {from=MaterialExpressionDivide_22}]
   properties {name=PerActorRandom, nodeColor=<rgba 0 1 0.7891 1>}
  }
  MaterialExpressionNamedRerouteDeclaration_23 {
   class=MaterialExpressionNamedRerouteDeclaration
   inputs [None {from=MaterialExpressionAdd_16}]
   properties {name=SubSurfaceColorLegacy, nodeColor=<rgba 0 1 0.3672 1>}
  }
  MaterialExpressionNamedRerouteDeclaration_24 {
   class=MaterialExpressionNamedRerouteDeclaration
   inputs [None {from=MaterialExpressionLinearInterpolate_42}]
   properties {name=SubSurfaceColorAsteria, nodeColor=<rgba 0 0.7187 1 1>}
  }
  MaterialExpressionNamedRerouteDeclaration_3 {
   class=MaterialExpressionNamedRerouteDeclaration
   inputs [None {from=MaterialExpressionMultiply_70}]
   properties {name=PixelDepthOffset, nodeColor=<rgba 0 0.8828 1 1>}
  }
  MaterialExpressionNamedRerouteDeclaration_4 {
   class=MaterialExpressionNamedRerouteDeclaration
   inputs [None {from=MaterialExpressionStaticSwitch_2}]
   properties {name=OpacityMask Final, nodeColor=<rgba 0.2656 1 0 1>}
  }
  MaterialExpressionNamedRerouteDeclaration_5 {
   class=MaterialExpressionNamedRerouteDeclaration
   inputs [None {from=MaterialExpressionStaticSwitchParameter_23}]
   properties {name=SubSurfaceColor Final, nodeColor=<rgba 1 0 0.2109 1>}
  }
  MaterialExpressionNamedRerouteDeclaration_6 {
   class=MaterialExpressionNamedRerouteDeclaration
   inputs [None {from=MaterialExpressionStaticSwitchParameter_10}]
   properties {name=Emissive Final, nodeColor=<rgba 1 0.6563 0 1>}
  }
  MaterialExpressionNamedRerouteDeclaration_7 {
   class=MaterialExpressionNamedRerouteDeclaration
   inputs [None {from=MaterialExpressionMultiply_43}]
   properties {name=Debug Frames, nodeColor=<rgba 0.4531 1 0 1>}
  }
  MaterialExpressionNamedRerouteDeclaration_8 {
   class=MaterialExpressionNamedRerouteDeclaration
   inputs [None {from=MaterialExpressionStaticSwitch_1}]
   properties {name=DebugFrame Emissive, nodeColor=<rgba 0.6875 1 0 1>}
  }
  MaterialExpressionNamedRerouteDeclaration_9 {
   class=MaterialExpressionNamedRerouteDeclaration
   inputs [None {from=MaterialExpressionMaterialFunctionCall_18, output=Result}]
   properties {name=BaseColorTexture, nodeColor=<rgba 0 0.6484 1 1>}
  }
  MaterialExpressionNamedRerouteUsage_0 {class=MaterialExpressionNamedRerouteUsage}
  MaterialExpressionNamedRerouteUsage_1 {class=MaterialExpressionNamedRerouteUsage}
  MaterialExpressionNamedRerouteUsage_10 {class=MaterialExpressionNamedRerouteUsage}
  MaterialExpressionNamedRerouteUsage_11 {class=MaterialExpressionNamedRerouteUsage}
  MaterialExpressionNamedRerouteUsage_12 {class=MaterialExpressionNamedRerouteUsage}
  MaterialExpressionNamedRerouteUsage_13 {class=MaterialExpressionNamedRerouteUsage}
  MaterialExpressionNamedRerouteUsage_14 {class=MaterialExpressionNamedRerouteUsage}
  MaterialExpressionNamedRerouteUsage_15 {class=MaterialExpressionNamedRerouteUsage}
  MaterialExpressionNamedRerouteUsage_16 {class=MaterialExpressionNamedRerouteUsage}
  MaterialExpressionNamedRerouteUsage_17 {class=MaterialExpressionNamedRerouteUsage}
  MaterialExpressionNamedRerouteUsage_18 {class=MaterialExpressionNamedRerouteUsage}
  MaterialExpressionNamedRerouteUsage_19 {class=MaterialExpressionNamedRerouteUsage}
  MaterialExpressionNamedRerouteUsage_2 {class=MaterialExpressionNamedRerouteUsage}
  MaterialExpressionNamedRerouteUsage_20 {class=MaterialExpressionNamedRerouteUsage}
  MaterialExpressionNamedRerouteUsage_21 {class=MaterialExpressionNamedRerouteUsage}
  MaterialExpressionNamedRerouteUsage_22 {class=MaterialExpressionNamedRerouteUsage}
  MaterialExpressionNamedRerouteUsage_23 {class=MaterialExpressionNamedRerouteUsage}
  MaterialExpressionNamedRerouteUsage_24 {class=MaterialExpressionNamedRerouteUsage}
  MaterialExpressionNamedRerouteUsage_25 {class=MaterialExpressionNamedRerouteUsage}
  MaterialExpressionNamedRerouteUsage_26 {class=MaterialExpressionNamedRerouteUsage}
  MaterialExpressionNamedRerouteUsage_27 {class=MaterialExpressionNamedRerouteUsage}
  MaterialExpressionNamedRerouteUsage_28 {class=MaterialExpressionNamedRerouteUsage}
  MaterialExpressionNamedRerouteUsage_29 {class=MaterialExpressionNamedRerouteUsage}
  MaterialExpressionNamedRerouteUsage_3 {class=MaterialExpressionNamedRerouteUsage}
  MaterialExpressionNamedRerouteUsage_30 {class=MaterialExpressionNamedRerouteUsage}
  MaterialExpressionNamedRerouteUsage_31 {class=MaterialExpressionNamedRerouteUsage}
  MaterialExpressionNamedRerouteUsage_32 {class=MaterialExpressionNamedRerouteUsage}
  MaterialExpressionNamedRerouteUsage_33 {class=MaterialExpressionNamedRerouteUsage}
  MaterialExpressionNamedRerouteUsage_34 {class=MaterialExpressionNamedRerouteUsage}
  MaterialExpressionNamedRerouteUsage_35 {class=MaterialExpressionNamedRerouteUsage}
  MaterialExpressionNamedRerouteUsage_36 {class=MaterialExpressionNamedRerouteUsage}
  MaterialExpressionNamedRerouteUsage_37 {class=MaterialExpressionNamedRerouteUsage}
  MaterialExpressionNamedRerouteUsage_38 {class=MaterialExpressionNamedRerouteUsage}
  MaterialExpressionNamedRerouteUsage_39 {class=MaterialExpressionNamedRerouteUsage}
  MaterialExpressionNamedRerouteUsage_4 {class=MaterialExpressionNamedRerouteUsage}
  MaterialExpressionNamedRerouteUsage_40 {class=MaterialExpressionNamedRerouteUsage}
  MaterialExpressionNamedRerouteUsage_41 {class=MaterialExpressionNamedRerouteUsage}
  MaterialExpressionNamedRerouteUsage_42 {class=MaterialExpressionNamedRerouteUsage}
  MaterialExpressionNamedRerouteUsage_43 {class=MaterialExpressionNamedRerouteUsage}
  MaterialExpressionNamedRerouteUsage_44 {class=MaterialExpressionNamedRerouteUsage}
  MaterialExpressionNamedRerouteUsage_45 {class=MaterialExpressionNamedRerouteUsage}
  MaterialExpressionNamedRerouteUsage_46 {class=MaterialExpressionNamedRerouteUsage}
  MaterialExpressionNamedRerouteUsage_47 {class=MaterialExpressionNamedRerouteUsage}
  MaterialExpressionNamedRerouteUsage_48 {class=MaterialExpressionNamedRerouteUsage}
  MaterialExpressionNamedRerouteUsage_49 {class=MaterialExpressionNamedRerouteUsage}
  MaterialExpressionNamedRerouteUsage_5 {class=MaterialExpressionNamedRerouteUsage}
  MaterialExpressionNamedRerouteUsage_6 {class=MaterialExpressionNamedRerouteUsage}
  MaterialExpressionNamedRerouteUsage_7 {class=MaterialExpressionNamedRerouteUsage}
  MaterialExpressionNamedRerouteUsage_8 {class=MaterialExpressionNamedRerouteUsage}
  MaterialExpressionNamedRerouteUsage_9 {class=MaterialExpressionNamedRerouteUsage}
  MaterialExpressionNormalize_0 {class=MaterialExpressionNormalize, inputs [VectorInput {from=MaterialExpressionSubtract_12, output=""}], properties {desc=Vector to Geo}}
  MaterialExpressionNormalize_1 {class=MaterialExpressionNormalize, inputs [VectorInput {from=MaterialExpressionSkyAtmosphereLightDirection_0, output=""}]}
  MaterialExpressionNormalize_11 {class=MaterialExpressionNormalize, inputs [VectorInput {from=MaterialExpressionStaticSwitchParameter_21, output=""}]}
  MaterialExpressionNormalize_14 {class=MaterialExpressionNormalize, inputs [VectorInput {from=MaterialExpressionNamedRerouteUsage_21, output=""}]}
  MaterialExpressionNormalize_15 {class=MaterialExpressionNormalize, inputs [VectorInput {from=MaterialExpressionCustom_10, output=""}]}
  MaterialExpressionNormalize_2 {class=MaterialExpressionNormalize, inputs [VectorInput {from=MaterialExpressionTransform_15, output=""}]}
  MaterialExpressionNormalize_3 {class=MaterialExpressionNormalize, inputs [VectorInput {from=MaterialExpressionSubtract_11, output=""}], properties {desc=CameraVector}}
  MaterialExpressionNormalize_4 {class=MaterialExpressionNormalize, inputs [VectorInput {from=MaterialExpressionSubtract_22, output=""}]}
  MaterialExpressionNormalize_5 {class=MaterialExpressionNormalize, inputs [VectorInput {from=MaterialExpressionSubtract_14, output=""}]}
  MaterialExpressionNormalize_6 {class=MaterialExpressionNormalize, inputs [VectorInput {from=MaterialExpressionTransform_3, output=""}]}
  MaterialExpressionNormalize_7 {class=MaterialExpressionNormalize, inputs [VectorInput {from=MaterialExpressionNamedRerouteUsage_17, output=""}]}
  MaterialExpressionNormalize_8 {class=MaterialExpressionNormalize, inputs [VectorInput {from=MaterialExpressionMaterialFunctionCall_40, output=Result}]}
  MaterialExpressionNormalize_9 {class=MaterialExpressionNormalize, inputs [VectorInput {from=MaterialExpressionVectorParameter_14, output=RGB}]}
  MaterialExpressionOneMinus_0 {class=MaterialExpressionOneMinus, inputs [None {from=MaterialExpressionMaterialFunctionCall_34, output=Result}]}
  MaterialExpressionOneMinus_1 {class=MaterialExpressionOneMinus, inputs [None {from=MaterialExpressionComponentMask_14}]}
  MaterialExpressionOneMinus_2 {class=MaterialExpressionOneMinus, inputs [None {from=MaterialExpressionSaturate_13}]}
  MaterialExpressionOneMinus_3 {class=MaterialExpressionOneMinus, inputs [None {from=MaterialExpressionComponentMask_18}]}
  MaterialExpressionOneMinus_4 {class=MaterialExpressionOneMinus, inputs [None {from=MaterialExpressionComponentMask_4}]}
  MaterialExpressionOneMinus_5 {class=MaterialExpressionOneMinus, inputs [None {from=MaterialExpressionCustom_4}]}
  MaterialExpressionOneMinus_7 {class=MaterialExpressionOneMinus, inputs [None {from=MaterialExpressionLength_1}]}
  MaterialExpressionPerInstanceCustomData3Vector_0 {class=MaterialExpressionPerInstanceCustomData3Vector, inputs [DefaultValue {from="", output=""}], properties {desc=Color}}
  MaterialExpressionPerInstanceCustomData3Vector_1 {class=MaterialExpressionPerInstanceCustomData3Vector, inputs [DefaultValue {from="", output=""}], properties {desc=Color}}
  MaterialExpressionPerInstanceCustomData3Vector_2 {class=MaterialExpressionPerInstanceCustomData3Vector, inputs [DefaultValue {from="", output=""}], properties {}}
  MaterialExpressionPerInstanceCustomData3Vector_3 {class=MaterialExpressionPerInstanceCustomData3Vector, inputs [DefaultValue {from="", output=""}], properties {dataIndex=17}}
  MaterialExpressionPerInstanceCustomData_0 {
   class=MaterialExpressionPerInstanceCustomData
   inputs [DefaultValue {from=MaterialExpressionConstant_29, output=""}]
   properties {dataIndex=3, desc=Alpha}
  }
  MaterialExpressionPerInstanceCustomData_1 {class=MaterialExpressionPerInstanceCustomData, inputs [DefaultValue {from="", output=""}], properties {dataIndex=4, desc=Alpha}}
  MaterialExpressionPerInstanceCustomData_2 {class=MaterialExpressionPerInstanceCustomData, inputs [DefaultValue {from="", output=""}], properties {dataIndex=5, desc=Alpha}}
  MaterialExpressionPerInstanceCustomData_3 {class=MaterialExpressionPerInstanceCustomData, inputs [DefaultValue {from="", output=""}], properties {constDefaultValue=0.5, dataIndex=5}}
  MaterialExpressionPerInstanceCustomData_4 {class=MaterialExpressionPerInstanceCustomData, inputs [DefaultValue {from="", output=""}], properties {dataIndex=3}}
  MaterialExpressionPixelNormalWS_0 {class=MaterialExpressionPixelNormalWS, inputs [], properties {}}
  MaterialExpressionPower_0 {
   class=MaterialExpressionPower
   inputs [
    Base {from=MaterialExpressionMax_2}
    Exp {from=MaterialExpressionScalarParameter_2}
   ]
  }
  MaterialExpressionQualitySwitch_0 {
   class=MaterialExpressionQualitySwitch
   inputs [
    Default {from=MaterialExpressionScalarParameter_8, output=""}
    Low {from=MaterialExpressionMultiply_71, output=""}
    High {from="", output=""}
    Medium {from=MaterialExpressionMultiply_71, output=""}
    Epic {from="", output=""}
   ]
   properties {}
  }
  MaterialExpressionQualitySwitch_1 {
   class=MaterialExpressionQualitySwitch
   inputs [
    Default {from=MaterialExpressionShadingPathSwitch_1, output=""}
    Low {from=MaterialExpressionConstant_15, output=""}
    High {from="", output=""}
    Medium {from=MaterialExpressionConstant_15, output=""}
    Epic {from="", output=""}
   ]
   properties {}
  }
  MaterialExpressionReroute_0 {class=MaterialExpressionReroute, inputs [None {from=MaterialExpressionMultiply_20}]}
  MaterialExpressionReroute_1 {class=MaterialExpressionReroute, inputs [None {from=MaterialExpressionReroute_0}]}
  MaterialExpressionReroute_10 {class=MaterialExpressionReroute, inputs [None {from=MaterialExpressionMultiply_9}]}
  MaterialExpressionReroute_11 {class=MaterialExpressionReroute, inputs [None {from=MaterialExpressionAdd_28}]}
  MaterialExpressionReroute_12 {class=MaterialExpressionReroute, inputs [None {from=MaterialExpressionNamedRerouteUsage_6}]}
  MaterialExpressionReroute_13 {class=MaterialExpressionReroute, inputs [None {from=MaterialExpressionLinearInterpolate_35}]}
  MaterialExpressionReroute_14 {class=MaterialExpressionReroute, inputs [None {from=MaterialExpressionMultiply_76}]}
  MaterialExpressionReroute_15 {class=MaterialExpressionReroute, inputs [None {from=MaterialExpressionVertexColor_2, output=R}]}
  MaterialExpressionReroute_16 {class=MaterialExpressionReroute, inputs [None {from=MaterialExpressionSaturate_11}]}
  MaterialExpressionReroute_17 {class=MaterialExpressionReroute, inputs [None {from=MaterialExpressionNormalize_1}]}
  MaterialExpressionReroute_18 {class=MaterialExpressionReroute, inputs [None {from=MaterialExpressionComponentMask_17}]}
  MaterialExpressionReroute_19 {class=MaterialExpressionReroute, inputs [None {from=MaterialExpressionAdd_16}]}
  MaterialExpressionReroute_2 {class=MaterialExpressionReroute, inputs [None {from=MaterialExpressionDivide_21}]}
  MaterialExpressionReroute_20 {class=MaterialExpressionReroute, inputs [None {from=MaterialExpressionSaturate_17}]}
  MaterialExpressionReroute_21 {class=MaterialExpressionReroute, inputs [None {from=MaterialExpressionStaticSwitchParameter_11}]}
  MaterialExpressionReroute_3 {class=MaterialExpressionReroute, inputs [None {from=MaterialExpressionDivide_21}]}
  MaterialExpressionReroute_31 {class=MaterialExpressionReroute, inputs [None {from=MaterialExpressionReroute_11}]}
  MaterialExpressionReroute_38 {class=MaterialExpressionReroute, inputs [None {from=MaterialExpressionTransform_2}]}
  MaterialExpressionReroute_39 {class=MaterialExpressionReroute, inputs [None {from=MaterialExpressionShadingPathSwitch_0}]}
  MaterialExpressionReroute_4 {class=MaterialExpressionReroute, inputs [None {from=MaterialExpressionComponentMask_6}]}
  MaterialExpressionReroute_40 {class=MaterialExpressionReroute, inputs [None {from=MaterialExpressionStaticSwitchParameter_14}]}
  MaterialExpressionReroute_41 {class=MaterialExpressionReroute, inputs [None {from=""}]}
  MaterialExpressionReroute_42 {class=MaterialExpressionReroute, inputs [None {from=""}]}
  MaterialExpressionReroute_45 {class=MaterialExpressionReroute, inputs [None {from=MaterialExpressionMultiply_43}]}
  MaterialExpressionReroute_5 {class=MaterialExpressionReroute, inputs [None {from=MaterialExpressionMaterialFunctionCall_20, output=Result}]}
  MaterialExpressionReroute_50 {class=MaterialExpressionReroute, inputs [None {from=MaterialExpressionGetMaterialAttributes_1, output=Base Color}]}
  MaterialExpressionReroute_6 {class=MaterialExpressionReroute, inputs [None {from=MaterialExpressionComponentMask_13}]}
  MaterialExpressionReroute_7 {class=MaterialExpressionReroute, inputs [None {from=MaterialExpressionOneMinus_0}]}
  MaterialExpressionReroute_8 {class=MaterialExpressionReroute, inputs [None {from=MaterialExpressionAdd_20}]}
  MaterialExpressionReroute_9 {class=MaterialExpressionReroute, inputs [None {from=MaterialExpressionMaterialFunctionCall_18, output=Result}]}
  MaterialExpressionRound_0 {class=MaterialExpressionRound, inputs [None {from=MaterialExpressionCustom_4, output=""}], properties {}}
  MaterialExpressionSaturate_0 {class=MaterialExpressionSaturate, inputs [None {from=MaterialExpressionMultiply_42}]}
  MaterialExpressionSaturate_1 {class=MaterialExpressionSaturate, inputs [None {from=MaterialExpressionDotProduct_0}]}
  MaterialExpressionSaturate_10 {class=MaterialExpressionSaturate, inputs [None {from=MaterialExpressionAdd_13}]}
  MaterialExpressionSaturate_11 {class=MaterialExpressionSaturate, inputs [None {from=MaterialExpressionDotProduct_7}]}
  MaterialExpressionSaturate_12 {class=MaterialExpressionSaturate, inputs [None {from=MaterialExpressionSubtract_24}]}
  MaterialExpressionSaturate_13 {class=MaterialExpressionSaturate, inputs [None {from=MaterialExpressionSubtract_25}]}
  MaterialExpressionSaturate_14 {class=MaterialExpressionSaturate, inputs [None {from=MaterialExpressionLinearInterpolate_36}]}
  MaterialExpressionSaturate_15 {class=MaterialExpressionSaturate, inputs [None {from=MaterialExpressionOneMinus_5}]}
  MaterialExpressionSaturate_17 {class=MaterialExpressionSaturate, inputs [None {from=MaterialExpressionOneMinus_7}]}
  MaterialExpressionSaturate_2 {class=MaterialExpressionSaturate, inputs [None {from=MaterialExpressionDivide_3}]}
  MaterialExpressionSaturate_3 {class=MaterialExpressionSaturate, inputs [None {from=MaterialExpressionDivide_11}]}
  MaterialExpressionSaturate_4 {class=MaterialExpressionSaturate, inputs [None {from=MaterialExpressionDivide_7}]}
  MaterialExpressionSaturate_5 {class=MaterialExpressionSaturate, inputs [None {from=MaterialExpressionVertexInterpolator_6, output=PS}]}
  MaterialExpressionSaturate_6 {class=MaterialExpressionSaturate, inputs [None {from=MaterialExpressionVertexInterpolator_1, output=PS}]}
  MaterialExpressionSaturate_7 {class=MaterialExpressionSaturate, inputs [None {from=MaterialExpressionVertexInterpolator_0, output=PS}]}
  MaterialExpressionSaturate_8 {class=MaterialExpressionSaturate, inputs [None {from=MaterialExpressionDivide_8}], properties {desc=Don't Push Verts into Camera}}
  MaterialExpressionSaturate_9 {class=MaterialExpressionSaturate, inputs [None {from=MaterialExpressionDivide_10}]}
  MaterialExpressionScalarParameter_0 {class=MaterialExpressionScalarParameter, properties {defaultValue=1, parameterName=BaseColorBrightness}}
  MaterialExpressionScalarParameter_1 {class=MaterialExpressionScalarParameter, properties {defaultValue=16, parameterName=FramesXY}}
  MaterialExpressionScalarParameter_10 {class=MaterialExpressionScalarParameter, properties {defaultValue=16, parameterName=FramesXY}}
  MaterialExpressionScalarParameter_11 {class=MaterialExpressionScalarParameter, properties {defaultValue=16, parameterName=FramesXY}}
  MaterialExpressionScalarParameter_12 {class=MaterialExpressionScalarParameter, properties {defaultValue=1, parameterName=MipForBleed}}
  MaterialExpressionScalarParameter_13 {class=MaterialExpressionScalarParameter, properties {defaultValue=16, parameterName=FramesXY}}
  MaterialExpressionScalarParameter_14 {class=MaterialExpressionScalarParameter, properties {defaultValue=16, parameterName=FramesXY}}
  MaterialExpressionScalarParameter_15 {class=MaterialExpressionScalarParameter, properties {defaultValue=16, parameterName=FramesXY}}
  MaterialExpressionScalarParameter_16 {class=MaterialExpressionScalarParameter, properties {defaultValue=350, parameterName=Default Mesh Size}}
  MaterialExpressionScalarParameter_17 {class=MaterialExpressionScalarParameter, properties {defaultValue=-1, parameterName=Center Align}}
  MaterialExpressionScalarParameter_18 {class=MaterialExpressionScalarParameter, properties {defaultValue=1, parameterName=WPOScale}}
  MaterialExpressionScalarParameter_19 {
   class=MaterialExpressionScalarParameter
   properties {group=Color VertexColor Tints, parameterName=Green Vertex Channel Mask Value LOW, sortPriority=5}
  }
  MaterialExpressionScalarParameter_2 {class=MaterialExpressionScalarParameter, properties {defaultValue=2, parameterName=fresnel}}
  MaterialExpressionScalarParameter_20 {class=MaterialExpressionScalarParameter, properties {defaultValue=0.03, parameterName=grid line}}
  MaterialExpressionScalarParameter_21 {class=MaterialExpressionScalarParameter, properties {defaultValue=1, parameterName=Debug Grid}}
  MaterialExpressionScalarParameter_22 {class=MaterialExpressionScalarParameter, properties {defaultValue=0.1, parameterName=Opacity}}
  MaterialExpressionScalarParameter_23 {class=MaterialExpressionScalarParameter, properties {defaultValue=1, parameterName=PDO}}
  MaterialExpressionScalarParameter_24 {class=MaterialExpressionScalarParameter, properties {defaultValue=16, parameterName=FramesXY}}
  MaterialExpressionScalarParameter_25 {class=MaterialExpressionScalarParameter, properties {defaultValue=1, parameterName=UVscale}}
  MaterialExpressionScalarParameter_26 {class=MaterialExpressionScalarParameter, properties {defaultValue=16, parameterName=FramesXY}}
  MaterialExpressionScalarParameter_27 {class=MaterialExpressionScalarParameter, properties {defaultValue=1, parameterName=G}}
  MaterialExpressionScalarParameter_28 {class=MaterialExpressionScalarParameter, properties {defaultValue=10000, parameterName=Camera Z Fadeout Loc}}
  MaterialExpressionScalarParameter_29 {class=MaterialExpressionScalarParameter, properties {defaultValue=0.7, parameterName=ColorMaskMin}}
  MaterialExpressionScalarParameter_3 {class=MaterialExpressionScalarParameter, properties {defaultValue=1, parameterName=MipForBleed}}
  MaterialExpressionScalarParameter_30 {class=MaterialExpressionScalarParameter, properties {defaultValue=1, parameterName=MipForBleed}}
  MaterialExpressionScalarParameter_31 {class=MaterialExpressionScalarParameter, properties {defaultValue=16, parameterName=FramesXY}}
  MaterialExpressionScalarParameter_32 {class=MaterialExpressionScalarParameter, properties {defaultValue=1, parameterName=TanScaleFix}}
  MaterialExpressionScalarParameter_33 {class=MaterialExpressionScalarParameter, properties {defaultValue=16, parameterName=FramesXY}}
  MaterialExpressionScalarParameter_34 {class=MaterialExpressionScalarParameter, properties {defaultValue=16, parameterName=FramesXY}}
  MaterialExpressionScalarParameter_35 {
   class=MaterialExpressionScalarParameter
   properties {defaultValue=0.1, desc=">> Start of inner edge for edge glow", parameterName=Scatter Mask Min}
  }
  MaterialExpressionScalarParameter_36 {class=MaterialExpressionScalarParameter, properties {defaultValue=16, parameterName=FramesXY}}
  MaterialExpressionScalarParameter_37 {class=MaterialExpressionScalarParameter, properties {defaultValue=16, parameterName=FramesXY}}
  MaterialExpressionScalarParameter_38 {class=MaterialExpressionScalarParameter, properties {defaultValue=16, parameterName=FramesXY}}
  MaterialExpressionScalarParameter_39 {class=MaterialExpressionScalarParameter, properties {defaultValue=2, parameterName=DebugBoost}}
  MaterialExpressionScalarParameter_4 {class=MaterialExpressionScalarParameter, properties {defaultValue=1, parameterName=InterpolatePosition}}
  MaterialExpressionScalarParameter_40 {class=MaterialExpressionScalarParameter, properties {defaultValue=16, parameterName=FramesXY}}
  MaterialExpressionScalarParameter_41 {class=MaterialExpressionScalarParameter, properties {defaultValue=1, parameterName=Enable Weights}}
  MaterialExpressionScalarParameter_42 {
   class=MaterialExpressionScalarParameter
   properties {defaultValue=2, desc=">> Width of gradient for edge glow", parameterName=Scatter Mask Len}
  }
  MaterialExpressionScalarParameter_43 {class=MaterialExpressionScalarParameter, properties {defaultValue=0.3, parameterName=ScatterMin}}
  MaterialExpressionScalarParameter_44 {class=MaterialExpressionScalarParameter, properties {defaultValue=2, parameterName=Specular}}
  MaterialExpressionScalarParameter_45 {class=MaterialExpressionScalarParameter, properties {defaultValue=1, parameterName=EnableParallax}}
  MaterialExpressionScalarParameter_46 {class=MaterialExpressionScalarParameter, properties {defaultValue=0.2, parameterName=ColorMaskLen}}
  MaterialExpressionScalarParameter_47 {class=MaterialExpressionScalarParameter, properties {defaultValue=35000, parameterName=Cam Z Fade Length}}
  MaterialExpressionScalarParameter_48 {class=MaterialExpressionScalarParameter, properties {defaultValue=0.6, parameterName=Mip UV Scale}}
  MaterialExpressionScalarParameter_49 {class=MaterialExpressionScalarParameter, properties {defaultValue=-0.2, group=Snow, parameterName=SnowNormal_Bias}}
  MaterialExpressionScalarParameter_5 {class=MaterialExpressionScalarParameter, properties {defaultValue=16, parameterName=FramesXY}}
  MaterialExpressionScalarParameter_50 {class=MaterialExpressionScalarParameter, properties {defaultValue=0.4951, group=Apollo, parameterName=Shift_Gradient}}
  MaterialExpressionScalarParameter_51 {class=MaterialExpressionScalarParameter, properties {defaultValue=-0.3204, group=Apollo, parameterName=Shift_Gradient_2}}
  MaterialExpressionScalarParameter_52 {class=MaterialExpressionScalarParameter, properties {group=Apollo, parameterName=Fake SSS Scale}}
  MaterialExpressionScalarParameter_53 {class=MaterialExpressionScalarParameter, properties {defaultValue=0.5, parameterName=Roughness}}
  MaterialExpressionScalarParameter_54 {
   class=MaterialExpressionScalarParameter
   properties {defaultValue=13000, group=World Hue Variation, parameterName=Color Variation Distance WS, sortPriority=10}
  }
  MaterialExpressionScalarParameter_55 {class=MaterialExpressionScalarParameter, properties {defaultValue=0.03, parameterName=Gcut}}
  MaterialExpressionScalarParameter_56 {
   class=MaterialExpressionScalarParameter
   properties {defaultValue=11000, group=World Hue Variation, parameterName=Color Variation Distance WS2, sortPriority=10}
  }
  MaterialExpressionScalarParameter_57 {
   class=MaterialExpressionScalarParameter
   properties {group=Color VertexColor Tints, parameterName=Green Vertex Channel Mask Value HIGH, sortPriority=5}
  }
  MaterialExpressionScalarParameter_58 {class=MaterialExpressionScalarParameter, properties {defaultValue=1, group=Color, parameterName=Fake_SSS_Color_Mult}}
  MaterialExpressionScalarParameter_59 {
   class=MaterialExpressionScalarParameter
   properties {defaultValue=0.0012, group=Apollo, parameterName=Color Variation Range WS, sortPriority=10}
  }
  MaterialExpressionScalarParameter_6 {class=MaterialExpressionScalarParameter, properties {defaultValue=0.03, parameterName=FrameDebugRadius}}
  MaterialExpressionScalarParameter_60 {
   class=MaterialExpressionScalarParameter
   properties {defaultValue=0.5, group=Apollo, parameterName=Fixed Color Manual Offset, sortPriority=-9}
  }
  MaterialExpressionScalarParameter_61 {class=MaterialExpressionScalarParameter, properties {defaultValue=1, parameterName=EmissiveScale}}
  MaterialExpressionScalarParameter_62 {class=MaterialExpressionScalarParameter, properties {defaultValue=0.2, parameterName=Desaturation}}
  MaterialExpressionScalarParameter_63 {class=MaterialExpressionScalarParameter, properties {defaultValue=0.9682, parameterName=TrunkMaskMin}}
  MaterialExpressionScalarParameter_64 {class=MaterialExpressionScalarParameter, properties {defaultValue=0, parameterName=TrunkMaskLen}}
  MaterialExpressionScalarParameter_65 {class=MaterialExpressionScalarParameter, properties {defaultValue=1.5, group=Color, parameterName=PrimDataColorMultiply_Lit}}
  MaterialExpressionScalarParameter_66 {class=MaterialExpressionScalarParameter, properties {defaultValue=0.5, group=Color, parameterName=PrimDataColorMultiply_Shadows}}
  MaterialExpressionScalarParameter_67 {class=MaterialExpressionScalarParameter, properties {defaultValue=-0.03, group=Color, parameterName=Hue Variation}}
  MaterialExpressionScalarParameter_68 {class=MaterialExpressionScalarParameter, properties {defaultValue=0.5, group=Color, parameterName=ValueVariation}}
  MaterialExpressionScalarParameter_69 {
   class=MaterialExpressionScalarParameter
   properties {bUseCustomPrimitiveData=true, defaultValue=-0.03, group=Color, parameterName=Hue Variation_primdata, primitiveDataIndex=4}
  }
  MaterialExpressionScalarParameter_7 {class=MaterialExpressionScalarParameter, properties {defaultValue=0.9503}}
  MaterialExpressionScalarParameter_70 {
   class=MaterialExpressionScalarParameter
   properties {bUseCustomPrimitiveData=true, defaultValue=0.5, group=Color, parameterName=ValueVariation_primdata, primitiveDataIndex=5}
  }
  MaterialExpressionScalarParameter_71 {class=MaterialExpressionScalarParameter, properties {defaultValue=-0.03, group=Color Variation WorldSpace, parameterName=Hue Variation}}
  MaterialExpressionScalarParameter_72 {
   class=MaterialExpressionScalarParameter
   properties {defaultValue=0.0125, group=World Hue Variation, parameterName=World Hue Variation Strength}
  }
  MaterialExpressionScalarParameter_73 {class=MaterialExpressionScalarParameter, properties {group=World Hue Variation, parameterName=World Hue Variation Bias}}
  MaterialExpressionScalarParameter_74 {
   class=MaterialExpressionScalarParameter
   properties {
    defaultValue=1
    desc="Multiplies Camera Push value by this value for Low && Medium quality to compensate for lack of PDO at these quality levels"
    parameterName=Camera Push LOW Fraction
    sliderMax=1
   }
  }
  MaterialExpressionScalarParameter_75 {class=MaterialExpressionScalarParameter, properties {group=Apollo, parameterName=Fake SSS Scale add Mobile}}
  MaterialExpressionScalarParameter_76 {class=MaterialExpressionScalarParameter, properties {defaultValue=0.5, group=Color Variation WorldSpace, parameterName=ValueVariation}}
  MaterialExpressionScalarParameter_77 {
   class=MaterialExpressionScalarParameter
   properties {bUseCustomPrimitiveData=true, defaultValue=-0.03, group=Color, parameterName=Hue Variation_primdata, primitiveDataIndex=4}
  }
  MaterialExpressionScalarParameter_78 {
   class=MaterialExpressionScalarParameter
   properties {bUseCustomPrimitiveData=true, defaultValue=0.5, group=Color, parameterName=ValueVariation_primdata, primitiveDataIndex=5}
  }
  MaterialExpressionScalarParameter_79 {class=MaterialExpressionScalarParameter, properties {defaultValue=0.2, group=Shading, parameterName=SubSurface SSSColor}}
  MaterialExpressionScalarParameter_8 {class=MaterialExpressionScalarParameter, properties {defaultValue=1, parameterName=Camera Push}}
  MaterialExpressionScalarParameter_80 {class=MaterialExpressionScalarParameter, properties {defaultValue=1, group=Shading, parameterName=Brightness}}
  MaterialExpressionScalarParameter_81 {
   class=MaterialExpressionScalarParameter
   properties {defaultValue=13000, group=World Hue Variation, parameterName=Color Variation Distance WS, sortPriority=10}
  }
  MaterialExpressionScalarParameter_82 {
   class=MaterialExpressionScalarParameter
   properties {defaultValue=11000, group=World Hue Variation, parameterName=Color Variation Distance WS2, sortPriority=10}
  }
  MaterialExpressionScalarParameter_85 {class=MaterialExpressionScalarParameter, properties {defaultValue=0.25, parameterName=Z_Offset}}
  MaterialExpressionScalarParameter_86 {class=MaterialExpressionScalarParameter, properties {defaultValue=26436, parameterName=Z_Offset_FadeLength}}
  MaterialExpressionScalarParameter_87 {class=MaterialExpressionScalarParameter, properties {defaultValue=36725, parameterName=Z_Offset_FadeOffset}}
  MaterialExpressionScalarParameter_9 {class=MaterialExpressionScalarParameter, properties {defaultValue=0.01, parameterName=FrameDebugRadius2}}
  MaterialExpressionSetMaterialAttributes_0 {
   class=MaterialExpressionSetMaterialAttributes
   inputs [
    MaterialAttributes {from="", output=""}
    Base Color {from=MaterialExpressionNamedRerouteUsage_1, output=""}
    Metallic {from="", output=""}
    Specular {from=MaterialExpressionScalarParameter_44, output=""}
    Roughness {from=MaterialExpressionScalarParameter_53, output=""}
    Emissive Color {from=MaterialExpressionNamedRerouteUsage_14, output=""}
    Opacity {from=MaterialExpressionScalarParameter_22, output=""}
    Opacity Mask {from=MaterialExpressionNamedRerouteUsage_10, output=""}
    Normal {from=MaterialExpressionNamedRerouteUsage_0, output=""}
    World Position Offset {from=MaterialExpressionNamedRerouteUsage_4, output=""}
    Subsurface Color {from=MaterialExpressionNamedRerouteUsage_13, output=""}
    Pixel Depth Offset {from=MaterialExpressionNamedRerouteUsage_8, output=""}
   ]
   properties {attributeSetTypes [
     #1
     #6
     #7
     #8
     #9
     #10
     #2
     #5
     #3
     #11
     #4
    ]}
  }
  MaterialExpressionSetMaterialAttributes_1 {
   class=MaterialExpressionSetMaterialAttributes
   inputs [
    MaterialAttributes {from=MaterialExpressionGetMaterialAttributes_0, output=MaterialAttributes}
    Base Color {from=MaterialExpressionLinearInterpolate_22, output=""}
    Opacity Mask {from=MaterialExpressionLinearInterpolate_21, output=""}
    World Position Offset {from=MaterialExpressionLinearInterpolate_19, output=""}
    Pixel Depth Offset {from=MaterialExpressionLinearInterpolate_20, output=""}
    Normal {from=MaterialExpressionLinearInterpolate_43, output=""}
   ]
   properties {attributeSetTypes [
     #1
     #2
     #3
     #4
     #5
    ]}
  }
  MaterialExpressionSetMaterialAttributes_2 {
   class=MaterialExpressionSetMaterialAttributes
   inputs [
    MaterialAttributes {from=MaterialExpressionGetMaterialAttributes_1, output=MaterialAttributes}
    Base Color {from=MaterialExpressionStaticSwitchParameter_4, output=""}
   ]
   properties {attributeSetTypes [#1]}
  }
  MaterialExpressionShadingPathSwitch_0 {
   class=MaterialExpressionShadingPathSwitch
   inputs [
    Default {from=MaterialExpressionMultiply_65, output=""}
    Deferred {from=MaterialExpressionMultiply_65, output=""}
    Forward {from=MaterialExpressionStaticSwitchParameter_43, output=""}
    Mobile {from=MaterialExpressionStaticSwitchParameter_43, output=""}
   ]
   properties {}
  }
  MaterialExpressionShadingPathSwitch_1 {
   class=MaterialExpressionShadingPathSwitch
   inputs [
    Default {from=MaterialExpressionMultiply_59, output=""}
    Deferred {from="", output=""}
    Forward {from="", output=""}
    Mobile {from=MaterialExpressionConstant_16, output=""}
   ]
   properties {desc=Migrated from FeatureLevelSwitch}
  }
  MaterialExpressionShadingPathSwitch_2 {
   class=MaterialExpressionShadingPathSwitch
   inputs [
    Default {from=MaterialExpressionAdd_17, output=""}
    Deferred {from=MaterialExpressionAdd_17, output=""}
    Forward {from="", output=""}
    Mobile {from=MaterialExpressionDistanceFieldsRenderingSwitch_1, output=""}
   ]
   properties {desc=Migrated from FeatureLevelSwitch}
  }
  MaterialExpressionShadingPathSwitch_3 {
   class=MaterialExpressionShadingPathSwitch
   inputs [
    Default {from=MaterialExpressionReroute_21, output=""}
    Deferred {from="", output=""}
    Forward {from="", output=""}
    Mobile {from=MaterialExpressionConstant_20, output=""}
   ]
   properties {desc=Migrated from FeatureLevelSwitch}
  }
  MaterialExpressionShadingPathSwitch_4 {
   class=MaterialExpressionShadingPathSwitch
   inputs [
    Default {from=MaterialExpressionLinearInterpolate_24, output=""}
    Deferred {from="", output=""}
    Forward {from="", output=""}
    Mobile {from=MaterialExpressionTextureSample_7, output=RGB}
   ]
   properties {desc=Migrated from FeatureLevelSwitch}
  }
  MaterialExpressionShadingPathSwitch_5 {
   class=MaterialExpressionShadingPathSwitch
   inputs [
    Default {from=MaterialExpressionScalarParameter_52, output=""}
    Deferred {from=MaterialExpressionScalarParameter_52, output=""}
    Forward {from="", output=""}
    Mobile {from=MaterialExpressionAdd_25, output=""}
   ]
   properties {desc=Migrated from FeatureLevelSwitch}
  }
  MaterialExpressionSine_0 {class=MaterialExpressionSine, inputs [None {from=MaterialExpressionDivide_13}]}
  MaterialExpressionSine_1 {class=MaterialExpressionSine, inputs [None {from=MaterialExpressionDivide_18}]}
  MaterialExpressionSine_2 {class=MaterialExpressionSine, inputs [None {from=MaterialExpressionDivide_9}]}
  MaterialExpressionSine_3 {class=MaterialExpressionSine, inputs [None {from=MaterialExpressionDivide_20}]}
  MaterialExpressionSkyAtmosphereLightDirection_0 {class=MaterialExpressionSkyAtmosphereLightDirection, inputs [], properties {}}
  MaterialExpressionSphereMask_1 {
   class=MaterialExpressionSphereMask
   inputs [
    A {from=MaterialExpressionFloor_3, output=""}
    B {from=MaterialExpressionFloor_2, output=""}
    Radius {from=MaterialExpressionScalarParameter_6, output=""}
    Hardness {from="", output=""}
   ]
   properties {}
  }
  MaterialExpressionSphereMask_2 {
   class=MaterialExpressionSphereMask
   inputs [
    A {from=MaterialExpressionTextureCoordinate_4, output=""}
    B {from=MaterialExpressionClamp_0, output=""}
    Radius {from=MaterialExpressionScalarParameter_9, output=""}
    Hardness {from="", output=""}
   ]
   properties {}
  }
  MaterialExpressionStaticBool_0 {class=MaterialExpressionStaticBool}
  MaterialExpressionStaticBool_1 {class=MaterialExpressionStaticBool}
  MaterialExpressionStaticBool_2 {class=MaterialExpressionStaticBool, properties {value=true}}
  MaterialExpressionStaticBool_3 {class=MaterialExpressionStaticBool, properties {value=true}}
  MaterialExpressionStaticBool_4 {class=MaterialExpressionStaticBool, properties {value=true}}
  MaterialExpressionStaticBool_5 {class=MaterialExpressionStaticBool, properties {value=true}}
  MaterialExpressionStaticSwitchParameter_0 {
   class=MaterialExpressionStaticSwitchParameter
   inputs [
    True {from=MaterialExpressionDesaturation_0, output=""}
    False {from=MaterialExpressionNamedRerouteUsage_22, output=""}
   ]
   properties {parameterName=UseDesaturation}
  }
  MaterialExpressionStaticSwitchParameter_1 {
   class=MaterialExpressionStaticSwitchParameter
   inputs [
    True {from=MaterialExpressionMultiply_58, output=""}
    False {from=MaterialExpressionMultiply_57, output=""}
   ]
   properties {parameterName=ExtractTrunkMask}
  }
  MaterialExpressionStaticSwitchParameter_10 {
   class=MaterialExpressionStaticSwitchParameter
   inputs [
    True {from=MaterialExpressionMultiply_66, output=""}
    False {from=MaterialExpressionNamedRerouteUsage_16, output=""}
   ]
   properties {parameterName=UseSimpleEmissiveOverlay}
  }
  MaterialExpressionStaticSwitchParameter_11 {
   class=MaterialExpressionStaticSwitchParameter
   inputs [
    True {from=MaterialExpressionDivide_14, output=""}
    False {from=MaterialExpressionScalarParameter_60, output=""}
   ]
   properties {group=Apollo, parameterName=UseRandomColor, sortPriority=-10}
  }
  MaterialExpressionStaticSwitchParameter_12 {
   class=MaterialExpressionStaticSwitchParameter
   inputs [
    True {from=MaterialExpressionMultiply_90, output=""}
    False {from=MaterialExpressionLinearInterpolate_25, output=""}
   ]
   properties {defaultValue=true, group=Apollo, parameterName=UseDiffuse, sortPriority=1}
  }
  MaterialExpressionStaticSwitchParameter_13 {
   class=MaterialExpressionStaticSwitchParameter
   inputs [
    True {from=MaterialExpressionNamedRerouteUsage_12, output=""}
    False {from=MaterialExpressionStaticSwitchParameter_15, output=""}
   ]
   properties {group=Coloring Systems, parameterName=UseAsteriaColoring}
  }
  MaterialExpressionStaticSwitchParameter_14 {
   class=MaterialExpressionStaticSwitchParameter
   inputs [
    True {from=MaterialExpressionSetMaterialAttributes_1, output=""}
    False {from=MaterialExpressionSetMaterialAttributes_2, output=""}
   ]
   properties {parameterName=Use_StaticTrunk}
  }
  MaterialExpressionStaticSwitchParameter_15 {
   class=MaterialExpressionStaticSwitchParameter
   inputs [
    True {from=MaterialExpressionNamedRerouteUsage_29, output=""}
    False {from=MaterialExpressionNamedRerouteUsage_28, output=""}
   ]
   properties {group=Coloring Systems, parameterName=UseApolloColoring, sortPriority=-100}
  }
  MaterialExpressionStaticSwitchParameter_16 {
   class=MaterialExpressionStaticSwitchParameter
   inputs [
    True {from=MaterialExpressionMaterialFunctionCall_48, output=Material}
    False {from=MaterialExpressionReroute_40, output=""}
   ]
   properties {group=Snow, parameterName=Use_BRSnow-Progressive}
  }
  MaterialExpressionStaticSwitchParameter_17 {
   class=MaterialExpressionStaticSwitchParameter
   inputs [
    True {from=MaterialExpressionPerInstanceCustomData_3, output=""}
    False {from=MaterialExpressionScalarParameter_78, output=""}
   ]
   properties {defaultValue=true, parameterName=UsePerInstanceCustomData}
  }
  MaterialExpressionStaticSwitchParameter_18 {
   class=MaterialExpressionStaticSwitchParameter
   inputs [
    True {from=MaterialExpressionScalarParameter_81, output=""}
    False {from=MaterialExpressionConstant_37, output=""}
   ]
   properties {group=World Hue Variation, parameterName=Enable Global Hue Variation Tweaking}
  }
  MaterialExpressionStaticSwitchParameter_19 {
   class=MaterialExpressionStaticSwitchParameter
   inputs [
    True {from=MaterialExpressionScalarParameter_82, output=""}
    False {from=MaterialExpressionConstant_38, output=""}
   ]
   properties {group=World Hue Variation, parameterName=Enable Global Hue Variation Tweaking}
  }
  MaterialExpressionStaticSwitchParameter_2 {
   class=MaterialExpressionStaticSwitchParameter
   inputs [
    True {from=MaterialExpressionLinearInterpolate_5, output=""}
    False {from=MaterialExpressionMultiply_13, output=""}
   ]
   properties {parameterName=UseTwoColors}
  }
  MaterialExpressionStaticSwitchParameter_20 {
   class=MaterialExpressionStaticSwitchParameter
   inputs [
    True {from=MaterialExpressionScalarParameter_72, output=""}
    False {from=MaterialExpressionConstant_30, output=""}
   ]
   properties {group=World Hue Variation, parameterName=Enable Global Hue Variation Tweaking}
  }
  MaterialExpressionStaticSwitchParameter_21 {
   class=MaterialExpressionStaticSwitchParameter
   inputs [
    True {from=MaterialExpressionVectorParameter_21, output=RGB}
    False {from=MaterialExpressionVectorParameter_17, output=RGB}
   ]
   properties {group=Coloring Systems, parameterName=UseAsteriaColoring}
  }
  MaterialExpressionStaticSwitchParameter_22 {
   class=MaterialExpressionStaticSwitchParameter
   inputs [
    True {from=MaterialExpressionMaterialFunctionCall_47, output=Result}
    False {from=MaterialExpressionStaticSwitchParameter_16, output=""}
   ]
   properties {group=Coloring Systems, parameterName=UseApolloColoring, sortPriority=-100}
  }
  MaterialExpressionStaticSwitchParameter_23 {
   class=MaterialExpressionStaticSwitchParameter
   inputs [
    True {from=MaterialExpressionNamedRerouteUsage_43, output=""}
    False {from=MaterialExpressionNamedRerouteUsage_41, output=""}
   ]
   properties {group=Coloring Systems, parameterName=UseAsteriaColoring}
  }
  MaterialExpressionStaticSwitchParameter_24 {
   class=MaterialExpressionStaticSwitchParameter
   inputs [
    True {from=MaterialExpressionAdd_31, output=""}
    False {from=MaterialExpressionNamedRerouteUsage_34, output=""}
   ]
   properties {group=World Hue Variation, parameterName=Enable Global Hue Variation Tweaking}
  }
  MaterialExpressionStaticSwitchParameter_3 {
   class=MaterialExpressionStaticSwitchParameter
   inputs [
    True {from=MaterialExpressionAdd_11, output=""}
    False {from=MaterialExpressionReroute_12, output=""}
   ]
   properties {parameterName=CubeDestroyPath}
  }
  MaterialExpressionStaticSwitchParameter_4 {
   class=MaterialExpressionStaticSwitchParameter
   inputs [
    True {from=MaterialExpressionLinearInterpolate_7, output=""}
    False {from=MaterialExpressionReroute_50, output=""}
   ]
   properties {group=Snow, parameterName=Use_BRSnow}
  }
  MaterialExpressionStaticSwitchParameter_43 {
   class=MaterialExpressionStaticSwitchParameter
   inputs [
    True {from=MaterialExpressionLinearInterpolate_27, output=""}
    False {from=MaterialExpressionVectorParameter_15, output=RGB}
   ]
   properties {parameterName=UsePrimitiveDataSeasonColoring}
  }
  MaterialExpressionStaticSwitchParameter_44 {
   class=MaterialExpressionStaticSwitchParameter
   inputs [
    True {from=MaterialExpressionLinearInterpolate_26, output=""}
    False {from=MaterialExpressionVectorParameter_12, output=RGB}
   ]
   properties {parameterName=UsePrimitiveDataSeasonColoring}
  }
  MaterialExpressionStaticSwitchParameter_46 {
   class=MaterialExpressionStaticSwitchParameter
   inputs [
    True {from=MaterialExpressionLinearInterpolate_28, output=""}
    False {from=MaterialExpressionVectorParameter_16, output=RGB}
   ]
   properties {parameterName=UsePrimitiveDataSeasonColoring}
  }
  MaterialExpressionStaticSwitchParameter_47 {
   class=MaterialExpressionStaticSwitchParameter
   inputs [
    True {from=MaterialExpressionLinearInterpolate_30, output=""}
    False {from=MaterialExpressionScalarParameter_67, output=""}
   ]
   properties {parameterName=UsePrimitiveDataSeasonColoring}
  }
  MaterialExpressionStaticSwitchParameter_48 {
   class=MaterialExpressionStaticSwitchParameter
   inputs [
    True {from=MaterialExpressionLinearInterpolate_29, output=""}
    False {from=MaterialExpressionScalarParameter_68, output=""}
   ]
   properties {parameterName=UsePrimitiveDataSeasonColoring}
  }
  MaterialExpressionStaticSwitchParameter_5 {
   class=MaterialExpressionStaticSwitchParameter
   inputs [
    True {from=MaterialExpressionLinearInterpolate_6, output=""}
    False {from=MaterialExpressionNamedRerouteUsage_2, output=""}
   ]
   properties {parameterName=Use_StaticTrunk}
  }
  MaterialExpressionStaticSwitchParameter_6 {
   class=MaterialExpressionStaticSwitchParameter
   inputs [
    True {from=MaterialExpressionLinearInterpolate_8, output=""}
    False {from=MaterialExpressionNamedRerouteUsage_5, output=""}
   ]
   properties {parameterName=Use_StaticTrunk}
  }
  MaterialExpressionStaticSwitchParameter_7 {
   class=MaterialExpressionStaticSwitchParameter
   inputs [
    True {from=MaterialExpressionLinearInterpolate_14, output=""}
    False {from=MaterialExpressionNamedRerouteUsage_11, output=""}
   ]
   properties {parameterName=Use_StaticTrunk}
  }
  MaterialExpressionStaticSwitchParameter_8 {
   class=MaterialExpressionStaticSwitchParameter
   inputs [
    True {from=MaterialExpressionLinearInterpolate_15, output=""}
    False {from=MaterialExpressionNamedRerouteUsage_9, output=""}
   ]
   properties {parameterName=Use_StaticTrunk}
  }
  MaterialExpressionStaticSwitchParameter_9 {
   class=MaterialExpressionStaticSwitchParameter
   inputs [
    True {from=MaterialExpressionNamedRerouteUsage_30, output=""}
    False {from=MaterialExpressionConstant_18, output=""}
   ]
   properties {parameterName=ExtractTrunkMask}
  }
  MaterialExpressionStaticSwitch_0 {
   class=MaterialExpressionStaticSwitch
   inputs [
    True {from=MaterialExpressionConstant3Vector_5, output=""}
    False {from=MaterialExpressionLinearInterpolate_37, output=""}
    Value {from=MaterialExpressionNamedRerouteUsage_7, output=""}
   ]
   properties {desc=For ImpostorBakeing}
  }
  MaterialExpressionStaticSwitch_1 {
   class=MaterialExpressionStaticSwitch
   inputs [
    True {from=MaterialExpressionComponentMask_6, output=""}
    False {from=MaterialExpressionConstant_3, output=""}
    Value {from=MaterialExpressionStaticBool_0, output=""}
   ]
   properties {}
  }
  MaterialExpressionStaticSwitch_2 {
   class=MaterialExpressionStaticSwitch
   inputs [
    True {from=MaterialExpressionAdd_5, output=""}
    False {from=MaterialExpressionReroute_6, output=""}
    Value {from=MaterialExpressionStaticBool_0, output=""}
   ]
   properties {}
  }
  MaterialExpressionStep_0 {
   class=MaterialExpressionStep
   inputs [
    Y {from="", output=""}
    X {from=MaterialExpressionCustom_4, output=""}
   ]
   properties {}
  }
  MaterialExpressionSubtract_0 {
   class=MaterialExpressionSubtract
   inputs [
    A {from=""}
    B {from=MaterialExpressionTextureSampleParameter2D_0, output=A}
   ]
   properties {constA=0.5, constB=0.5}
  }
  MaterialExpressionSubtract_1 {
   class=MaterialExpressionSubtract
   inputs [
    A {from=""}
    B {from=MaterialExpressionStep_0}
   ]
  }
  MaterialExpressionSubtract_10 {
   class=MaterialExpressionSubtract
   inputs [
    A {from=MaterialExpressionCustom_1}
    B {from=MaterialExpressionMaterialFunctionCall_15, output=Position}
   ]
  }
  MaterialExpressionSubtract_11 {
   class=MaterialExpressionSubtract
   inputs [
    A {from=MaterialExpressionCustom_6}
    B {from=MaterialExpressionAdd_12}
   ]
  }
  MaterialExpressionSubtract_12 {
   class=MaterialExpressionSubtract
   inputs [
    A {from=MaterialExpressionCustom_6}
    B {from=MaterialExpressionAdd_12}
   ]
  }
  MaterialExpressionSubtract_13 {
   class=MaterialExpressionSubtract
   inputs [
    A {from=MaterialExpressionDistance_0}
    B {from=MaterialExpressionMaterialFunctionCall_8, output=V3 Length}
   ]
  }
  MaterialExpressionSubtract_14 {
   class=MaterialExpressionSubtract
   inputs [
    A {from=MaterialExpressionMaterialFunctionCall_26, output=Position}
    B {from=MaterialExpressionCustom_6}
   ]
  }
  MaterialExpressionSubtract_15 {
   class=MaterialExpressionSubtract
   inputs [
    A {from=MaterialExpressionAbs_1}
    B {from=MaterialExpressionScalarParameter_63}
   ]
  }
  MaterialExpressionSubtract_16 {
   class=MaterialExpressionSubtract
   inputs [
    A {from=MaterialExpressionComponentMask_7}
    B {from=MaterialExpressionScalarParameter_29}
   ]
  }
  MaterialExpressionSubtract_17 {
   class=MaterialExpressionSubtract
   inputs [
    A {from=MaterialExpressionAdd_6}
    B {from=MaterialExpressionCustom_8}
   ]
  }
  MaterialExpressionSubtract_18 {
   class=MaterialExpressionSubtract
   inputs [
    A {from=MaterialExpressionCustom_8}
    B {from=MaterialExpressionMaterialFunctionCall_10, output=Position}
   ]
  }
  MaterialExpressionSubtract_19 {
   class=MaterialExpressionSubtract
   inputs [
    A {from=MaterialExpressionDistance_2}
    B {from=MaterialExpressionMultiply_4}
   ]
  }
  MaterialExpressionSubtract_2 {
   class=MaterialExpressionSubtract
   inputs [
    A {from=MaterialExpressionScalarParameter_10}
    B {from=""}
   ]
  }
  MaterialExpressionSubtract_20 {
   class=MaterialExpressionSubtract
   inputs [
    A {from=MaterialExpressionComponentMask_16}
    B {from=MaterialExpressionScalarParameter_28}
   ]
  }
  MaterialExpressionSubtract_21 {
   class=MaterialExpressionSubtract
   inputs [
    A {from=MaterialExpressionWorldPosition_2, output=XYZ}
    B {from=MaterialExpressionVectorParameter_3, output=RGB}
   ]
  }
  MaterialExpressionSubtract_22 {
   class=MaterialExpressionSubtract
   inputs [
    A {from=MaterialExpressionVectorParameter_5, output=RGB}
    B {from=MaterialExpressionVectorParameter_3, output=RGB}
   ]
  }
  MaterialExpressionSubtract_24 {
   class=MaterialExpressionSubtract
   inputs [
    A {from=MaterialExpressionReroute_16}
    B {from=MaterialExpressionConstant_17}
   ]
  }
  MaterialExpressionSubtract_25 {
   class=MaterialExpressionSubtract
   inputs [
    A {from=MaterialExpressionReroute_16}
    B {from=MaterialExpressionConstant_19}
   ]
  }
  MaterialExpressionSubtract_3 {
   class=MaterialExpressionSubtract
   inputs [
    A {from=""}
    B {from=MaterialExpressionTextureSampleParameter2D_1, output=A}
   ]
   properties {constA=0.5, constB=0.5}
  }
  MaterialExpressionSubtract_33 {
   class=MaterialExpressionSubtract
   inputs [
    A {from=MaterialExpressionFrac_5}
    B {from=""}
   ]
   properties {constB=0.5}
  }
  MaterialExpressionSubtract_34 {
   class=MaterialExpressionSubtract
   inputs [
    A {from=MaterialExpressionAdd_32}
    B {from=""}
   ]
   properties {constB=0.5}
  }
  MaterialExpressionSubtract_35 {
   class=MaterialExpressionSubtract
   inputs [
    A {from=MaterialExpressionAdd_34}
    B {from=MaterialExpressionStaticSwitchParameter_9}
   ]
  }
  MaterialExpressionSubtract_4 {
   class=MaterialExpressionSubtract
   inputs [
    A {from=MaterialExpressionScalarParameter_24}
    B {from=""}
   ]
  }
  MaterialExpressionSubtract_5 {
   class=MaterialExpressionSubtract
   inputs [
    A {from=MaterialExpressionCustom_1}
    B {from=MaterialExpressionAdd_6}
   ]
  }
  MaterialExpressionSubtract_6 {
   class=MaterialExpressionSubtract
   inputs [
    A {from=MaterialExpressionReroute_7}
    B {from=MaterialExpressionScalarParameter_35}
   ]
  }
  MaterialExpressionSubtract_7 {
   class=MaterialExpressionSubtract
   inputs [
    A {from=MaterialExpressionComponentMask_20}
    B {from=MaterialExpressionScalarParameter_55}
   ]
   properties {constB=0.3}
  }
  MaterialExpressionSubtract_8 {
   class=MaterialExpressionSubtract
   inputs [
    A {from=""}
    B {from=MaterialExpressionTextureSampleParameter2D_9, output=A}
   ]
   properties {constA=0.5, constB=0.5}
  }
  MaterialExpressionSubtract_9 {
   class=MaterialExpressionSubtract
   inputs [
    A {from=MaterialExpressionScalarParameter_38}
    B {from=""}
   ]
  }
  MaterialExpressionTextureCoordinate_0 {class=MaterialExpressionTextureCoordinate}
  MaterialExpressionTextureCoordinate_1 {class=MaterialExpressionTextureCoordinate, properties {uTiling=10, vTiling=10}}
  MaterialExpressionTextureCoordinate_2 {class=MaterialExpressionTextureCoordinate, properties {uTiling=10, vTiling=10}}
  MaterialExpressionTextureCoordinate_3 {class=MaterialExpressionTextureCoordinate}
  MaterialExpressionTextureCoordinate_4 {class=MaterialExpressionTextureCoordinate}
  MaterialExpressionTextureCoordinate_5 {class=MaterialExpressionTextureCoordinate, properties {uTiling=10, vTiling=10}}
  MaterialExpressionTextureCoordinate_6 {class=MaterialExpressionTextureCoordinate}
  MaterialExpressionTextureCoordinate_7 {class=MaterialExpressionTextureCoordinate}
  MaterialExpressionTextureObjectParameter_2 {
   class=MaterialExpressionTextureObjectParameter
   inputs [
    UVs {from="", output=""}
    Apply View MipBias {from="", output=""}
   ]
   properties {group=Apollo, parameterName=ColorPalette, sortPriority=99, texture=/Game/Environments/Apollo/Foliage/ColorLUT/Apollo_Foliage_LUT_Green}
  }
  MaterialExpressionTextureObjectParameter_3 {
   class=MaterialExpressionTextureObjectParameter
   inputs [
    UVs {from="", output=""}
    Apply View MipBias {from="", output=""}
   ]
   properties {parameterName=BaseColor, texture=&Black}
  }
  MaterialExpressionTextureObjectParameter_4 {
   class=MaterialExpressionTextureObjectParameter
   inputs [
    UVs {from="", output=""}
    Apply View MipBias {from="", output=""}
   ]
   properties {parameterName=Normal, texture=&Black}
  }
  MaterialExpressionTextureSampleParameter2D_0 {
   class=MaterialExpressionTextureSampleParameter2D
   inputs [
    UVs {from=MaterialExpressionAdd_20, output=""}
    Apply View MipBias {from="", output=""}
   ]
   properties {constMipValue=0, parameterName=Normal, samplerSource=SSM_Wrap_WorldGroupSettings, texture=&Black}
  }
  MaterialExpressionTextureSampleParameter2D_1 {
   class=MaterialExpressionTextureSampleParameter2D
   inputs [
    UVs {from=MaterialExpressionAdd_15, output=""}
    Apply View MipBias {from="", output=""}
   ]
   properties {constMipValue=0, parameterName=Normal, samplerSource=SSM_Wrap_WorldGroupSettings, texture=&Black}
  }
  MaterialExpressionTextureSampleParameter2D_2 {
   class=MaterialExpressionTextureSampleParameter2D
   inputs [
    UVs {from="", output=""}
    Apply View MipBias {from="", output=""}
   ]
   properties {parameterName=BaseColor, samplerSource=SSM_Wrap_WorldGroupSettings, texture=&Black}
  }
  MaterialExpressionTextureSampleParameter2D_4 {
   class=MaterialExpressionTextureSampleParameter2D
   inputs [
    UVs {from=MaterialExpressionMultiply_67, output=""}
    Apply View MipBias {from="", output=""}
   ]
   properties {parameterName=EmissiveOverlay, samplerSource=SSM_Wrap_WorldGroupSettings, texture=&Black}
  }
  MaterialExpressionTextureSampleParameter2D_6 {
   class=MaterialExpressionTextureSampleParameter2D
   inputs [
    UVs {from="", output=""}
    Apply View MipBias {from="", output=""}
   ]
   properties {parameterName=TrunkBaseColor, texture=&Black}
  }
  MaterialExpressionTextureSampleParameter2D_9 {
   class=MaterialExpressionTextureSampleParameter2D
   inputs [
    UVs {from=MaterialExpressionAdd_23, output=""}
    Apply View MipBias {from="", output=""}
   ]
   properties {constMipValue=0, parameterName=Normal, samplerSource=SSM_Wrap_WorldGroupSettings, texture=&Black}
  }
  MaterialExpressionTextureSample_6 {
   class=MaterialExpressionTextureSample
   inputs [
    UVs {from=MaterialExpressionAppendVector_9}
    Tex {from=MaterialExpressionTextureObjectParameter_2}
    Apply View MipBias {from=""}
   ]
   properties {samplerSource=SSM_Wrap_WorldGroupSettings, texture=&Apollo_Foliage_LUT_Green}
  }
  MaterialExpressionTextureSample_7 {
   class=MaterialExpressionTextureSample
   inputs [
    UVs {from=MaterialExpressionAppendVector_10}
    Tex {from=MaterialExpressionTextureObjectParameter_2}
    Apply View MipBias {from=""}
   ]
   properties {samplerSource=SSM_Wrap_WorldGroupSettings, texture=&Apollo_Foliage_LUT_Green}
  }
  MaterialExpressionTextureSample_8 {
   class=MaterialExpressionTextureSample
   inputs [
    UVs {from=MaterialExpressionAppendVector_11}
    Tex {from=MaterialExpressionTextureObjectParameter_2}
    Apply View MipBias {from=""}
   ]
   properties {samplerSource=SSM_Wrap_WorldGroupSettings, texture=&Apollo_Foliage_LUT_Green}
  }
  MaterialExpressionTransformPosition_0 {
   class=MaterialExpressionTransformPosition
   inputs [None {from=MaterialExpressionConstant3Vector_6, output=""}]
   properties {transformSourceType=TRANSFORMPOSSOURCE_Instance, transformType=TRANSFORMPOSSOURCE_World}
  }
  MaterialExpressionTransformPosition_1 {
   class=MaterialExpressionTransformPosition
   inputs [None {from=MaterialExpressionConstant3Vector_3, output=""}]
   properties {transformType=TRANSFORMPOSSOURCE_World}
  }
  MaterialExpressionTransform_13 {
   class=MaterialExpressionTransform
   inputs [None {from=MaterialExpressionSubtract_18}]
   properties {transformSourceType=TRANSFORMSOURCE_World, transformType=TRANSFORM_Tangent}
  }
  MaterialExpressionTransform_14 {
   class=MaterialExpressionTransform
   inputs [None {from=MaterialExpressionSubtract_17}]
   properties {transformSourceType=TRANSFORMSOURCE_World, transformType=TRANSFORM_Tangent}
  }
  MaterialExpressionTransform_15 {
   class=MaterialExpressionTransform
   inputs [None {from=MaterialExpressionSubtract_10}]
   properties {transformSourceType=TRANSFORMSOURCE_World, transformType=TRANSFORM_Tangent}
  }
  MaterialExpressionTransform_2 {class=MaterialExpressionTransform, inputs [None {from=MaterialExpressionConstantBiasScale_0}]}
  MaterialExpressionTransform_3 {
   class=MaterialExpressionTransform
   inputs [None {from=MaterialExpressionSubtract_5}]
   properties {transformSourceType=TRANSFORMSOURCE_World, transformType=TRANSFORM_Tangent}
  }
  MaterialExpressionVectorParameter_0 {class=MaterialExpressionVectorParameter, properties {parameterName=BaseColor Multiplier_Highlights, defaultValue=<rgba 1 1 1 1>}}
  MaterialExpressionVectorParameter_1 {class=MaterialExpressionVectorParameter, properties {parameterName=MobileZOffset, defaultValue=<rgba 0 0 550 0>}}
  MaterialExpressionVectorParameter_10 {
   class=MaterialExpressionVectorParameter
   properties {group=Color VertexColor Tints, parameterName=Green Vertex Channel Color, sortPriority=4, defaultValue=<rgba 1 1 1 1>}
  }
  MaterialExpressionVectorParameter_11 {class=MaterialExpressionVectorParameter, properties {parameterName=Emissive Overlay, defaultValue=<rgba 1 1 1 1>}}
  MaterialExpressionVectorParameter_12 {class=MaterialExpressionVectorParameter, properties {group=Color, parameterName=Color2_Lit, defaultValue=<rgba 0.2016 0.2195 0.0423 1>}}
  MaterialExpressionVectorParameter_13 {class=MaterialExpressionVectorParameter, properties {parameterName=Color, defaultValue=<rgba 1 0.4475 0.0933 5>}}
  MaterialExpressionVectorParameter_14 {class=MaterialExpressionVectorParameter, properties {parameterName=LightVector, defaultValue=<rgba -1 0 1 0>}}
  MaterialExpressionVectorParameter_15 {class=MaterialExpressionVectorParameter, properties {group=Color, parameterName=Color1_Base, defaultValue=<rgba 0.0887 0.1046 0.0152 1>}}
  MaterialExpressionVectorParameter_16 {class=MaterialExpressionVectorParameter, properties {group=Color, parameterName=Color3_Shadows, defaultValue=<rgba 0.0356 0.0612 0.0116 1>}}
  MaterialExpressionVectorParameter_17 {class=MaterialExpressionVectorParameter, properties {parameterName=ExtractColor, defaultValue=<rgba 0.0239 0.5 0.0255 1>}}
  MaterialExpressionVectorParameter_18 {class=MaterialExpressionVectorParameter, properties {parameterName=BaseColor Multiplier_Trunk, defaultValue=<rgba 1 1 1 1>}}
  MaterialExpressionVectorParameter_19 {class=MaterialExpressionVectorParameter, properties {bUseCustomPrimitiveData=true, parameterName=Color1_Base_primdata}}
  MaterialExpressionVectorParameter_2 {class=MaterialExpressionVectorParameter, properties {parameterName=Manual Weights, defaultValue=<rgba 0 0 1 1>}}
  MaterialExpressionVectorParameter_20 {
   class=MaterialExpressionVectorParameter
   properties {bUseCustomPrimitiveData=true, desc=20 not used, parameterName=Color2_Base_primdata, primitiveDataIndex=17, defaultValue=<rgba 0 0 0 1>}
  }
  MaterialExpressionVectorParameter_21 {class=MaterialExpressionVectorParameter, properties {parameterName=TrunkExtractColor, defaultValue=<rgba 1 1 1 1>}}
  MaterialExpressionVectorParameter_22 {class=MaterialExpressionVectorParameter, properties {parameterName=TrunkExtractColor, defaultValue=<rgba 1 1 1 1>}}
  MaterialExpressionVectorParameter_24 {class=MaterialExpressionVectorParameter, properties {parameterName=CustomTrunkColorImposterHack, defaultValue=<rgba 1 1 1 0>}}
  MaterialExpressionVectorParameter_3 {class=MaterialExpressionVectorParameter, properties {parameterName=ReferencePoint, defaultValue=<rgba 65995.8828 46048.0703 0 1>}}
  MaterialExpressionVectorParameter_4 {class=MaterialExpressionVectorParameter, properties {group=Color, parameterName=Color1_Base, defaultValue=<rgba 0.0887 0.1046 0.0152 1>}}
  MaterialExpressionVectorParameter_5 {class=MaterialExpressionVectorParameter, properties {parameterName=ReferencePoint_1, defaultValue=<rgba 88680.0391 35843.0195 0 1>}}
  MaterialExpressionVectorParameter_6 {class=MaterialExpressionVectorParameter, properties {parameterName=SubsurfaceColor, defaultValue=<rgba 1 1 1 5>}}
  MaterialExpressionVectorParameter_7 {class=MaterialExpressionVectorParameter, properties {parameterName=LightVector, defaultValue=<rgba -1 0 1 0>}}
  MaterialExpressionVectorParameter_8 {class=MaterialExpressionVectorParameter, properties {parameterName=BaseColor Multiplier, defaultValue=<rgba 1 1 1 1>}}
  MaterialExpressionVectorParameter_9 {class=MaterialExpressionVectorParameter, properties {bUseCustomPrimitiveData=true, parameterName=Color1_Base_primdata}}
  MaterialExpressionVertexColor_0 {class=MaterialExpressionVertexColor, inputs [], properties {}}
  MaterialExpressionVertexColor_1 {class=MaterialExpressionVertexColor, inputs [], properties {}}
  MaterialExpressionVertexColor_2 {class=MaterialExpressionVertexColor, inputs [], properties {}}
  MaterialExpressionVertexColor_3 {class=MaterialExpressionVertexColor, inputs [], properties {}}
  MaterialExpressionVertexInterpolator_0 {class=MaterialExpressionVertexInterpolator, inputs [VS {from=MaterialExpressionMaterialFunctionCall_24, output=UVs}], properties {}}
  MaterialExpressionVertexInterpolator_1 {class=MaterialExpressionVertexInterpolator, inputs [VS {from=MaterialExpressionMaterialFunctionCall_13, output=UVs}], properties {}}
  MaterialExpressionVertexInterpolator_10 {class=MaterialExpressionVertexInterpolator, inputs [VS {from=MaterialExpressionPerInstanceCustomData3Vector_0, output=""}], properties {}}
  MaterialExpressionVertexInterpolator_11 {class=MaterialExpressionVertexInterpolator, inputs [VS {from=MaterialExpressionPerInstanceCustomData_0, output=""}], properties {}}
  MaterialExpressionVertexInterpolator_2 {class=MaterialExpressionVertexInterpolator, inputs [VS {from=MaterialExpressionDivide_15, output=""}], properties {}}
  MaterialExpressionVertexInterpolator_3 {class=MaterialExpressionVertexInterpolator, inputs [VS {from=MaterialExpressionDivide_4, output=""}], properties {}}
  MaterialExpressionVertexInterpolator_4 {class=MaterialExpressionVertexInterpolator, inputs [VS {from=MaterialExpressionDivide_0, output=""}], properties {}}
  MaterialExpressionVertexInterpolator_5 {class=MaterialExpressionVertexInterpolator, inputs [VS {from=MaterialExpressionMultiply_1, output=""}], properties {}}
  MaterialExpressionVertexInterpolator_6 {class=MaterialExpressionVertexInterpolator, inputs [VS {from=MaterialExpressionMaterialFunctionCall_4, output=UVs}], properties {}}
  MaterialExpressionVertexInterpolator_7 {class=MaterialExpressionVertexInterpolator, inputs [VS {from=MaterialExpressionDivide_14, output=""}], properties {}}
  MaterialExpressionVertexInterpolator_8 {class=MaterialExpressionVertexInterpolator, inputs [VS {from="", output=""}], properties {}}
  MaterialExpressionVertexInterpolator_9 {class=MaterialExpressionVertexInterpolator, inputs [VS {from=MaterialExpressionDivide_19, output=""}], properties {}}
  MaterialExpressionVertexNormalWS_0 {class=MaterialExpressionVertexNormalWS, inputs [], properties {}}
  MaterialExpressionWorldPosition_0 {class=MaterialExpressionWorldPosition, inputs [], properties {worldPositionShaderOffset=WPT_ExcludeAllShaderOffsets}}
  MaterialExpressionWorldPosition_1 {class=MaterialExpressionWorldPosition, inputs [], properties {}}
  MaterialExpressionWorldPosition_2 {class=MaterialExpressionWorldPosition, inputs [], properties {}}
  MaterialExpressionWorldPosition_3 {
   class=MaterialExpressionWorldPosition
   inputs []
   properties {desc=Imposter is tiny and matches almost actor location, worldPositionShaderOffset=WPT_ExcludeAllShaderOffsets}
  }
  MaterialExpressionWorldPosition_4 {
   class=MaterialExpressionWorldPosition
   inputs []
   properties {desc=Imposter is tiny and matches almost actor location, worldPositionShaderOffset=WPT_ExcludeAllShaderOffsets}
  }
 }
 properties {
  bAlwaysEvaluateWorldPositionOffset=true
  bTangentSpaceNormal=false
  bUseMaterialAttributes=true
  bUsedWithInstancedStaticMeshes=true
  bUsedWithNanite=true
  bUsedWithStaticLighting=true
  blendMode=BLEND_Masked
  ditheredLODTransition=true
  opacityMaskClipValue=0.5
  pixelDepthOffsetMode=PDOM_Legacy
  twoSided=true
  usedShadingModels=Default Lit
 }
}
