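# Unlit material for the impostor baker that computes a normalized depth value of an object along a configurable axis.
# World position minus ComponentOrigin goes through an inverse transform built from the X, Y and Z basis vector parameters.
# The result is divided by scalar R and bias-scaled, then its Z component is saturated and inverted, so depth falls as position along that axis rises.
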Material /Game/ImpostorBaker/Materials/M_Imposter_Depth_01
graph {
 properties {bUsedWithStaticLighting=true, pixelDepthOffsetMode=PDOM_Legacy, shadingModel=MSM_Unlit, usedShadingModels=Unlit}
 nodes [
  n1 = Subtract(A=WorldPosition().XYZ, B=VectorParameter() {parameterName=ComponentOrigin}.RGB)
  n2 = VectorParameter() {parameterName=X, defaultValue=<rgba 1 0 0 0>}
  n3 = VectorParameter() {parameterName=Y, defaultValue=<rgba 0 1 0 0>}
  n4 = VectorParameter() {parameterName=Z, defaultValue=<rgba 0 0 1 0>}
  n5 = ConstantBiasScale(None=Divide(
    A=MaterialFunctionCall!(VectorToTransform=@n1, BasisX=@n2.RGB, BasisY=@n3.RGB, BasisZ=@n4.RGB) {materialFunction=/Engine/Functions/Engine_MaterialFunctions02/Math/InverseTransformMatrix}.Transformed Vector 3
    B=ScalarParameter() {defaultValue=256, parameterName=R}
   ))
  OneMinus(None=Saturate(None=ComponentMask(None=@n5) {b=true}))
 ]
}
